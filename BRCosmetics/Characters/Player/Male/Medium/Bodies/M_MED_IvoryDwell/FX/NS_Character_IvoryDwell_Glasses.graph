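# Niagara system for the IvoryDwell character's glasses effect, with a single FN_Mesh emitter that renders a mesh particle.
# An instantaneous burst spawns particles that do not die when their lifetime elapses.
# SocketLocation positions them through a system-level socket reader with a sampled position offset.
# ScaleColor multiplies alpha by User.Reactivity, zeroing it when User.bIsHeadHidden is true, so reactivity and head visibility drive the effect.
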NiagaraSystem /BRCosmetics/Characters/Player/Male/Medium/Bodies/M_MED_IvoryDwell/FX/NS_Character_IvoryDwell_Glasses
dependencies {
 usedDataInterfaces [&NiagaraDataInterfaceSocketReader]
 usedDynamicInputs []
 usedModules [
  &self:SystemSpawnScript.SystemScriptSource.SystemScriptGraph.NiagaraNodeAssignment_0.SetVariables_#1
  &SystemState
 ]
}
emitters {FN_Mesh {
  emitterUpdateScript [
   &EmitterState(
    Inactive Response=::NewEnumerator1
    Life Cycle Mode=::NewEnumerator1
    Loop Behavior=::NewEnumerator1
    Loop Duration Mode=::NewEnumerator1
    UseLoopDelay=false
   )
   &SpawnBurst_Instantaneous(Use Spawn Probability=false, UseLoopCountLimit=false)
  ]
  particleSpawnScript [&InitializeParticle(UsePositionOffset=false, Write Material Random=true)]
  particleUpdateScript [
   &ParticleState(Kill Particles When Lifetime Has Elapsed=false)
   &SocketLocation(
    SamplePosOffset=true
    Sampled Position Offset=<8 9 0>
    Socket Reader=$System.Socket Reader
    Socket Sampling Mode=::NewEnumerator3
    VelocitySpeedLimit=false
   )
   &ScaleColor(
    Scale Alpha=&Multiply_Float(A=$User.Reactivity, B=&MakeCustomFloatFromBool(Bool=$User.bIsHeadHidden, True Float=0, False Float=1))
    ScaleA=true
    ScaleRGB=true
   )
  ]
  rendererClasses [&NiagaraMeshRendererProperties]
  renderers [{rendererClass=&NiagaraMeshRendererProperties, rendererIndex=0}]
 }}
summary {
 systemName=NS_Character_IvoryDwell_Glasses
 emitters [{emitterName=FN_Mesh, rendererClasses [&NiagaraMeshRendererProperties]}]
 userVariables [
  User.bIsHeadHidden {type=NiagaraBool}
  User.Reactivity {type=NiagaraFloat, defaultValue=0, description=""}
 ]
}
