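# Physics asset for the hanging carabiners on the IvoryDwell medium male player body cosmetic.
# Two chains (F and S) each run from a base body through two dyn bodies, with swing limits and zero twist, the end body using paired capsules.
# Capsules on the pelvis, left thigh twist and two left strap bones form the surrounding body shapes.
# These bodies and both chain bases connect to the pelvis with angularly free constraints.
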PhysicsAsset /BRCosmetics/Characters/Player/Male/Medium/Bodies/M_MED_IvoryDwell/Meshes/M_MED_IvoryDwell_Carabiners_Physics
bodies {
 base_hangCarabinerF_1 [{shapeType=Capsule, length=0.5, radius=0.25, shapeName=base_hangCarabinerF_1_capsule}]
 base_hangCarabinerS_1 [{shapeType=Capsule, length=0.5, radius=0.25, shapeName=base_hangCarabinerS_1_capsule}]
 coll_strap_01_l [{shapeType=Capsule, length=15.0076, radius=2.8483, shapeName=coll_strap_01_l_capsule, center=<7.3875 -1 0>, rotation=<rot -90 0 0>}]
 coll_strap_02_l [{shapeType=Capsule, length=14.9091, radius=2.9467, shapeName=coll_strap_01_l_capsule, center=<7.3875 0 0>, rotation=<rot -90 0 0>}]
 dyn_hangCarabinerF_1 [{
   shapeType=Capsule
   length=4.6273
   radius=0.7003
   shapeName=dyn_hangCarabinerF_1_capsule
   center=<2.5935 0.0406 0.0031>
   rotation=<rot 0 -89.2893 -84.6063>
  }]
 dyn_hangCarabinerF_2 [
  {
   shapeType=Capsule
   length=6.0488
   radius=1.2008
   shapeName=dyn_hangCarabinerF_2_capsule
   center=<3.9162 -2.2033 0.3126>
   rotation=<rot 90 0.3763 0>
  }
  {
   shapeType=Capsule
   length=8.0068
   radius=1.2746
   shapeName=dyn_hangCarabinerF_2_capsule
   center=<3.9162 1.1902 0.1654>
   rotation=<rot 90 2.6049 0>
  }
 ]
 dyn_hangCarabinerS_1 [{
   shapeType=Capsule
   length=6.7427
   radius=0.7181
   shapeName=dyn_hangCarabinerS_1_capsule
   center=<3.39 -0.1128 -0.1561>
   rotation=<rot 0 91.5232 -93.5174>
  }]
 dyn_hangCarabinerS_2 [
  {
   shapeType=Capsule
   length=5.9937
   radius=1.2559
   shapeName=dyn_hangCarabinerF_2_capsule
   center=<3.9162 -2.2033 0.4769>
   rotation=<rot 90 0 0>
  }
  {shapeType=Capsule, length=8.2814, radius=1, shapeName=dyn_hangCarabinerF_2_capsule, center=<3.9162 0.8782 0.1654>, rotation=<rot 90 0 0>}
 ]
 pelvis [{shapeType=Capsule, length=12.6666, radius=11.2684, shapeName=pelvis_capsule}]
 thigh_twist_01_l [{
   shapeType=Capsule
   length=22.5075
   radius=10.7771
   shapeName=thigh_l_capsule
   center=<4.878 -1.4153 1.025>
   rotation=<rot 88.6733 179.9998 179.9998>
  }]
}
constraints [
 {bone1Name=dyn_hangCarabinerF_2, bone2Name=dyn_hangCarabinerF_1, swing1LimitDegrees=10, swing2LimitDegrees=40, twistLimitDegrees=0}
 {bone1Name=dyn_hangCarabinerF_1, bone2Name=base_hangCarabinerF_1, swing1LimitDegrees=30, swing2LimitDegrees=30, twistLimitDegrees=0}
 {
  bone1Name=base_hangCarabinerF_1
  bone2Name=pelvis
  swing1LimitDegrees=45
  swing1Motion=Free
  swing2LimitDegrees=45
  swing2Motion=Free
  twistLimitDegrees=45
  twistMotion=Free
 }
 {bone1Name=dyn_hangCarabinerS_2, bone2Name=dyn_hangCarabinerS_1, swing1LimitDegrees=10, swing2LimitDegrees=30, twistLimitDegrees=0}
 {bone1Name=dyn_hangCarabinerS_1, bone2Name=base_hangCarabinerS_1, swing1LimitDegrees=30, swing2LimitDegrees=30, twistLimitDegrees=0}
 {
  bone1Name=base_hangCarabinerS_1
  bone2Name=pelvis
  swing1LimitDegrees=45
  swing1Motion=Free
  swing2LimitDegrees=45
  swing2Motion=Free
  twistLimitDegrees=45
  twistMotion=Free
 }
 {
  bone1Name=thigh_twist_01_l
  bone2Name=pelvis
  swing1LimitDegrees=45
  swing1Motion=Free
  swing2LimitDegrees=45
  swing2Motion=Free
  twistLimitDegrees=45
  twistMotion=Free
 }
 {
  bone1Name=coll_strap_02_l
  bone2Name=pelvis
  swing1LimitDegrees=45
  swing1Motion=Free
  swing2LimitDegrees=45
  swing2Motion=Free
  twistLimitDegrees=45
  twistMotion=Free
 }
 {
  bone1Name=coll_strap_01_l
  bone2Name=pelvis
  swing1LimitDegrees=45
  swing1Motion=Free
  swing2LimitDegrees=45
  swing2Motion=Free
  twistLimitDegrees=45
  twistMotion=Free
 }
]
