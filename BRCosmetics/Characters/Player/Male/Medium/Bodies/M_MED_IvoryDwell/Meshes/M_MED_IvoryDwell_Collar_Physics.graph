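# Physics asset for the collar mesh of the IvoryDwell medium male player body cosmetic, covering its hoodie and jacket collar bones.
# Each collar segment pairs a tiny capsule on a base bone with a larger, offset capsule on its dyn bone.
# A constraint joins each pair, locking twist and one swing axis and limiting the other swing to a narrow hinge-like arc.
# Two capsules on the head bone add head collision volume.
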PhysicsAsset /BRCosmetics/Characters/Player/Male/Medium/Bodies/M_MED_IvoryDwell/Meshes/M_MED_IvoryDwell_Collar_Physics
bodies {
 base_hoodieCollar_1_c [{shapeType=Capsule, length=0.5, radius=0.25, shapeName=base_hoodieCollar_1_c_capsule}]
 base_hoodieCollar_1_l [{shapeType=Capsule, length=0.5, radius=0.25, shapeName=base_hoodieCollar_1_l_capsule}]
 base_hoodieCollar_1_r [{shapeType=Capsule, length=0.5, radius=0.25, shapeName=base_hoodieCollar_1_r_capsule}]
 base_hoodieCollar_2_l [{shapeType=Capsule, length=0.5, radius=0.25, shapeName=base_hoodieCollar_2_l_capsule}]
 base_hoodieCollar_2_r [{shapeType=Capsule, length=0.5, radius=0.25, shapeName=base_hoodieCollar_2_r_capsule}]
 base_hoodieCollar_3_l [{shapeType=Capsule, length=0.5, radius=0.25, shapeName=base_hoodieCollar_3_l_capsule}]
 base_hoodieCollar_3_r [{shapeType=Capsule, length=0.5, radius=0.25, shapeName=base_hoodieCollar_3_r_capsule}]
 base_jacketCollar_1_l [{shapeType=Capsule, length=0.5, radius=0.25, shapeName=base_jacketCollar_1_l_capsule}]
 base_jacketCollar_1_r [{shapeType=Capsule, length=0.5, radius=0.25, shapeName=base_jacketCollar_1_l_capsule}]
 dyn_hoodieCollar_1_c [{shapeType=Capsule, length=4.1659, radius=3.9811, shapeName=dyn_hoodieCollar_1_c_capsule, center=<7.4791 0 -0.5085>, rotation=<rot 0 0 90>}]
 dyn_hoodieCollar_1_l [{
   shapeType=Capsule
   length=8.2022
   radius=3.917
   shapeName=dyn_hoodieCollar_1_l_capsule
   center=<5.6819 0.0449 -0.942>
   rotation=<rot 0 7.7914 90>
  }]
 dyn_hoodieCollar_1_r [{
   shapeType=Capsule
   length=8.2022
   radius=3.917
   shapeName=dyn_hoodieCollar_1_r_capsule
   center=<5.8305 -0.0653 -1.1774>
   rotation=<rot 0 -7.7914 -90>
  }]
 dyn_hoodieCollar_2_l [{shapeType=Capsule, length=8.3573, radius=2.1391, shapeName=dyn_hoodieCollar_2_l_capsule, center=<3.78 1 -0.65>, rotation=<rot 0 -150 -90>}]
 dyn_hoodieCollar_2_r [{shapeType=Capsule, length=8.3573, radius=2.1391, shapeName=dyn_hoodieCollar_2_l_capsule, center=<3.78 -1 -0.65>, rotation=<rot 0 150 -90>}]
 dyn_hoodieCollar_3_l [{shapeType=Capsule, length=3.9368, radius=2.0632, shapeName=dyn_hoodieCollar_3_l_capsule, center=<2.0573 0 -0.4272>, rotation=<rot 90 0 0>}]
 dyn_hoodieCollar_3_r [{shapeType=Capsule, length=3.9368, radius=2.0632, shapeName=dyn_hoodieCollar_3_r_capsule, center=<2.0573 0 -0.4272>, rotation=<rot 90 180 0>}]
 dyn_jacketCollar_1_l [{
   shapeType=Capsule
   length=9.5864
   radius=3.5335
   shapeName=dyn_jacketCollar_1_l_capsule
   center=<5.8994 0.894 3.1582>
   rotation=<rot 0.9579 -177.2272 -94.1647>
  }]
 dyn_jacketCollar_1_r [{
   shapeType=Capsule
   length=9.5864
   radius=3.5335
   shapeName=dyn_jacketCollar_1_r_capsule
   center=<6.4237 0.0478 1.3686>
   rotation=<rot 0.9579 -177.2272 85.8353>
  }]
 head [
  {shapeType=Capsule, length=2.0228, radius=8.7543, shapeName=head_capsule, center=<9.157 1.4691 0>, rotation=<rot 90 28.6817 0>}
  {shapeType=Capsule, length=11.6545, radius=4.9216, shapeName=head_capsule, center=<3.1438 3.081 0>, rotation=<rot 90 -34.3575 0>}
 ]
}
constraints [
 {bone1Name=dyn_hoodieCollar_1_c, bone2Name=base_hoodieCollar_1_c, swing1LimitDegrees=0, swing2LimitDegrees=20, twistLimitDegrees=0}
 {bone1Name=dyn_hoodieCollar_1_l, bone2Name=base_hoodieCollar_1_l, swing1LimitDegrees=0, swing2LimitDegrees=20, twistLimitDegrees=0}
 {bone1Name=dyn_hoodieCollar_1_r, bone2Name=base_hoodieCollar_1_r, swing1LimitDegrees=0, swing2LimitDegrees=20, twistLimitDegrees=0}
 {bone1Name=dyn_hoodieCollar_2_l, bone2Name=base_hoodieCollar_2_l, swing1LimitDegrees=0, swing2LimitDegrees=20, twistLimitDegrees=0}
 {bone1Name=dyn_hoodieCollar_2_r, bone2Name=base_hoodieCollar_2_r, swing1LimitDegrees=0, swing2LimitDegrees=20, twistLimitDegrees=0}
 {bone1Name=dyn_hoodieCollar_3_l, bone2Name=base_hoodieCollar_3_l, swing1LimitDegrees=0, swing2LimitDegrees=20, twistLimitDegrees=0}
 {bone1Name=dyn_hoodieCollar_3_r, bone2Name=base_hoodieCollar_3_r, swing1LimitDegrees=0, swing2LimitDegrees=20, twistLimitDegrees=0}
 {bone1Name=dyn_jacketCollar_1_l, bone2Name=base_jacketCollar_1_l, swing1LimitDegrees=0, swing2LimitDegrees=20, twistLimitDegrees=0}
 {bone1Name=dyn_jacketCollar_1_r, bone2Name=base_jacketCollar_1_r, swing1LimitDegrees=0, swing2LimitDegrees=20, twistLimitDegrees=0}
]
