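# Control Rig for the IvoryDwell male character cosmetic, driving body deform bones and secondary clothing bones.
# Setup spawns reference nulls for key limb bones, aim targets for hoodie collar and jacket hem, and hip-chain helper bones.
# Forward solve runs FortDeformRig, a lower-body function, then hoodie setup.
# Weighted parent and aim constraints place collar bones, while spherical pose readers on head and right upper arm blend right-side corrective offsets.
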ControlRigBlueprint /BRCosmetics/Characters/Player/Male/Medium/Bodies/M_MED_IvoryDwell/Meshes/M_MED_IvoryDwell_CtrlRig
graphs [
 CollapseNode_ContainedGraph {nodes [
   n1 = RigVMFunctionEntryNode()
   RigVMFunction_Sequence = RigVMAggregateNode(ExecuteContext=@n1.ExecutePin)
   &RigUnit_ParentConstraint(
    AdvancedSettings=(RotationOrderForFilter=XZY)
    Child=(Type=Bone,Name="base_jacketCollar_1_l")
    Parents=((Item=(Type=Bone,Name="spine_03"),Weight=0.0209),(Item=(Type=Bone,Name="spine_04"),Weight=0.0517),(Item=(Type=Bone,Name="deform_pec_l"),Weight=0.0725),(Item=(Type=Bone,Name="clavicle_l"),Weight=0.3492),(Item=(Type=Bone,Name="spine_05"),Weight=0.5057))
    Weight=1
    ExecutePin=@RigVMFunction_Sequence.A
   )
   > &RigUnit_ParentConstraint(
    AdvancedSettings=(RotationOrderForFilter=XZY)
    Child=(Type=Bone,Name="base_jacketCollar_1_r")
    Parents=((Item=(Type=Bone,Name="spine_03"),Weight=0.0209),(Item=(Type=Bone,Name="spine_04"),Weight=0.0517),(Item=(Type=Bone,Name="deform_pec_r"),Weight=0.0725),(Item=(Type=Bone,Name="clavicle_r"),Weight=0.3493),(Item=(Type=Bone,Name="spine_05"),Weight=0.5057))
    Weight=1
   )
   > n2 = &RigUnit_SphericalPoseReader(
    ActiveRegionScaleFactors=(PositiveWidth=1.5,NegativeWidth=1,PositiveHeight=1.5,NegativeHeight=1.5)
    ActiveRegionSize=0.1
    Debug=(bDrawDebug=false,bDraw2D=False,bDrawLocalAxes=False,DebugScale=25,DebugSegments=20,DebugThickness=0.25)
    DriverAxis=(X=-1,Y=0,Z=0)
    DriverItem=(Type=Bone,Name="upperarm_r")
    FalloffRegionScaleFactors=(PositiveWidth=1.5,NegativeWidth=1,PositiveHeight=1,NegativeHeight=1)
    FalloffSize=0.2
    FlipHeightScaling=False
    FlipWidthScaling=False
    OptionalParentItem=(Type=Bone,Name="None")
    RotationOffset=(X=-93.7294,Y=25.4067,Z=35.8226)
   )
   > &RigUnit_ModifyTransforms(
    ItemToModify=((Item=(Type=Bone,Name="base_jacketCollar_1_r"),Transform=(Rotation=(X=0.0623,Y=0.3149,Z=-0.0502,W=0.9458),Translation=(X=0.6,Y=0,Z=0.75),Scale3D=(X=1,Y=1,Z=1))))
    WeightMinimum=0
    Weight=@n2.OutputParam
   )
   constraints = RigVMCollapseNode(ExecutePin=@RigVMFunction_Sequence.B)
   n3 = &RigUnit_SphericalPoseReader(
    ActiveRegionScaleFactors=(PositiveWidth=1,NegativeWidth=1,PositiveHeight=1,NegativeHeight=1)
    ActiveRegionSize=0.08
    Debug=(bDrawDebug=false,bDraw2D=False,bDrawLocalAxes=False,DebugScale=25,DebugSegments=20,DebugThickness=0.25)
    DriverAxis=(X=1,Y=0,Z=0)
    DriverItem=(Type=Bone,Name="head")
    FalloffRegionScaleFactors=(PositiveWidth=1.2,NegativeWidth=1,PositiveHeight=1,NegativeHeight=1.5)
    FalloffSize=0.2
    FlipHeightScaling=False
    FlipWidthScaling=False
    OptionalParentItem=(Type=Bone,Name="spine_05")
    RotationOffset=(X=-59.254,Y=-13.2863,Z=21.2653)
    ExecutePin=@constraints.ExecutePin
   )
   > &RigUnit_ModifyTransforms(
    ItemToModify=((Item=(Type=Bone,Name="base_hoodieCollar_2_r"),Transform=(Rotation=(X=0,Y=0.1736,Z=0,W=0.9848),Translation=(X=0,Y=0,Z=0),Scale3D=(X=1,Y=1,Z=1))))
    WeightMinimum=0
    Weight=@n3.OutputParam
   )
   > n4 = &RigUnit_SphericalPoseReader(
    ActiveRegionScaleFactors=(PositiveWidth=1.3,NegativeWidth=0.5,PositiveHeight=1,NegativeHeight=1.3)
    ActiveRegionSize=0.08
    Debug=(bDrawDebug=false,bDraw2D=False,bDrawLocalAxes=False,DebugScale=25,DebugSegments=20,DebugThickness=0.25)
    DriverAxis=(X=1,Y=0,Z=0)
    DriverItem=(Type=Bone,Name="head")
    FalloffRegionScaleFactors=(PositiveWidth=1.2,NegativeWidth=1,PositiveHeight=1,NegativeHeight=1.5)
    FalloffSize=0.2
    FlipHeightScaling=False
    FlipWidthScaling=False
    OptionalParentItem=(Type=Bone,Name="spine_05")
    RotationOffset=(X=-67.4776,Y=-0.6199,Z=-0.6176)
   )
   > &RigUnit_ModifyTransforms(
    ItemToModify="((Item=(Type=Bone,Name=\"base_hoodieCollar_1_r\"),Transform=(Rotation=(X=0,Y=0.1736,Z=0,W=0.9848),Translation=(X=0,Y=0,Z=0),Scale3D=(X=1,Y=1,Z=1))),(Item=(Type=Bone,Name=\"base_hoodieCollar_2_r\"),Transform=(Rotation=(X=0,Y=0.061,Z=0,W=0.9981),Translation=(X=0 ...(truncated)"
    WeightMinimum=0
    Weight=@n4.OutputParam
   )
   RigVMFunctionReturnNode(ExecutePin=@RigVMFunction_Sequence.C)
   &RigUnit_ModifyTransforms(
    ExecutePin=()
    ItemToModify="((Item=(Type=Bone,Name=\"base_hoodieCollar_1_c\"),Transform=(Rotation=(X=0,Y=0.1045,Z=0,W=0.9945),Translation=(X=0,Y=0,Z=0),Scale3D=(X=1,Y=1,Z=1))),(Item=(Type=Bone,Name=\"base_hoodieCollar_1_r\"),Transform=(Rotation=(X=0,Y=0.1045,Z=0,W=0.9945),Translation=(X= ...(truncated)"
    Weight=1
    WeightMinimum=0
   )
  ]}
 CollapseNode_ContainedGraph {nodes [
   n1 = RigVMFunctionEntryNode()
   &RigUnit_ParentConstraint(
    AdvancedSettings=(RotationOrderForFilter=XZY)
    Child=(Type=Bone,Name="base_hoodieCollar_1_c")
    Filter=(TranslationFilter=(bX=True,bY=True,bZ=True),RotationFilter=(bX=False,bY=False,bZ=False),ScaleFilter=(bX=False,bY=False,bZ=False))
    Parents=((Item=(Type=Bone,Name="clavicle_l"),Weight=0.0319),(Item=(Type=Bone,Name="clavicle_r"),Weight=0.0319),(Item=(Type=Bone,Name="spine_05"),Weight=0.9361))
    Weight=1
    ExecutePin=@n1.ExecutePin
   )
   > &RigUnit_ParentConstraint(
    AdvancedSettings=(RotationOrderForFilter=XZY)
    Child=(Type=Bone,Name="base_hoodieCollar_3_l")
    Filter=(TranslationFilter=(bX=True,bY=True,bZ=True),RotationFilter=(bX=False,bY=False,bZ=False),ScaleFilter=(bX=False,bY=False,bZ=False))
    Parents=((Item=(Type=Bone,Name="clavicle_l"),Weight=0.158),(Item=(Type=Bone,Name="spine_05"),Weight=0.842))
    Weight=1
   )
   > &RigUnit_ParentConstraint(
    AdvancedSettings=(RotationOrderForFilter=XZY)
    Child=(Type=Bone,Name="base_hoodieCollar_2_l")
    Filter=(TranslationFilter=(bX=True,bY=True,bZ=True),RotationFilter=(bX=False,bY=False,bZ=False),ScaleFilter=(bX=False,bY=False,bZ=False))
    Parents=((Item=(Type=Bone,Name="clavicle_l"),Weight=0.2021),(Item=(Type=Bone,Name="spine_05"),Weight=0.7979))
    Weight=1
   )
   > &RigUnit_ParentConstraint(
    AdvancedSettings=(RotationOrderForFilter=XZY)
    Child=(Type=Bone,Name="base_hoodieCollar_1_l")
    Filter=(TranslationFilter=(bX=True,bY=True,bZ=True),RotationFilter=(bX=False,bY=False,bZ=False),ScaleFilter=(bX=False,bY=False,bZ=False))
    Parents=((Item=(Type=Bone,Name="clavicle_l"),Weight=0.2775),(Item=(Type=Bone,Name="spine_05"),Weight=0.7225))
    Weight=1
   )
   > &RigUnit_ParentConstraint(
    AdvancedSettings=(RotationOrderForFilter=XZY)
    Child=(Type=Bone,Name="base_hoodieCollar_3_r")
    Filter=(TranslationFilter=(bX=True,bY=True,bZ=True),RotationFilter=(bX=False,bY=False,bZ=False),ScaleFilter=(bX=False,bY=False,bZ=False))
    Parents=((Item=(Type=Bone,Name="clavicle_r"),Weight=0.158),(Item=(Type=Bone,Name="spine_05"),Weight=0.842))
    Weight=1
   )
   > &RigUnit_ParentConstraint(
    AdvancedSettings=(RotationOrderForFilter=XZY)
    Child=(Type=Bone,Name="base_hoodieCollar_2_r")
    Filter=(TranslationFilter=(bX=True,bY=True,bZ=True),RotationFilter=(bX=False,bY=False,bZ=False),ScaleFilter=(bX=False,bY=False,bZ=False))
    Parents=((Item=(Type=Bone,Name="clavicle_r"),Weight=0.2021),(Item=(Type=Bone,Name="spine_05"),Weight=0.7979))
    Weight=1
   )
   > &RigUnit_ParentConstraint(
    AdvancedSettings=(RotationOrderForFilter=XZY)
    Child=(Type=Bone,Name="base_hoodieCollar_1_r")
    Filter=(TranslationFilter=(bX=True,bY=True,bZ=True),RotationFilter=(bX=False,bY=False,bZ=False),ScaleFilter=(bX=False,bY=False,bZ=False))
    Parents=((Item=(Type=Bone,Name="clavicle_r"),Weight=0.2775),(Item=(Type=Bone,Name="spine_05"),Weight=0.7225))
    Weight=1
   )
   > &RigUnit_ParentConstraint(
    AdvancedSettings=(RotationOrderForFilter=XZY)
    Child=(Type=Null,Name="target_hoodieCollar_1_c_null")
    Filter=(TranslationFilter=(bX=True,bY=True,bZ=True),RotationFilter=(bX=False,bY=False,bZ=False),ScaleFilter=(bX=False,bY=False,bZ=False))
    Parents=((Item=(Type=Bone,Name="clavicle_l"),Weight=0.0332),(Item=(Type=Bone,Name="clavicle_r"),Weight=0.0332),(Item=(Type=Bone,Name="spine_05"),Weight=0.9337))
    Weight=1
   )
   > &RigUnit_ParentConstraint(
    AdvancedSettings=(RotationOrderForFilter=XZY)
    Child=(Type=Null,Name="target_hoodieCollar_3_l_null")
    Filter=(TranslationFilter=(bX=True,bY=True,bZ=True),RotationFilter=(bX=False,bY=False,bZ=False),ScaleFilter=(bX=False,bY=False,bZ=False))
    Parents=((Item=(Type=Bone,Name="clavicle_l"),Weight=0.2568),(Item=(Type=Bone,Name="spine_05"),Weight=0.7432))
    Weight=1
   )
   > &RigUnit_ParentConstraint(
    AdvancedSettings=(RotationOrderForFilter=XZY)
    Child=(Type=Null,Name="target_hoodieCollar_2_l_null")
    Filter=(TranslationFilter=(bX=True,bY=True,bZ=True),RotationFilter=(bX=False,bY=False,bZ=False),ScaleFilter=(bX=False,bY=False,bZ=False))
    Parents=((Item=(Type=Bone,Name="clavicle_l"),Weight=0.3948),(Item=(Type=Bone,Name="spine_05"),Weight=0.6052))
    Weight=1
   )
   > &RigUnit_ParentConstraint(
    AdvancedSettings=(RotationOrderForFilter=XZY)
    Child=(Type=Null,Name="target_hoodieCollar_1_l_null")
    Filter=(TranslationFilter=(bX=True,bY=True,bZ=True),RotationFilter=(bX=False,bY=False,bZ=False),ScaleFilter=(bX=False,bY=False,bZ=False))
    Parents=((Item=(Type=Bone,Name="clavicle_l"),Weight=0.2689),(Item=(Type=Bone,Name="spine_05"),Weight=0.7311))
    Weight=1
   )
   > &RigUnit_ParentConstraint(
    AdvancedSettings=(RotationOrderForFilter=XZY)
    Child=(Type=Null,Name="target_hoodieCollar_3_r_null")
    Filter=(TranslationFilter=(bX=True,bY=True,bZ=True),RotationFilter=(bX=False,bY=False,bZ=False),ScaleFilter=(bX=False,bY=False,bZ=False))
    Parents=((Item=(Type=Bone,Name="clavicle_r"),Weight=0.2568),(Item=(Type=Bone,Name="spine_05"),Weight=0.7432))
    Weight=1
   )
   > &RigUnit_ParentConstraint(
    AdvancedSettings=(RotationOrderForFilter=XZY)
    Child=(Type=Null,Name="target_hoodieCollar_2_r_null")
    Filter=(TranslationFilter=(bX=True,bY=True,bZ=True),RotationFilter=(bX=False,bY=False,bZ=False),ScaleFilter=(bX=False,bY=False,bZ=False))
    Parents=((Item=(Type=Bone,Name="clavicle_r"),Weight=0.3948),(Item=(Type=Bone,Name="spine_05"),Weight=0.6052))
    Weight=1
   )
   > &RigUnit_ParentConstraint(
    AdvancedSettings=(RotationOrderForFilter=XZY)
    Child=(Type=Null,Name="target_hoodieCollar_1_r_null")
    Filter=(TranslationFilter=(bX=True,bY=True,bZ=True),RotationFilter=(bX=False,bY=False,bZ=False),ScaleFilter=(bX=False,bY=False,bZ=False))
    Parents=((Item=(Type=Bone,Name="clavicle_r"),Weight=0.2689),(Item=(Type=Bone,Name="spine_05"),Weight=0.7311))
    Weight=1
   )
   > internalName0 = &RigUnit_AimConstraintLocalSpaceOffset(
    Child=(Type=Bone,Name="base_hoodieCollar_3_l")
    Filter=(bX=True,bY=True,bZ=True)
    Parents=((Item=(Type=Null,Name="target_hoodieCollar_3_l_null"),Weight=1))
    UpAxis=(X=0,Y=-1,Z=0)
    Weight=1
    WorldUp=(Target=(X=0.387,Y=-0.5905,Z=-0.7082),Kind=Direction,Space=(Type=Bone,Name="spine_05"))
    bMaintainOffset=False
   )
   > internalName0_1 = &RigUnit_AimConstraintLocalSpaceOffset(
    Child=(Type=Bone,Name="base_hoodieCollar_2_l")
    Filter=(bX=True,bY=True,bZ=True)
    Parents=((Item=(Type=Null,Name="target_hoodieCollar_2_l_null"),Weight=1))
    UpAxis=(X=0,Y=-1,Z=0)
    Weight=1
    WorldUp=(Target=(X=0.2025,Y=-0.9787,Z=-0.0337),Kind=Direction,Space=(Type=Bone,Name="spine_05"))
    bMaintainOffset=False
   )
   > internalName0_2 = &RigUnit_AimConstraintLocalSpaceOffset(
    Child=(Type=Bone,Name="base_hoodieCollar_1_l")
    Filter=(bX=True,bY=True,bZ=True)
    Parents=((Item=(Type=Null,Name="target_hoodieCollar_1_l_null"),Weight=1))
    UpAxis=(X=0,Y=-1,Z=0)
    Weight=1
    WorldUp=(Target=(X=0.0482,Y=-0.635,Z=0.771),Kind=Direction,Space=(Type=Bone,Name="spine_05"))
    bMaintainOffset=False
   )
   > internalName0_3 = &RigUnit_AimConstraintLocalSpaceOffset(
    Child=(Type=Bone,Name="base_hoodieCollar_1_c")
    Filter=(bX=True,bY=True,bZ=True)
    Parents=((Item=(Type=Null,Name="target_hoodieCollar_1_c_null"),Weight=1))
    UpAxis=(X=0,Y=-1,Z=0)
    Weight=1
    WorldUp=(Target=(X=0,Y=0,Z=1),Kind=Direction,Space=(Type=Bone,Name="spine_05"))
    bMaintainOffset=False
   )
   > internalName0_4 = &RigUnit_AimConstraintLocalSpaceOffset(
    Child=(Type=Bone,Name="base_hoodieCollar_1_r")
    Filter=(bX=True,bY=True,bZ=True)
    Parents=((Item=(Type=Null,Name="target_hoodieCollar_1_r_null"),Weight=1))
    UpAxis=(X=0,Y=-1,Z=0)
    Weight=1
    WorldUp=(Target=(X=-0.0482,Y=0.635,Z=0.771),Kind=Direction,Space=(Type=Bone,Name="spine_05"))
    bMaintainOffset=false
   )
   > internalName0_5 = &RigUnit_AimConstraintLocalSpaceOffset(
    Child=(Type=Bone,Name="base_hoodieCollar_2_r")
    Filter=(bX=True,bY=True,bZ=True)
    Parents=((Item=(Type=Null,Name="target_hoodieCollar_2_r_null"),Weight=1))
    UpAxis=(X=0,Y=-1,Z=0)
    Weight=1
    WorldUp=(Target=(X=-0.2025,Y=0.9787,Z=-0.0337),Kind=Direction,Space=(Type=Bone,Name="spine_05"))
    bMaintainOffset=False
   )
   > internalName0_6 = &RigUnit_AimConstraintLocalSpaceOffset(
    Child=(Type=Bone,Name="base_hoodieCollar_3_r")
    Filter=(bX=True,bY=True,bZ=True)
    Parents=((Item=(Type=Null,Name="target_hoodieCollar_3_r_null"),Weight=1))
    UpAxis=(X=0,Y=-1,Z=0)
    Weight=1
    WorldUp=(Target=(X=-0.387,Y=0.5905,Z=-0.7082),Kind=Direction,Space=(Type=Bone,Name="spine_05"))
    bMaintainOffset=False
   )
   RigVMFunctionReturnNode(ExecutePin=@internalName0_6.ExecutePin)
  ]}
 RigVMFunctionLibrary {links [], nodes []}
 RigVMFunction_Sequence_1_ContainedGraph {nodes [
   n1 = RigVMFunctionEntryNode()
   RigVMFunction_Sequence = &RigVMFunction_Sequence(ExecuteContext=@n1.ExecuteContext)
   n2 = &RigVMFunction_Sequence(ExecuteContext=@RigVMFunction_Sequence.B)
   RigVMFunctionReturnNode(A=@RigVMFunction_Sequence.A, B=@n2.A, C=@n2.B)
  ]}
 RigVMFunction_Sequence_1_ContainedGraph {nodes [
   n1 = RigVMFunctionEntryNode()
   RigVMFunction_Sequence = &RigVMFunction_Sequence(ExecuteContext=@n1.ExecuteContext)
   n2 = &RigVMFunction_Sequence(ExecuteContext=@RigVMFunction_Sequence.B)
   n3 = &RigVMFunction_Sequence(ExecuteContext=@n2.B)
   RigVMFunctionReturnNode(A=@RigVMFunction_Sequence.A, B=@n2.A, C=@n3.A, D=@n3.B)
  ]}
 RigVMModel {nodes [
   RigUnit_BeginExecution = &RigUnit_BeginExecution()
   /Script/FortniteGame.RigUnit_FortDeformRig(
    Bool=(bUpperArm=true,bElbow=true,bWrist=true,bGroin=false,bGlute=false,bKnee=true,bPec=true,bBicep=False,bLat=False)
    Curve="(bLUpperArmUpAlpha=False,bRUpperArmUpAlpha=False,bLUpperArmFrAlpha=False,bRUpperArmFrAlpha=False,bLUpperArmBkAlpha=False,bRUpperArmBkAlpha=False,bLWristUpAlpha=False,bRWristUpAlpha=False,bLWristDnAlpha=False,bRWristDnAlpha=False,bLKneeAlpha=False,bRKneeAlp ...(truncated)"
    Debug=(bUpperArmDebug=False,bWristDebug=False,bLatDebug=False,bGroinGluteDebug=False)
    Value="(UpperArmUp=(Rotation=(X=0.0,Y=-0.0,Z=0.0,W=1.0),Translation=(X=0.0,Y=0.0,Z=6),Scale3D=(X=0.5,Y=1.0,Z=1.0)),UpperArmFwdFr=(Rotation=(X=0.0,Y=-0.0,Z=0.0,W=1.0),Translation=(X=0.0,Y=10.0,Z=0.0),Scale3D=(X=0.50,Y=1.0,Z=1.0)),UpperArmFwdBk=(Rotation=(X=0.0,Y=- ...(truncated)"
    ExecutePin=@RigUnit_BeginExecution.ExecutePin
   ) {is=struct RigVMUnitNode}
   > IvoryDwellChar_lowerBody = RigVMFunctionReferenceNode()
   hoodieSetup = RigVMCollapseNode(ExecutePin=@IvoryDwellChar_lowerBody.ExecuteContext)
  ]}
 RigVMModel Rig {nodes [
   n1 = &RigUnit_PrepareForExecution()
   RigVMFunction_Sequence = RigVMAggregateNode(ExecuteContext=@n1.ExecutePin)
   internalNametarget_hoodieCollar_1_c_null = &RigUnit_HierarchyAddNull(
    Name=target_hoodieCollar_1_c_null
    Parent=(Type=Bone,Name="spine_05")
    Space=LocalSpace
    Transform=(Rotation=(X=0,Y=0,Z=0,W=1),Translation=(X=22.0973,Y=-11.7982,Z=0),Scale3D=(X=1,Y=1,Z=1))
    ExecutePin=@RigVMFunction_Sequence.B
   )
   > internalNametarget_hoodieCollar_3_l_null = &RigUnit_HierarchyAddNull(
    Name=target_hoodieCollar_3_l_null
    Parent=(Type=Bone,Name="spine_05")
    Space=LocalSpace
    Transform=(Rotation=(X=0,Y=0,Z=0,W=1),Translation=(X=11.4447,Y=8.885,Z=-5.8925),Scale3D=(X=1,Y=1,Z=1))
   )
   > internalNametarget_hoodieCollar_2_l_null = &RigUnit_HierarchyAddNull(
    Name=target_hoodieCollar_2_l_null
    Parent=(Type=Bone,Name="spine_05")
    Space=LocalSpace
    Transform=(Rotation=(X=0,Y=0,Z=0,W=1),Translation=(X=17.3515,Y=1.9876,Z=-11.4143),Scale3D=(X=1,Y=1,Z=1))
   )
   > internalNametarget_hoodieCollar_1_l_null = &RigUnit_HierarchyAddNull(
    Name=target_hoodieCollar_1_l_null
    Parent=(Type=Bone,Name="spine_05")
    Space=LocalSpace
    Transform=(Rotation=(X=0,Y=0,Z=0,W=1),Translation=(X=20.4135,Y=-7.0318,Z=-8.2927),Scale3D=(X=1,Y=1,Z=1))
   )
   > internalNametarget_hoodieCollar_3_r_null = &RigUnit_HierarchyAddNull(
    Name=target_hoodieCollar_3_r_null
    Parent=(Type=Bone,Name="spine_05")
    Space=LocalSpace
    Transform=(Rotation=(X=0,Y=0,Z=0,W=1),Translation=(X=11.4447,Y=8.8848,Z=5.8924),Scale3D=(X=1,Y=1,Z=1))
   )
   > internalNametarget_hoodieCollar_2_r_null = &RigUnit_HierarchyAddNull(
    Name=target_hoodieCollar_2_r_null
    Parent=(Type=Bone,Name="spine_05")
    Space=LocalSpace
    Transform=(Rotation=(X=0,Y=0,Z=0,W=1),Translation=(X=17.3517,Y=1.9877,Z=11.4146),Scale3D=(X=1,Y=1,Z=1))
   )
   > internalNametarget_hoodieCollar_1_r_null = &RigUnit_HierarchyAddNull(
    Name=target_hoodieCollar_1_r_null
    Parent=(Type=Bone,Name="spine_05")
    Space=LocalSpace
    Transform=(Rotation=(X=0,Y=0,Z=0,W=1),Translation=(X=20.4136,Y=-7.0318,Z=8.2927),Scale3D=(X=1,Y=1,Z=1))
   )
   > internalNamehead_sp_null = &RigUnit_HierarchyAddNull(
    Name=head_sp_null
    Parent=(Type=Bone,Name="spine_05")
    Space=LocalSpace
    Transform=(Rotation=(X=0,Y=0,Z=0.036,W=0.9994),Translation=(X=20.6573,Y=3.5489,Z=0),Scale3D=(X=1,Y=1,Z=1))
   )
   internalNamehipChains_twist_bone_1 = /Script/ControlRig.RigUnit_HierarchyAddBone(
    Name=hipChains_twist_bone
    Parent=(Type=Bone,Name="pelvis")
    Space=LocalSpace
    Transform=(Rotation=(X=0.0075,Y=-0.975,Z=0.1399,W=0.1726),Translation=(X=-3.9069,Y=3.4534,Z=12.7954),Scale3D=(X=1,Y=1,Z=1))
    ExecutePin=@RigVMFunction_Sequence.C
   ) {is=struct RigVMUnitNode}
   > internalNamehipChains_target_bone_1 = &RigUnit_HierarchyAddBone(
    Name=hipChains_target_bone
    Parent=(Type=Bone,Name="pelvis")
    Space=LocalSpace
    Transform=(Rotation=(X=0.0075,Y=-0.975,Z=0.1399,W=0.1726),Translation=(X=-19.5746,Y=4.015,Z=18.4371),Scale3D=(X=1,Y=1,Z=1))
   )
   > internalNamebase_hemJacket_1_l_null = &RigUnit_HierarchyAddNull(
    Name=base_hemJacket_1_l_null
    Parent=(Type=Bone,Name="kin_hemJacket")
    Space=LocalSpace
    Transform=(Rotation=(X=0.0283,Y=-0.2471,Z=-0.9685,W=0.0068),Translation=(X=1.1224,Y=13.0099,Z=-9.0976),Scale3D=(X=1,Y=1,Z=1))
   )
   > internalNamebase_hemJacket_1_r_null = &RigUnit_HierarchyAddNull(
    Name=base_hemJacket_1_r_null
    Parent=(Type=Bone,Name="kin_hemJacket")
    Space=LocalSpace
    Transform=(Rotation=(X=-0.0283,Y=0.2471,Z=-0.9685,W=0.0068),Translation=(X=1.1219,Y=13.0098,Z=9.0976),Scale3D=(X=1,Y=1,Z=1))
   )
   internalNametarget_hemJacket_1_r_null = &RigUnit_HierarchyAddNull(
    Name=target_hemJacket_1_r_null
    Parent=(Type=Bone,Name="spine_02")
    Space=LocalSpace
    Transform=(Rotation=(X=0,Y=0,Z=0,W=1),Translation=(X=-14.2961,Y=12.5915,Z=9.8928),Scale3D=(X=1,Y=1,Z=1))
    ExecutePin=@RigVMFunction_Sequence.D
   )
   > internalNametarget_hemJacket_1_l_null = &RigUnit_HierarchyAddNull(
    Name=target_hemJacket_1_l_null
    Parent=(Type=Bone,Name="spine_02")
    Space=LocalSpace
    Transform=(Rotation=(X=0,Y=0,Z=0,W=1),Translation=(X=-14.2961,Y=12.5915,Z=-9.8928),Scale3D=(X=1,Y=1,Z=1))
   )
   > internalNametarget_hemJacket_2_r_null = &RigUnit_HierarchyAddNull(
    Name=target_hemJacket_2_r_null
    Parent=(Type=Bone,Name="spine_02")
    Space=LocalSpace
    Transform=(Rotation=(X=0,Y=0,Z=0,W=1),Translation=(X=-13.4146,Y=4.6931,Z=18.7961),Scale3D=(X=1,Y=1,Z=1))
   )
   > internalNametarget_hemJacket_2_l_null = &RigUnit_HierarchyAddNull(
    Name=target_hemJacket_2_l_null
    Parent=(Type=Bone,Name="spine_02")
    Space=LocalSpace
    Transform=(Rotation=(X=0,Y=0,Z=0,W=1),Translation=(X=-13.4149,Y=4.6931,Z=-18.7961),Scale3D=(X=1,Y=1,Z=1))
   )
   > internalNametarget_hemJacket_3_r_null = &RigUnit_HierarchyAddNull(
    Name=target_hemJacket_3_r_null
    Parent=(Type=Bone,Name="spine_02")
    Space=LocalSpace
    Transform=(Rotation=(X=0,Y=0,Z=0,W=1),Translation=(X=-12.5935,Y=-7.6634,Z=17.0597),Scale3D=(X=1,Y=1,Z=1))
   )
   > internalNametarget_hemJacket_3_l_null = &RigUnit_HierarchyAddNull(
    Name=target_hemJacket_3_l_null
    Parent=(Type=Bone,Name="spine_02")
    Space=LocalSpace
    Transform=(Rotation=(X=0,Y=0,Z=0,W=1),Translation=(X=-12.5935,Y=-7.6634,Z=-17.0597),Scale3D=(X=1,Y=1,Z=1))
   )
   > internalNametarget_jacketHem_back_null = &RigUnit_HierarchyAddNull(
    Name=target_jacketHem_back_null
    Parent=(Type=Bone,Name="spine_02")
    Space=LocalSpace
    Transform=(Rotation=(X=0,Y=0,Z=0,W=1),Translation=(X=-12.6545,Y=-13.5338,Z=-0.4968),Scale3D=(X=1,Y=1,Z=1))
   )
   ItemArray = RigVMDispatchNode(Value="((Type=Bone,Name=\"calf_l\"),(Type=Bone,Name=\"calf_r\"),(Type=Bone,Name=\"thigh_l\"),(Type=Bone,Name=\"thigh_r\"),(Type=Bone,Name=\"foot_l\"),(Type=Bone,Name=\"foot_r\"),(Type=Bone,Name=\"clavicle_l\"),(Type=Bone,Name=\"clavicle_r\"),(Type=Bone,Name=\"hand_l\"),(Type=Bone, ...(truncated)") {template="DISPATCH_RigVMDispatch_Constant(io Value)"}
   DISPATCH_RigVMDispatch_ArrayIterator = RigVMDispatchNode(Array=@ItemArray.Value, ExecuteContext=@RigVMFunction_Sequence.A) {template="DISPATCH_RigVMDispatch_ArrayIterator(in Array,out Element,out Index,out Count,out Ratio)"}
   Concat = /Script/RigVM.RigVMFunction_NameConcat(B=_null, A=@DISPATCH_RigVMDispatch_ArrayIterator.Element.Name) {is=struct RigVMUnitNode, template="Concat::Execute(in A,in B,out Result)"}
   > &RigUnit_HierarchyAddNull(
    Space=LocalSpace
    Name=@Concat.Result
    Transform=&RigUnit_GetTransform(Item=(Type=Bone,Name="clavicle_l"), Space=LocalSpace, bInitial=true, Item.Name=@DISPATCH_RigVMDispatch_ArrayIterator.Element.Name).Transform
    Parent=/Script/ControlRig.RigUnit_HierarchyGetParent(bDefaultParent=True, Child=@DISPATCH_RigVMDispatch_ArrayIterator.Element) {is=struct RigVMUnitNode}.Parent
   )
   ItemArray_1 = RigVMDispatchNode(Value=((Type=Bone,Name="upperarm_l"),(Type=Bone,Name="upperarm_r"),(Type=Bone,Name="hand_l"),(Type=Bone,Name="hand_r"),(Type=Bone,Name="thigh_l"),(Type=Bone,Name="thigh_r"),(Type=Bone,Name="foot_l"),(Type=Bone,Name="foot_r"))) {template="DISPATCH_RigVMDispatch_Constant(io Value)"}
   RigVMDispatchNode(ExecuteContext=(), Other=(), Array=@ItemArray_1.Value) {template="DISPATCH_RigVMDispatch_ArrayUnion(io Array,in Other)"}
  ]}
]
hierarchy [
 root {loc=<0 0 0>, rot=<0 0 0 1>}
 pelvis {parent=root, loc=<0 -0.7071 93.1191>, rot=<0.0127 0.707 -0.0127 -0.707>}
 spine_01 {parent=pelvis, loc=<6.7494 -0.6388 0>, rot=<0 0 0.034 -0.9994>}
 spine_02 {parent=spine_01, loc=<7.2518 -0.035 0>, rot=<0 0 0.0238 -0.9997>}
 spine_03 {parent=spine_02, loc=<6.0104 0 0>, rot=<0 0 0.0149 -0.9999>}
 spine_04 {parent=spine_03, loc=<8.1024 0 0>, rot=<0 0 -0.0132 -0.9999>}
 spine_05 {parent=spine_04, loc=<10.0146 0 0>, rot=<0 0 -0.0015 -1>}
 neck_01 {parent=spine_05, loc=<11.3138 0.1529 0>, rot=<0 0 -0.2003 -0.9797>}
 neck_02 {parent=neck_01, loc=<4.5627 0 0>, rot=<0 0 0.0504 -0.9987>}
 head {parent=neck_02, loc=<5.3914 0 0>, rot=<0 0 0.115 -0.9934>}
 attach_cape {parent=spine_05, loc=<9.017 -11.714 0>, rot=<0.0616 -0.7044 -0.0616 -0.7044>}
 cape {parent=attach_cape, type=Null, loc=<0 -2.1 -5.2>, rot=<0 0 0 1>}
 attach_backpack {parent=spine_05, loc=<-0.123 -13.473 0>, rot=<-0.0093 -0.707 0.0093 -0.707>}
 Backpack_BR {parent=attach_backpack, type=Null, loc=<0 -0.4786 -1.4116>, rot=<0 0 0 1>}
 clavicle_l {parent=spine_05, loc=<4.3575 -0.1466 -2.3223>, rot=<0.0465 -0.7289 -0.009 -0.6829>}
 upperarm_l {parent=clavicle_l, loc=<15.5111 0 0>, rot=<0.0226 -0.3234 -0.0088 -0.9459>}
 lowerarm_l {parent=upperarm_l, loc=<26.6514 0 0>, rot=<0.0391 -0.0099 -0.2453 -0.9686>}
 hand_l {parent=lowerarm_l, loc=<22.3491 0 0>, rot=<-0.7081 -0.0507 0.0975 -0.6975>}
 thumb_01_l {parent=hand_l, loc=<1.3118 -0.3191 -2.459>, rot=<0.6546 -0.4042 0.0664 -0.6354>}
 thumb_02_l {parent=thumb_01_l, loc=<5.9336 -0.0448 0.0401>, rot=<-0.0037 0.0213 0.298 -0.9543>}
 thumb_03_l {parent=thumb_02_l, loc=<2.2223 0.3682 -0.0907>, rot=<0.0118 -0.013 0.0066 -0.9998>}
 index_metacarpal_l {parent=hand_l, loc=<1.6535 0.0183 -2.96>, rot=<0.0158 -0.0124 -0.0173 -0.9996>}
 index_01_l {parent=index_metacarpal_l, loc=<8.3008 0.0357 -0.0524>, rot=<-0.0044 0.0485 0.0924 -0.9945>}
 index_02_l {parent=index_01_l, loc=<3.3587 -0.3004 0.0452>, rot=<-0.002 0.0027 0.3225 -0.9466>}
 index_03_l {parent=index_02_l, loc=<2.5578 -0.123 -0.0252>, rot=<-0.002 0.0067 0.2558 -0.9667>}
 middle_metacarpal_l {parent=hand_l, loc=<2.1856 -0.1725 -0.5984>, rot=<0.006 0.0168 -0.0329 -0.9993>}
 middle_01_l {parent=middle_metacarpal_l, loc=<7.3275 -0.2447 0.0129>, rot=<-0.0208 0.0042 0.0858 -0.9961>}
 middle_02_l {parent=middle_01_l, loc=<4.0123 0.1401 0.1175>, rot=<-0.0218 0.0258 0.3512 -0.9357>}
 middle_03_l {parent=middle_02_l, loc=<2.6077 0.0534 0.1373>, rot=<0.0082 0.0188 0.1708 -0.9851>}
 ring_metacarpal_l {parent=hand_l, loc=<2.2074 -0.3625 1.3638>, rot=<-0.0284 0.0722 -0.0458 -0.9959>}
 ring_01_l {parent=ring_metacarpal_l, loc=<7.0295 -0.6197 -0.1446>, rot=<-0.0489 -0.0252 0.1892 -0.9804>}
 ring_02_l {parent=ring_01_l, loc=<3.6174 0.4647 -0.1027>, rot=<0 0 0.2806 -0.9598>}
 ring_03_l {parent=ring_02_l, loc=<2.4152 0.114 0.0199>, rot=<0 0 0.1189 -0.9929>}
 pinky_metacarpal_l {parent=hand_l, loc=<1.7616 -0.9753 3.3814>, rot=<-0.0397 0.0814 -0.069 -0.9935>}
 pinky_01_l {parent=pinky_metacarpal_l, loc=<7.1709 -0.4831 -0.1094>, rot=<-0.0654 -0.0035 0.2242 -0.9723>}
 pinky_02_l {parent=pinky_01_l, loc=<2.8274 0.0893 -0.1043>, rot=<0.0037 0.0011 0.2854 -0.9584>}
 pinky_03_l {parent=pinky_02_l, loc=<2.2046 0.0307 -0.0078>, rot=<-0.0002 0.0098 0.1445 -0.9895>}
 deform_wrist_dn_base_l {parent=hand_l, loc=<0 -2.2342 0>, rot=<0 0 0 -1>}
 deform_wrist_dn_l {parent=deform_wrist_dn_base_l, loc=<0 0 0>, rot=<0 0 0 1>}
 deform_wrist_up_base_l {parent=hand_l, loc=<0 2.2682 0>, rot=<0 0 0 -1>}
 deform_wrist_up_l {parent=deform_wrist_up_base_l, loc=<0 0 0>, rot=<0 0 0 1>}
 lowerarm_twist_02_l {parent=lowerarm_l, loc=<9.6199 0 0>, rot=<0 0 0 -1>}
 lowerarm_twist_01_l {parent=lowerarm_l, loc=<17.8549 0 0>, rot=<0 0 0 -1>}
 deform_elbow_in_base_l {parent=lowerarm_l, loc=<0.2112 5.259 0.8274>, rot=<0 0 0 -1>}
 deform_elbow_in_l {parent=deform_elbow_in_base_l, loc=<0 0 0>, rot=<0 0 0 1>}
 deform_elbow_out_base_l {parent=lowerarm_l, loc=<0 -4 0>, rot=<0 0 0 -1>}
 deform_elbow_out_l {parent=deform_elbow_out_base_l, loc=<0 0 0>, rot=<0 0 0 1>}
 upperarm_twist_01_l {parent=upperarm_l, loc=<0.0041 0 0>, rot=<0 0 0 -1>}
 deform_upperarm_bk_base_l {parent=upperarm_twist_01_l, loc=<1.0813 -7.2193 -1.9453>, rot=<0 0 0 1>}
 deform_upperarm_bk_l {parent=deform_upperarm_bk_base_l, loc=<0 0 0>, rot=<0 0 0 1>}
 deform_upperarm_fr_base_l {parent=upperarm_twist_01_l, loc=<-0.3691 7.0506 -2.765>, rot=<0 0 0 1>}
 deform_upperarm_fr_l {parent=deform_upperarm_fr_base_l, loc=<0 0 0>, rot=<0 0 0 1>}
 deform_upperarm_up_base_l {parent=upperarm_twist_01_l, loc=<-3.2216 -1.1391 3.3311>, rot=<0 0 0 1>}
 deform_upperarm_up_l {parent=deform_upperarm_up_base_l, loc=<0 0 0>, rot=<0 0 0 1>}
 upperarm_twist_02_l {parent=upperarm_l, loc=<13.7941 0 0>, rot=<0 0 0 -1>}
 clavicle_r {parent=spine_05, loc=<4.3575 -0.1466 2.3223>, rot=<0.7289 0.0465 -0.6829 0.009>}
 upperarm_r {parent=clavicle_r, loc=<-15.5111 0 0>, rot=<0.0226 -0.3234 -0.0088 -0.9459>}
 lowerarm_r {parent=upperarm_r, loc=<-26.6514 0 0>, rot=<0.0391 -0.0099 -0.2453 -0.9686>}
 hand_r {parent=lowerarm_r, loc=<-22.3491 0 0>, rot=<-0.7081 -0.0507 0.0975 -0.6975>}
 thumb_01_r {parent=hand_r, loc=<-1.3118 0.3191 2.459>, rot=<0.6546 -0.4042 0.0664 -0.6354>}
 thumb_02_r {parent=thumb_01_r, loc=<-5.9336 0.0448 -0.0401>, rot=<-0.0037 0.0213 0.298 -0.9543>}
 thumb_03_r {parent=thumb_02_r, loc=<-2.2223 -0.3682 0.0907>, rot=<0.0118 -0.013 0.0066 -0.9998>}
 index_metacarpal_r {parent=hand_r, loc=<-1.6535 -0.0183 2.96>, rot=<0.0158 -0.0124 -0.0173 -0.9996>}
 index_01_r {parent=index_metacarpal_r, loc=<-8.3008 -0.0357 0.0524>, rot=<-0.0044 0.0485 0.0924 -0.9945>}
 index_02_r {parent=index_01_r, loc=<-3.3587 0.3004 -0.0452>, rot=<-0.002 0.0027 0.3225 -0.9466>}
 index_03_r {parent=index_02_r, loc=<-2.5578 0.123 0.0252>, rot=<-0.002 0.0067 0.2558 -0.9667>}
 middle_metacarpal_r {parent=hand_r, loc=<-2.1856 0.1725 0.5984>, rot=<0.006 0.0168 -0.0329 -0.9993>}
 middle_01_r {parent=middle_metacarpal_r, loc=<-7.3275 0.2447 -0.0129>, rot=<-0.0208 0.0042 0.0858 -0.9961>}
 middle_02_r {parent=middle_01_r, loc=<-4.0123 -0.1401 -0.1175>, rot=<-0.0218 0.0258 0.3512 -0.9357>}
 middle_03_r {parent=middle_02_r, loc=<-2.6077 -0.0534 -0.1373>, rot=<0.0082 0.0188 0.1708 -0.9851>}
 ring_metacarpal_r {parent=hand_r, loc=<-2.2074 0.3625 -1.3638>, rot=<-0.0284 0.0722 -0.0458 -0.9959>}
 ring_01_r {parent=ring_metacarpal_r, loc=<-7.0295 0.6197 0.1446>, rot=<-0.0489 -0.0252 0.1892 -0.9804>}
 ring_02_r {parent=ring_01_r, loc=<-3.6174 -0.4647 0.1027>, rot=<0 0 0.2806 -0.9598>}
 ring_03_r {parent=ring_02_r, loc=<-2.4152 -0.114 -0.0199>, rot=<0 0 0.1189 -0.9929>}
 pinky_metacarpal_r {parent=hand_r, loc=<-1.7616 0.9753 -3.3814>, rot=<-0.0397 0.0814 -0.069 -0.9935>}
 pinky_01_r {parent=pinky_metacarpal_r, loc=<-7.1709 0.4831 0.1094>, rot=<-0.0654 -0.0035 0.2242 -0.9723>}
 pinky_02_r {parent=pinky_01_r, loc=<-2.8274 -0.0893 0.1043>, rot=<0.0037 0.0011 0.2854 -0.9584>}
 pinky_03_r {parent=pinky_02_r, loc=<-2.2046 -0.0307 0.0078>, rot=<-0.0002 0.0098 0.1445 -0.9895>}
 deform_wrist_dn_base_r {parent=hand_r, loc=<-0.0004 2.2345 -0.0001>, rot=<0 0 0 -1>}
 deform_wrist_dn_r {parent=deform_wrist_dn_base_r, loc=<0 0 0>, rot=<0 0 0 1>}
 deform_wrist_up_base_r {parent=hand_r, loc=<0.0003 -2.2684 0.0001>, rot=<0 0 0 -1>}
 deform_wrist_up_r {parent=deform_wrist_up_base_r, loc=<0 0 0>, rot=<0 0 0 1>}
 lowerarm_twist_02_r {parent=lowerarm_r, loc=<-9.6199 0 0>, rot=<0 0 0 -1>}
 lowerarm_twist_01_r {parent=lowerarm_r, loc=<-17.8549 0 0>, rot=<0 0 0 -1>}
 deform_elbow_in_base_r {parent=lowerarm_r, loc=<-0.2111 -5.2591 -0.8275>, rot=<0 0 0 -1>}
 deform_elbow_in_r {parent=deform_elbow_in_base_r, loc=<-0.0001 0 0.0002>, rot=<0 0 0 1>}
 deform_elbow_out_base_r {parent=lowerarm_r, loc=<0 4 -0.0001>, rot=<0 0 0 -1>}
 deform_elbow_out_r {parent=deform_elbow_out_base_r, loc=<0 0 0>, rot=<0 0 0 1>}
 upperarm_twist_01_r {parent=upperarm_r, loc=<-0.0041 0 0>, rot=<0 0 0 -1>}
 deform_upperarm_bk_base_r {parent=upperarm_twist_01_r, loc=<-1.0814 7.2192 1.9455>, rot=<0 0 0 1>}
 deform_upperarm_bk_r {parent=deform_upperarm_bk_base_r, loc=<0 0 0>, rot=<0 0 0 1>}
 deform_upperarm_fr_base_r {parent=upperarm_twist_01_r, loc=<0.3689 -7.0506 2.7652>, rot=<0 0 0 1>}
 deform_upperarm_fr_r {parent=deform_upperarm_fr_base_r, loc=<0 0 0>, rot=<0 0 0 1>}
 deform_upperarm_up_base_r {parent=upperarm_twist_01_r, loc=<3.2214 1.1391 -3.3308>, rot=<0 0 0 1>}
 deform_upperarm_up_r {parent=deform_upperarm_up_base_r, loc=<0 0 0>, rot=<0 0 0 -1>}
 upperarm_twist_02_r {parent=upperarm_r, loc=<-13.7941 0 0>, rot=<0 0 0 -1>}
 base_jacketCollar_1_l {parent=spine_05, loc=<7.9885 6.5741 -7.859>, rot=<0.8424 0.4031 -0.3441 -0.0967>}
 dyn_jacketCollar_1_l {parent=base_jacketCollar_1_l, loc=<0 0 0>, rot=<0 0 0 1>}
 base_jacketCollar_1_r {parent=spine_05, loc=<7.5644 6.6 8.2754>, rot=<0.1555 0.3723 -0.3485 -0.8461>}
 dyn_jacketCollar_1_r {parent=base_jacketCollar_1_r, loc=<0 0 0>, rot=<0 0 0 1>}
 base_hoodieCollar_1_c {parent=spine_05, loc=<12.2591 -3.9745 0>, rot=<0.6676 -0.2331 0.2331 -0.6676>}
 dyn_hoodieCollar_1_c {parent=base_hoodieCollar_1_c, loc=<0 0 0>, rot=<0 0 0 1>}
 base_hoodieCollar_3_l {parent=spine_05, loc=<8.7833 5.0518 -4.1504>, rot=<-0.2874 -0.3328 -0.3495 -0.8274>}
 dyn_hoodieCollar_3_l {parent=base_hoodieCollar_3_l, loc=<0 0 0>, rot=<0 0 0 1>}
 base_hoodieCollar_2_l {parent=spine_05, loc=<12.0968 0.7383 -6.6949>, rot=<0.0201 -0.3514 -0.1012 -0.9305>}
 dyn_hoodieCollar_2_l {parent=base_hoodieCollar_2_l, loc=<0 0 0>, rot=<0 0 0 1>}
 base_hoodieCollar_1_l {parent=spine_05, loc=<12.6465 -2.9942 -4.4824>, rot=<0.413 -0.2851 0.1092 -0.858>}
 dyn_hoodieCollar_1_l {parent=base_hoodieCollar_1_l, loc=<0 0 0>, rot=<0 0 0 1>}
 base_hoodieCollar_3_r {parent=spine_05, loc=<8.7834 5.0519 4.1504>, rot=<0.8274 0.3495 0.3328 0.2874>}
 dyn_hoodieCollar_3_r {parent=base_hoodieCollar_3_r, loc=<0 0 0>, rot=<0 0 0 1>}
 base_hoodieCollar_2_r {parent=spine_05, loc=<12.0967 0.7383 6.6949>, rot=<0.9305 0.1012 0.3514 -0.0201>}
 dyn_hoodieCollar_2_r {parent=base_hoodieCollar_2_r, loc=<0 0 0>, rot=<0 0 0 1>}
 base_hoodieCollar_1_r {parent=spine_05, loc=<12.6466 -2.9942 4.4824>, rot=<0.858 -0.1092 0.2851 -0.413>}
 dyn_hoodieCollar_1_r {parent=base_hoodieCollar_1_r, loc=<0 0 0>, rot=<0 0 0 1>}
 base_hood {parent=spine_05, loc=<16.6559 -6.0763 0>, rot=<0.3664 -0.9305 0 0>}
 dyn_hood {parent=base_hood, loc=<0 0 0>, rot=<0 0 0 1>}
 deform_pec_aim_base_l {parent=spine_04, loc=<7.8732 12.6388 -2.4123>, rot=<0.0102 -0.4602 0.3979 -0.7936>}
 deform_pec_aim_l {parent=deform_pec_aim_base_l, loc=<0 0 0>, rot=<0 0 0 1>}
 deform_pec_base_l {parent=deform_pec_aim_l, loc=<10.766 0 0>, rot=<0 0 0 1>}
 deform_pec_l {parent=deform_pec_base_l, loc=<0.055 0 0>, rot=<0 0 0 1>}
 deform_pec_aim_base_r {parent=spine_04, loc=<7.8727 12.6388 2.4123>, rot=<-0.4602 -0.0102 0.7936 0.3979>}
 deform_pec_aim_r {parent=deform_pec_aim_base_r, loc=<0 0 0>, rot=<0 0 0 1>}
 deform_pec_base_r {parent=deform_pec_aim_r, loc=<-10.7662 -0.0006 -0.0007>, rot=<0 0 0 1>}
 deform_pec_r {parent=deform_pec_base_r, loc=<0 0 0>, rot=<0 0 0 1>}
 deform_lat_base_l {parent=spine_04, loc=<-0.11 0.3046 -15.4494>, rot=<0.0294 -0.7065 -0.0294 -0.7065>}
 deform_lat_l {parent=deform_lat_base_l, loc=<0 0 0>, rot=<0 0 0 1>}
 deform_lat_base_r {parent=spine_04, loc=<-0.11 0.3046 15.4494>, rot=<0.7065 0.0294 0.7065 -0.0294>}
 deform_lat_r {parent=deform_lat_base_r, loc=<0 0 0>, rot=<0 0 0 1>}
 kin_hemJacket {parent=spine_02, loc=<0 0 0>, rot=<0 0 0 1>}
 base_hemJacket_1_r {parent=kin_hemJacket, loc=<1.1219 13.0098 9.0976>, rot=<0.0283 -0.2471 0.9685 -0.0068>}
 dyn_hemJacket_1_r {parent=base_hemJacket_1_r, loc=<0 0 0>, rot=<0 0 0 1>}
 base_hemJacket_1_l {parent=kin_hemJacket, loc=<1.1224 13.0099 -9.0976>, rot=<-0.0283 0.2471 0.9685 -0.0068>}
 dyn_hemJacket_1_l {parent=base_hemJacket_1_l, loc=<0 0 0>, rot=<0 0 0 1>}
 base_hemJacket_2_r {parent=kin_hemJacket, loc=<1.8724 5.6926 16.8384>, rot=<0.0707 -0.5709 0.8179 0.0098>}
 dyn_hemJacket_2_r {parent=base_hemJacket_2_r, loc=<0 0 0>, rot=<0 0 0 1>}
 base_hemJacket_2_l {parent=kin_hemJacket, loc=<1.8725 5.6926 -16.8384>, rot=<-0.0707 0.5709 0.8179 0.0098>}
 dyn_hemJacket_2_l {parent=base_hemJacket_2_l, loc=<0 0 0>, rot=<0 0 0 1>}
 base_hemJacket_3_r {parent=kin_hemJacket, loc=<2.7438 -6.4373 15.7136>, rot=<0.0567 -0.85 0.5235 0.0163>}
 dyn_hemJacket_3_r {parent=base_hemJacket_3_r, loc=<0 0 0>, rot=<0 0 0 1>}
 base_hemJacket_3_l {parent=kin_hemJacket, loc=<2.7435 -6.4373 -15.7136>, rot=<0.0567 -0.85 -0.5235 -0.0163>}
 dyn_hemJacket_3_l {parent=base_hemJacket_3_l, loc=<0 0 0>, rot=<0 0 0 1>}
 base_hemJacket_1_c {parent=kin_hemJacket, loc=<2.741 -12.3031 -0.4968>, rot=<0.0399 -0.9992 0 0>}
 dyn_hemJacket_1_c {parent=base_hemJacket_1_c, loc=<0 0 0>, rot=<0 0 0 1>}
 kin_delta_hemJacket_1_l {parent=kin_hemJacket, loc=<1.6216 13.0234 -9.0719>, rot=<-0.0283 0.2471 0.9685 -0.0068>}
 kin_delta_hemJacket_1_r {parent=kin_hemJacket, loc=<1.6211 13.0234 9.0719>, rot=<0.0283 -0.2471 0.9685 -0.0068>}
 thigh_l {parent=pelvis, loc=<-2.5693 -0.0109 -7.7501>, rot=<0.0012 0.0572 0.0155 -0.9982>}
 calf_l {parent=thigh_l, loc=<-39.0484 0 0>, rot=<0.0006 -0.0161 -0.0391 -0.9991>}
 foot_l {parent=calf_l, loc=<-42.561 0.1077 0.4292>, rot=<0.0418 -0.0455 0.0343 -0.9975>}
 ball_l {parent=foot_l, loc=<-6.1599 13.8264 0.6806>, rot=<0.0001 -0.0001 -0.7063 -0.7079>}
 deform_ankle_in_base_l {parent=foot_l, loc=<2.9481 5.9274 0.65>, rot=<-0.0253 0.0419 -0.526 -0.8491>}
 deform_ankle_in_l {parent=deform_ankle_in_base_l, loc=<0 0 0>, rot=<0 0 0 1>}
 deform_ankle_out_base_l {parent=foot_l, loc=<-2.4371 -4.9 -0.5374>, rot=<-0.0253 0.0419 -0.526 -0.8491>}
 deform_ankle_out_l {parent=deform_ankle_out_base_l, loc=<0 0 0>, rot=<0 0 0 1>}
 calf_twist_02_l {parent=calf_l, loc=<-18.8064 -0.0967 0.0397>, rot=<0.0237 0.0095 -0.0035 -0.9997>}
 calf_twist_01_l {parent=calf_l, loc=<-35.813 -0.0967 0.0397>, rot=<0.0237 0.0095 -0.0035 -0.9997>}
 deform_knee_in_base_l {parent=calf_l, loc=<1.298 -5.2957 -0.0496>, rot=<0 0 0 -1>}
 deform_knee_in_l {parent=deform_knee_in_base_l, loc=<0 0 0>, rot=<0 0 0 1>}
 deform_knee_out_base_l {parent=calf_l, loc=<0 4 0>, rot=<0 0 0 -1>}
 deform_knee_out_l {parent=deform_knee_out_base_l, loc=<0 0 0>, rot=<0 0 0 1>}
 calf_twist_00_l {parent=calf_l, loc=<0 0 0>, rot=<0 0 0 1>}
 thigh_twist_01_l {parent=thigh_l, loc=<-21.6303 0 0>, rot=<0.0013 0 0 -1>}
 thigh_tshirt_l {parent=thigh_twist_01_l, loc=<20.6303 0 0>, rot=<-0.0013 0 0 -1>}
 deform_groin_base_l {parent=thigh_l, loc=<-2.1908 8.1142 -1.7442>, rot=<0 0 0 -1>}
 deform_groin_l {parent=deform_groin_base_l, loc=<0 0 0>, rot=<0 0 0 1>}
 thigh_twist_00_l {parent=thigh_l, loc=<-24.348 0 0>, rot=<0.0013 0 0 -1>}
 base_strapHang_out_l {parent=thigh_twist_00_l, loc=<-0.0785 0.9313 -0.9309>, rot=<-0.0028 -0.9153 -0.0002 -0.4027>}
 dyn_strapHang_out_l {parent=base_strapHang_out_l, loc=<0 0 0>, rot=<0 0 0 1>}
 base_strapHang_in_l {parent=thigh_twist_00_l, loc=<-0.2906 0.9234 2.3819>, rot=<0.3453 -0.0021 0.9385 0.0018>}
 dyn_strapHang_in_l {parent=base_strapHang_in_l, loc=<0 0 0>, rot=<0 0 0 1>}
 kin_strap_frontTris_l {parent=thigh_twist_00_l, loc=<1.2 10 1.2>, rot=<0 0 0 1>}
 target_strap_01_l {parent=thigh_l, loc=<-25.6021 10.4002 0.6359>, rot=<0 0 0 1>}
 target_strap_02_l {parent=thigh_l, loc=<-32.9711 3.8803 -9.6501>, rot=<0 0 0 1>}
 coll_strap_01_l {parent=thigh_l, loc=<-8.3471 10.8121 3.5943>, rot=<0.0117 -0.9963 0.0013 -0.0848>}
 coll_strap_02_l {parent=thigh_l, loc=<-23.9279 10.1298 -0.7387>, rot=<0.2091 -0.901 0.0877 -0.3698>}
 thigh_r {parent=pelvis, loc=<-2.5693 -0.0109 7.7501>, rot=<0.0572 -0.0012 0.9982 0.0155>}
 calf_r {parent=thigh_r, loc=<39.0484 0 0>, rot=<0.0006 -0.0161 -0.0391 -0.9991>}
 foot_r {parent=calf_r, loc=<42.561 -0.1077 -0.4292>, rot=<0.0418 -0.0455 0.0343 -0.9975>}
 ball_r {parent=foot_r, loc=<6.1599 -13.8264 -0.6806>, rot=<0.0001 -0.0001 -0.7063 -0.7079>}
 deform_ankle_in_base_r {parent=foot_r, loc=<-2.9481 -5.9274 -0.6501>, rot=<-0.0253 0.0419 -0.526 -0.8491>}
 deform_ankle_in_r {parent=deform_ankle_in_base_r, loc=<0 0 0>, rot=<0 0 0 1>}
 deform_ankle_out_base_r {parent=foot_r, loc=<2.4371 4.9 0.5373>, rot=<-0.0253 0.0419 -0.526 -0.8491>}
 deform_ankle_out_r {parent=deform_ankle_out_base_r, loc=<0 0 0>, rot=<0 0 0 1>}
 calf_twist_02_r {parent=calf_r, loc=<18.8064 0.0967 -0.0397>, rot=<0.0237 0.0095 -0.0035 -0.9997>}
 base_calfLace_1_l {parent=calf_twist_02_r, loc=<12.3206 4.037 4.6723>, rot=<0.8692 -0.099 0.3205 -0.3632>}
 dyn_calfLace_1_l {parent=base_calfLace_1_l, loc=<0 0 0>, rot=<0 0 0 1>}
 dyn_calfLace_2_l {parent=dyn_calfLace_1_l, loc=<3.623 0 0>, rot=<0.0857 0.1663 0.0775 -0.9793>}
 calf_twist_01_r {parent=calf_r, loc=<35.813 0.0967 -0.0397>, rot=<0.0237 0.0095 -0.0035 -0.9997>}
 deform_knee_in_base_r {parent=calf_r, loc=<-1.298 5.2957 0.0496>, rot=<0 0 0 -1>}
 deform_knee_in_r {parent=deform_knee_in_base_r, loc=<0 0 0>, rot=<0 0 0 1>}
 deform_knee_out_base_r {parent=calf_r, loc=<0 -4 0>, rot=<0 0 0 -1>}
 deform_knee_out_r {parent=deform_knee_out_base_r, loc=<0 0 0>, rot=<0 0 0 1>}
 calf_twist_00_r {parent=calf_r, loc=<0 0 0>, rot=<0 0 0 1>}
 thigh_twist_01_r {parent=thigh_r, loc=<21.6303 0 0>, rot=<0.0013 0 0 -1>}
 thigh_tshirt_r {parent=thigh_twist_01_r, loc=<-20.6303 0 0>, rot=<-0.0013 0 0 -1>}
 deform_groin_base_r {parent=thigh_r, loc=<2.1908 -8.1142 1.7442>, rot=<0 0 0 -1>}
 deform_groin_r {parent=deform_groin_base_r, loc=<0 0 0>, rot=<0 0 0 1>}
 thigh_twist_00_r {parent=thigh_r, loc=<24.348 0 0>, rot=<0.0013 0 0 -1>}
 base_strapHang_out_r {parent=thigh_twist_00_r, loc=<0.0785 -0.9313 0.9309>, rot=<-0.0002 0.4027 0.0028 -0.9153>}
 dyn_strapHang_out_r {parent=base_strapHang_out_r, loc=<0 0 0>, rot=<0 0 0 1>}
 base_strapHang_in_r {parent=thigh_twist_00_r, loc=<0.2906 -0.9234 -2.3818>, rot=<0.9526 -0.0019 -0.3042 -0.002>}
 dyn_strapHang_in_r {parent=base_strapHang_in_r, loc=<0 0 0>, rot=<0 0 0 1>}
 kin_strap_frontTris_r {parent=thigh_twist_00_r, loc=<-1.2 -10 -1.2>, rot=<0 0 0 1>}
 deform_glute_base_l {parent=pelvis, loc=<-8.8064 -8.8222 -8.4287>, rot=<0.0275 0.0501 -0.4472 -0.8926>}
 deform_glute_l {parent=deform_glute_base_l, loc=<0 0 0>, rot=<0 0 0 1>}
 deform_glute_base_r {parent=pelvis, loc=<-8.8062 -8.8222 8.4287>, rot=<0.0501 -0.0275 0.8926 -0.4472>}
 deform_glute_r {parent=deform_glute_base_r, loc=<0 0 0>, rot=<0 0 0 1>}
 base_hipChains {parent=pelvis, loc=<-3.9069 3.4534 12.7954>, rot=<0.0075 -0.975 0.1399 0.1726>}
 dyn_hipChains {parent=base_hipChains, loc=<0 0 0>, rot=<0 0 0 1>}
 dyn_hipChains_front {parent=dyn_hipChains, loc=<8.2695 7.1653 1.0993>, rot=<0 0 0 1>}
 dyn_hipChains_back {parent=dyn_hipChains, loc=<9.5755 -7.7709 -0.2513>, rot=<0 0 0 1>}
 kin_carabiners_01_r {parent=pelvis, loc=<-0.1112 9.6521 11.9341>, rot=<-0.2431 -0.6744 -0.1867 -0.6718>}
 kin_hoodie_hem_01_r {parent=pelvis, loc=<5.885 7.3222 12.7507>, rot=<-0.0127 -0.707 0.0127 -0.707>}
 kin_carabiners_01_l {parent=pelvis, loc=<-0.3502 8.6716 -13.0299>, rot=<0.6951 0.13 0.6899 0.1548>}
 base_hangCarabinerF_1 {parent=kin_carabiners_01_l, loc=<-0.0253 -0.2536 -4.1373>, rot=<-0.4976 -0.4826 0.5393 -0.4782>}
 dyn_hangCarabinerF_1 {parent=base_hangCarabinerF_1, loc=<0 0 0>, rot=<0 0 0 1>}
 dyn_hangCarabinerF_2 {parent=dyn_hangCarabinerF_1, loc=<5.4674 0 0>, rot=<-0.0399 0.0392 -0.0016 -0.9984>}
 kin_carabiners_02_l {parent=pelvis, loc=<-0.1246 5.3503 -16.0407>, rot=<0.6206 0.3388 0.6081 0.3608>}
 base_hangCarabinerS_1 {parent=kin_carabiners_02_l, loc=<0.174 -0.6164 -3.8873>, rot=<-0.4044 -0.5523 0.5061 -0.5246>}
 dyn_hangCarabinerS_1 {parent=base_hangCarabinerS_1, loc=<0 0 0>, rot=<0 0 0 1>}
 dyn_hangCarabinerS_2 {parent=dyn_hangCarabinerS_1, loc=<6.6494 -0.0053 -0.0157>, rot=<0 0.0633 0 -0.998>}
 coll_hem_f_r {parent=pelvis, loc=<-19.5365 5.0036 17.6577>, rot=<-0.4365 -0.3413 -0.1164 -0.8243>}
 target_hem_f_r {parent=pelvis, loc=<-11.3146 10.4464 12.535>, rot=<-0.0127 -0.707 0.0127 -0.707>}
 kin_strapStart_l {parent=pelvis, loc=<-2.7956 10.9807 -3.1745>, rot=<0 0 0 1>}
 kin_strapStart_r {parent=pelvis, loc=<-2.7956 10.9807 3.3659>, rot=<0 0 0 1>}
]
