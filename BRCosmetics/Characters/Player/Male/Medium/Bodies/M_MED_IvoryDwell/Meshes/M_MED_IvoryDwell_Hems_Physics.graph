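# Physics asset for secondary motion of the IvoryDwell male medium body's jacket hem and hanging front carabiner.
# Box-shaped hem flaps hang from small base bodies on spine_02 through angle-limited joints with locked twist, one adding a capsule.
# Capsules on pelvis, lower spine, thighs, carabiner mounts and a front-right hem collider approximate surrounding volumes.
# All other joints, including the hanging carabiner's link to its left mount, leave rotation free.
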PhysicsAsset /BRCosmetics/Characters/Player/Male/Medium/Bodies/M_MED_IvoryDwell/Meshes/M_MED_IvoryDwell_Hems_Physics
bodies {
 base_hemJacket_1_c [{shapeType=Capsule, length=0.25, radius=0.5, shapeName=base_hemJacket_1_c_capsule}]
 base_hemJacket_1_l [{shapeType=Capsule, length=0.25, radius=0.5, shapeName=base_hemJacket_1_l_capsule}]
 base_hemJacket_1_r [{shapeType=Capsule, length=0.25, radius=0.5, shapeName=base_hemJacket_1_r_capsule}]
 base_hemJacket_2_l [{shapeType=Capsule, length=0.25, radius=0.5, shapeName=base_hemJacket_2_l_capsule}]
 base_hemJacket_2_r [{shapeType=Capsule, length=0.25, radius=0.5, shapeName=base_hemJacket_2_r_capsule}]
 base_hemJacket_3_l [{shapeType=Capsule, length=0.25, radius=0.5, shapeName=base_hemJacket_3_l_capsule}]
 base_hemJacket_3_r [{shapeType=Capsule, length=0.25, radius=0.5, shapeName=base_hemJacket_3_r_capsule}]
 coll_hem_f_r [{
   shapeType=Capsule
   length=9.1017
   radius=2.4581
   shapeName=coll_hem_f_r_capsule
   center=<5.6536 -0.4645 -0.2933>
   rotation=<rot -0.0347 -92.3736 89.164>
  }]
 dyn_hangCarabinerF_2 [{
   shapeType=Capsule
   length=9.1393
   radius=1.7548
   shapeName=dyn_hangCarabinerF_2_capsule
   center=<3.546 0.14 0.0068>
   rotation=<rot -0.0754 -72.4221 -89.7954>
  }]
 dyn_hemJacket_1_c [{
   shapeType=Box
   extentX=16.4805
   extentY=1.9388
   extentZ=24.6134
   shapeName=dyn_hemJacket_1_l_box
   center=<8.2764 -1.0605 0>
   rotation=<rot 0 -2.6619 0>
  }]
 dyn_hemJacket_1_l [{
   shapeType=Box
   extentX=16.0619
   extentY=1.7044
   extentZ=11.4453
   shapeName=dyn_hemJacket_1_l_box
   center=<9.0185 0 -3.4239>
   rotation=<rot 0 0 -7.5>
  }]
 dyn_hemJacket_1_r [{
   shapeType=Box
   extentX=16.0619
   extentY=1.7044
   extentZ=11.4453
   shapeName=dyn_hemJacket_1_l_box
   center=<9.0185 0 3.4239>
   rotation=<rot 0 0 7.5>
  }]
 dyn_hemJacket_2_l [{shapeType=Box, extentX=15.395, extentY=1.7044, extentZ=11.4453, shapeName=dyn_hemJacket_1_l_box, center=<9.0185 0 0>}]
 dyn_hemJacket_2_r [
  {
   shapeType=Capsule
   length=10.0239
   radius=3.7184
   shapeName=dyn_hemJacket_2_r_capsule
   center=<8.2145 -0.1458 -9.3342>
   rotation=<rot 88.236 0.0001 0.0001>
  }
  {shapeType=Box, extentX=15.395, extentY=1.7044, extentZ=11.4453, shapeName=dyn_hemJacket_1_l_box, center=<9.0185 0 0>}
 ]
 dyn_hemJacket_3_l [{
   shapeType=Box
   extentX=15.6851
   extentY=1.7044
   extentZ=11.4453
   shapeName=dyn_hemJacket_1_r_box
   center=<9.0185 0.7168 0>
   rotation=<rot 0 0 5.8>
  }]
 dyn_hemJacket_3_r [{
   shapeType=Box
   extentX=15.6851
   extentY=1.7044
   extentZ=11.4453
   shapeName=dyn_hemJacket_1_r_box
   center=<9.0185 0.7168 0>
   rotation=<rot 0 0 5.8>
  }]
 kin_carabiners_01_l [{shapeType=Capsule, length=3.0096, radius=2.6267, shapeName=kin_carabiners_01_l_capsule, center=<-0.858 0.5576 0>}]
 kin_carabiners_01_r [{shapeType=Capsule, length=8.0192, radius=2.6267, shapeName=kin_carabiners_01_l_capsule, center=<-1.8875 -0.3518 -1.2386>}]
 kin_carabiners_02_l [{shapeType=Capsule, length=3.0096, radius=2.6267, shapeName=kin_carabiners_01_l_capsule, center=<-0.858 0.5576 0>}]
 pelvis [
  {shapeType=Capsule, length=12.6666, radius=11.2684, shapeName=pelvis_capsule}
  {shapeType=Capsule, length=9.5136, radius=11.2684, shapeName=pelvis_capsule_back, center=<-2.0646 -2.8997 0>}
 ]
 spine_01 [{shapeType=Capsule, length=13.3885, radius=11.0671, shapeName=pelvis_capsule, center=<-0.0893 0 0>}]
 spine_02 [{shapeType=Capsule, length=11.1633, radius=8.0826, shapeName=pelvis_capsule}]
 thigh_l [{
   shapeType=Capsule
   length=21.3142
   radius=11.9704
   shapeName=thigh_l_capsule
   center=<-14.8238 -1.1277 1.0571>
   rotation=<rot 88.6733 179.9998 179.9998>
  }]
 thigh_r [{
   shapeType=Capsule
   length=21.3142
   radius=11.9704
   shapeName=thigh_r_capsule
   center=<14.8238 1.1277 -1.0571>
   rotation=<rot 88.6733 179.9998 -0.0002>
  }]
 thigh_twist_00_l [{shapeType=Capsule, length=6.055, radius=9.5639, shapeName=thigh_twist_00_l_capsule, center=<-5.4465 -0.5383 0.9579>, rotation=<rot 0 0 -90>}]
 thigh_twist_00_r [{shapeType=Capsule, length=6.055, radius=9.5639, shapeName=thigh_twist_00_r_capsule, center=<5.4465 0.5383 -0.9579>, rotation=<rot 0 0 90>}]
}
constraints [
 {
  bone1Name=spine_02
  bone2Name=spine_01
  swing1LimitDegrees=45
  swing1Motion=Free
  swing2LimitDegrees=45
  swing2Motion=Free
  twistLimitDegrees=45
  twistMotion=Free
 }
 {
  bone1Name=spine_01
  bone2Name=pelvis
  swing1LimitDegrees=45
  swing1Motion=Free
  swing2LimitDegrees=45
  swing2Motion=Free
  twistLimitDegrees=45
  twistMotion=Free
 }
 {
  bone1Name=thigh_l
  bone2Name=pelvis
  swing1LimitDegrees=45
  swing1Motion=Free
  swing2LimitDegrees=45
  swing2Motion=Free
  twistLimitDegrees=45
  twistMotion=Free
 }
 {
  bone1Name=thigh_r
  bone2Name=pelvis
  swing1LimitDegrees=45
  swing1Motion=Free
  swing2LimitDegrees=45
  swing2Motion=Free
  twistLimitDegrees=45
  twistMotion=Free
 }
 {bone1Name=dyn_hemJacket_3_r, bone2Name=base_hemJacket_3_r, swing1LimitDegrees=15, swing2LimitDegrees=15, twistLimitDegrees=0}
 {
  bone1Name=base_hemJacket_1_c
  bone2Name=spine_02
  swing1LimitDegrees=45
  swing1Motion=Free
  swing2LimitDegrees=45
  swing2Motion=Free
  twistLimitDegrees=45
  twistMotion=Free
 }
 {bone1Name=dyn_hemJacket_1_c, bone2Name=base_hemJacket_1_c, swing1LimitDegrees=15, swing2LimitDegrees=0, twistLimitDegrees=0}
 {
  bone1Name=base_hemJacket_1_l
  bone2Name=spine_02
  swing1LimitDegrees=45
  swing1Motion=Free
  swing2LimitDegrees=45
  swing2Motion=Free
  twistLimitDegrees=45
  twistMotion=Free
 }
 {bone1Name=dyn_hemJacket_1_l, bone2Name=base_hemJacket_1_l, swing1LimitDegrees=20, swing2LimitDegrees=15, twistLimitDegrees=0}
 {
  bone1Name=base_hemJacket_1_r
  bone2Name=spine_02
  swing1LimitDegrees=45
  swing1Motion=Free
  swing2LimitDegrees=45
  swing2Motion=Free
  twistLimitDegrees=45
  twistMotion=Free
 }
 {bone1Name=dyn_hemJacket_1_r, bone2Name=base_hemJacket_1_r, swing1LimitDegrees=20, swing2LimitDegrees=15, twistLimitDegrees=0}
 {
  bone1Name=base_hemJacket_2_l
  bone2Name=spine_02
  swing1LimitDegrees=45
  swing1Motion=Free
  swing2LimitDegrees=45
  swing2Motion=Free
  twistLimitDegrees=45
  twistMotion=Free
 }
 {bone1Name=dyn_hemJacket_2_l, bone2Name=base_hemJacket_2_l, swing1LimitDegrees=15, swing2LimitDegrees=15, twistLimitDegrees=0}
 {
  bone1Name=base_hemJacket_2_r
  bone2Name=spine_02
  swing1LimitDegrees=45
  swing1Motion=Free
  swing2LimitDegrees=45
  swing2Motion=Free
  twistLimitDegrees=45
  twistMotion=Free
 }
 {bone1Name=dyn_hemJacket_2_r, bone2Name=base_hemJacket_2_r, swing1LimitDegrees=15, swing2LimitDegrees=15, twistLimitDegrees=0}
 {
  bone1Name=base_hemJacket_3_l
  bone2Name=spine_02
  swing1LimitDegrees=45
  swing1Motion=Free
  swing2LimitDegrees=45
  swing2Motion=Free
  twistLimitDegrees=45
  twistMotion=Free
 }
 {bone1Name=dyn_hemJacket_3_l, bone2Name=base_hemJacket_3_l, swing1LimitDegrees=15, swing2LimitDegrees=15, twistLimitDegrees=0}
 {
  bone1Name=base_hemJacket_3_r
  bone2Name=spine_02
  swing1LimitDegrees=45
  swing1Motion=Free
  swing2LimitDegrees=45
  swing2Motion=Free
  twistLimitDegrees=45
  twistMotion=Free
 }
 {
  bone1Name=thigh_twist_00_l
  bone2Name=thigh_l
  swing1LimitDegrees=45
  swing1Motion=Free
  swing2LimitDegrees=45
  swing2Motion=Free
  twistLimitDegrees=45
  twistMotion=Free
 }
 {
  bone1Name=thigh_twist_00_r
  bone2Name=thigh_r
  swing1LimitDegrees=45
  swing1Motion=Free
  swing2LimitDegrees=45
  swing2Motion=Free
  twistLimitDegrees=45
  twistMotion=Free
 }
 {
  bone1Name=dyn_hangCarabinerF_2
  bone2Name=kin_carabiners_01_l
  swing1LimitDegrees=45
  swing1Motion=Free
  swing2LimitDegrees=45
  swing2Motion=Free
  twistLimitDegrees=45
  twistMotion=Free
 }
 {
  bone1Name=coll_hem_f_r
  bone2Name=pelvis
  swing1LimitDegrees=45
  swing1Motion=Free
  swing2LimitDegrees=45
  swing2Motion=Free
  twistLimitDegrees=45
  twistMotion=Free
 }
 {
  bone1Name=kin_carabiners_02_l
  bone2Name=pelvis
  swing1LimitDegrees=45
  swing1Motion=Free
  swing2LimitDegrees=45
  swing2Motion=Free
  twistLimitDegrees=45
  twistMotion=Free
 }
 {
  bone1Name=kin_carabiners_01_r
  bone2Name=pelvis
  swing1LimitDegrees=45
  swing1Motion=Free
  swing2LimitDegrees=45
  swing2Motion=Free
  twistLimitDegrees=45
  twistMotion=Free
 }
 {
  bone1Name=kin_carabiners_01_l
  bone2Name=pelvis
  swing1LimitDegrees=45
  swing1Motion=Free
  swing2LimitDegrees=45
  swing2Motion=Free
  twistLimitDegrees=45
  twistMotion=Free
 }
]
