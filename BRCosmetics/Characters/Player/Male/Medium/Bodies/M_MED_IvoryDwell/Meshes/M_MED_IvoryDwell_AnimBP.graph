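# Post-pose AnimBlueprint for the IvoryDwell player body cosmetic, adding secondary motion and fit fixes to the pose copied from the attached parent mesh.
# Cached stages chain Control Rig, toggled bone offsets, AnimDynamics straps/chains/laces/hood, constraint/look-at helpers, RigidBody carabiners/hems/collar, and a final hem constraint.
# The EventGraph tracks backbling, glider and leg visibility, picks front-end sim settings, and computes gravity and strap alphas.
# It also drives WPO cloth material parameters and swaps skin-weight profiles by LOD.
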AnimBlueprint /BRCosmetics/Characters/Player/Male/Medium/Bodies/M_MED_IvoryDwell/Meshes/M_MED_IvoryDwell_AnimBP
target_skeleton=&/M_MED_IvoryDwell_Skeleton
graphs [
 AnimGraph {nodes [
   Output Pose(Result=Use cached pose 'postSimFixes'()) {layerGroup=None}
   n1 = Control Rig(Source=Copy Pose From Mesh() {bUseAttachedParent=true}) {controlRigAssetReference {blueprintRigClass=&/M_MED_IvoryDwell_CtrlRig.M_MED_IvoryDwell_CtrlRig_C}}
   Save cached pose 'controlRigCP'(Pose=@n1) {cacheName=controlRigCP}
   n2 = "Constraint - attach_cape (P)"(ComponentPose=Local To Component(LocalPose=Use cached pose 'controlRigCP'()), Alpha=1, ConstraintWeights_0=1) {
    alphaBoolBlend {blendOption=Linear}
    boneToModify {boneName=attach_cape}
    constraintSetup [{targetBone {boneName=spine_05}}]
    constraintWeights [1]
   }
   n3 = "Transform (Modify) Bone - Bone: attach_cape"!(ComponentPose=@n2, bAlphaBoolEnabled:bool=True, Translation="0,-1,0", Rotation="0,0,16") -> (Pose) {
    alphaInputType=Bool
    rotationMode=BMM_Additive
    rotationSpace=BCS_BoneSpace
    translationMode=BMM_Additive
    translationSpace=BCS_BoneSpace
    alphaBoolBlend {blendInTime=0.5, blendOption=CircularOut, blendOutTime=0.5}
    boneToModify {boneName=attach_cape}
    rotation {roll=16}
    translation {y=-1}
   }
   n4 = "Transform (Modify) Bone - Bone: attach_cape"(ComponentPose=@n3, bAlphaBoolEnabled=True, Translation="0,-2,0", Rotation="0,0,20") {
    alphaInputType=Bool
    rotationMode=BMM_Additive
    rotationSpace=BCS_BoneSpace
    translationMode=BMM_Additive
    translationSpace=BCS_BoneSpace
    alphaBoolBlend {blendInTime=0.5, blendOption=CircularOut, blendOutTime=0.5}
    boneToModify {boneName=attach_cape}
    rotation {roll=20}
    translation {y=-2}
   }
   n5 = "Transform (Modify) Bone - Bone: attach_cape"(ComponentPose=@n4, bAlphaBoolEnabled=True, Translation="0,0,0", Rotation="0,0,10") {
    alphaInputType=Bool
    rotationMode=BMM_Additive
    rotationSpace=BCS_BoneSpace
    translationMode=BMM_Additive
    translationSpace=BCS_BoneSpace
    alphaBoolBlend {blendInTime=0.5, blendOption=Cubic, blendOutTime=0.5}
    boneToModify {boneName=attach_cape}
    rotation {roll=10}
   }
   n6 = "Transform (Modify) Bone - Bone: base_hood"!(ComponentPose=@n5, bAlphaBoolEnabled:bool=True, Translation="-0.366,0.468,-0.000", Scale="0.9,0.9,0.9") -> (Pose) {
    alphaInputType=Bool
    rotationMode=BMM_Additive
    rotationSpace=BCS_BoneSpace
    scaleMode=BMM_Additive
    scaleSpace=BCS_BoneSpace
    translationMode=BMM_Additive
    translationSpace=BCS_BoneSpace
    alphaBoolBlend {blendInTime=0.5, blendOption=Cubic, blendOutTime=0.5}
    boneToModify {boneName=base_hood}
    scale=<0.9 0.9 0.9>
    translation=<-0.366 0.468>
   }
   n7 = "Transform (Modify) Bone - Bone: calf_twist_00_l"!(ComponentPose=@n6, bAlphaBoolEnabled:bool=True, Translation="0,0,0", Scale="1,1,0.3") -> (Pose) {
    alphaInputType=Bool
    rotationMode=BMM_Additive
    rotationSpace=BCS_BoneSpace
    scaleMode=BMM_Additive
    scaleSpace=BCS_BoneSpace
    translationMode=BMM_Additive
    translationSpace=BCS_BoneSpace
    alphaBoolBlend {blendOption=Cubic}
    boneToModify {boneName=calf_twist_00_l}
    scale {z=0.3}
   }
   n8 = "Transform (Modify) Bone - Bone: calf_twist_00_r"!(ComponentPose=@n7, bAlphaBoolEnabled:bool=True, Translation="0,0,0", Scale="1,1,0.3") -> (Pose) {
    alphaInputType=Bool
    rotationMode=BMM_Additive
    rotationSpace=BCS_BoneSpace
    scaleMode=BMM_Additive
    scaleSpace=BCS_BoneSpace
    translationMode=BMM_Additive
    translationSpace=BCS_BoneSpace
    alphaBoolBlend {blendOption=Cubic}
    boneToModify {boneName=calf_twist_00_r}
    scale {z=0.3}
   }
   n9 = "Transform (Modify) Bone - Bone: base_jacketCollar_1_l"!(ComponentPose=@n8, bAlphaBoolEnabled:bool=True, Rotation="-25,0,0") -> (Pose) {
    alphaInputType=Bool
    rotationMode=BMM_Additive
    rotationSpace=BCS_BoneSpace
    scaleMode=BMM_Additive
    scaleSpace=BCS_BoneSpace
    translationMode=BMM_Additive
    translationSpace=BCS_BoneSpace
    alphaBoolBlend {blendInTime=0.5, blendOption=Cubic, blendOutTime=0.5}
    boneToModify {boneName=base_jacketCollar_1_l}
    rotation {pitch=-25}
   }
   n10 = "Transform (Modify) Bone - Bone: base_jacketCollar_1_r"!(ComponentPose=@n9, bAlphaBoolEnabled:bool=True, Rotation="-25,0,0") -> (Pose) {
    alphaInputType=Bool
    rotationMode=BMM_Additive
    rotationSpace=BCS_BoneSpace
    scaleMode=BMM_Additive
    scaleSpace=BCS_BoneSpace
    translationMode=BMM_Additive
    translationSpace=BCS_BoneSpace
    alphaBoolBlend {blendInTime=0.5, blendOption=Cubic, blendOutTime=0.5}
    boneToModify {boneName=base_jacketCollar_1_r}
    rotation {pitch=-25}
   }
   n11 = "Transform (Modify) Bone - Bone: attach_backpack"!(ComponentPose=@n10, bAlphaBoolEnabled:bool=True, Translation="0,2.5,0", Rotation="0,0,5") -> (Pose) {
    alphaInputType=Bool
    rotationMode=BMM_Additive
    rotationSpace=BCS_BoneSpace
    scaleMode=BMM_Additive
    scaleSpace=BCS_BoneSpace
    translationMode=BMM_Additive
    translationSpace=BCS_BoneSpace
    alphaBoolBlend {blendInTime=0.1, blendOption=Linear, blendOutTime=0.1}
    boneToModify {boneName=attach_backpack}
    rotation {roll=5}
    translation {y=2.5}
   }
   Save cached pose 'generalAdjuistementsCP'(Pose=Component To Local(ComponentPose=@n11)) {cacheName=generalAdjuistementsCP}
   n12 = "Constraint - dyn_hemJacket_1_c (PP)"!(
    ComponentPose=Local To Component(LocalPose=Use cached pose 'DynamicsCP'())
    bAlphaBoolEnabled:bool=True
    ConstraintWeights_0=0.5
    ConstraintWeights_1:float=0.5
   ) -> (Pose) {
    alphaInputType=Bool
    alphaBoolBlend {blendOption=Linear}
    boneToModify {boneName=dyn_hemJacket_1_c}
    constraintSetup [{targetBone {boneName=base_hemJacket_1_c}}, {targetBone {boneName=dyn_hemJacket_1_c}}]
    constraintWeights [1,1]
   }
   Save cached pose 'postSimFixes'(Pose=Component To Local(ComponentPose=@n12)) {cacheName=postSimFixes}
   n13 = RigidBody!(
    ComponentPose=Local To Component(LocalPose=Use cached pose 'animDynPostCalcCP'())
    Alpha
    OverrideWorldGravity:Vector="0,0,0"
    SimSpaceSettings:Sim Space Settings Structure=(WorldAlpha=1,VelocityScaleZ=0,DampingAlpha=0,MaxLinearVelocity=400,MaxAngularVelocity=3,MaxLinearAcceleration=800,MaxAngularAcceleration=1,ExternalLinearDragV=(X=1,Y=1,Z=1),ExternalLinearVelocity=(X=0,Y=0,Z=0),ExternalAngularVelocity=(X=0,Y=0,Z=0))
   ) -> (Pose) {
    bClampLinearTranslationLimitToRefPose=true
    bOverrideWorldGravity=true
    overridePhysicsAsset=&/M_MED_IvoryDwell_Carabiners_Physics
    simulationSpace=BaseBoneSpace
    alphaBoolBlend {blendOption=Linear}
    baseBoneRef {boneName=thigh_l}
    componentAppliedLinearAccClamp=<3000 3000 3000>
    overrideWorldGravity {z=-980}
    simSpaceSettings {
     dampingAlpha=0
     maxAngularAcceleration=1
     maxAngularVelocity=3
     maxLinearAcceleration=800
     maxLinearVelocity=400
     velocityScaleZ=0
     worldAlpha=1
     externalLinearDragV=<1 1 1>
    }
   }
   n14 = RigidBody!(ComponentPose=@n13, Alpha, OverrideWorldGravity="0,0,0") -> (Pose) {
    bClampLinearTranslationLimitToRefPose=true
    bOverrideWorldGravity=true
    overridePhysicsAsset=&/M_MED_IvoryDwell_Hems_Physics
    simulationSpace=BaseBoneSpace
    alphaBoolBlend {blendOption=Linear}
    baseBoneRef {boneName=pelvis}
    componentAppliedLinearAccClamp=<3000 3000 3000>
    overrideWorldGravity {z=-980}
    simSpaceSettings {
     dampingAlpha=0
     maxAngularAcceleration=1
     maxAngularVelocity=3
     maxLinearAcceleration=800
     maxLinearVelocity=400
     worldAlpha=1
     externalLinearDragV=<0.2 0.2 0.2>
    }
   }
   n15 = RigidBody(ComponentPose=@n14, OverrideWorldGravity="0,0,980") {
    bClampLinearTranslationLimitToRefPose=true
    bOverrideWorldGravity=true
    overridePhysicsAsset=&/M_MED_IvoryDwell_Collar_Physics
    simulationSpace=BaseBoneSpace
    alphaBoolBlend {blendOption=Linear}
    baseBoneRef {boneName=spine_05}
    componentAppliedLinearAccClamp=<3000 3000 3000>
    overrideWorldGravity {z=-980}
    simSpaceSettings {dampingAlpha=0, maxAngularAcceleration=1, maxAngularVelocity=3, maxLinearAcceleration=500, maxLinearVelocity=500, worldAlpha=1}
   }
   Save cached pose 'DynamicsCP'(Pose=Component To Local(ComponentPose=@n15)) {cacheName=DynamicsCP}
   n16 = Anim Dynamics - Bone: dyn_strapHang_in_l!(
    ComponentPose=Local To Component(LocalPose=Use cached pose 'generalAdjuistementsCP'())
    Alpha=1
    GravityOverride:Vector="0,0,-980"
    SimSpaceSettings:Anim Phys Sim Space Settings Structure=(SimSpaceAngularAlpha=0,MaxAngularVelocity=10000,MaxAngularAcceleration=10000,ExternalAngularVelocity=(X=0,Y=0,Z=0))
   ) -> (Pose) {
    angularDampingOverride=1
    angularSpringConstant=2000
    bAngularSpring=true
    bOverrideAngularDamping=true
    bUseGravityOverride=true
    gravityScale=0.35
    simulationSpace=RootRelative
    boundBone {boneName=dyn_strapHang_in_l}
    componentAppliedLinearAccClamp=<3000 3000 3000>
    componentLinearAccScale=<0.5 0.5 1>
    componentLinearVelScale=<0.5 0.5 1>
    gravityOverride {z=-980}
    physicsBodyDefinitions [{
      boundBone {boneName=dyn_strapHang_in_l}
      boxExtents=<20 30 30>
      constraintSetup {angularLimitsMax=<0 0 0>, angularLimitsMin=<0 -10 0>, angularTarget=<1 0 0>}
      localJointOffset=<5 0 0>
     }]
   }
   n17 = Anim Dynamics - Bone: dyn_strapHang_out_l!(
    ComponentPose=@n16
    Alpha=1
    GravityOverride:Vector="0,0,-980"
    SimSpaceSettings:Anim Phys Sim Space Settings Structure=(SimSpaceAngularAlpha=1,MaxAngularVelocity=10,MaxAngularAcceleration=3,ExternalAngularVelocity=(X=0,Y=0,Z=0))
   ) -> (Pose) {
    angularDampingOverride=1
    angularSpringConstant=2000
    bAngularSpring=true
    bOverrideAngularDamping=true
    bUseGravityOverride=true
    gravityScale=0.35
    simulationSpace=RootRelative
    boundBone {boneName=dyn_strapHang_out_l}
    componentAppliedLinearAccClamp=<3000 3000 3000>
    componentLinearAccScale=<0.5 0.5 1>
    componentLinearVelScale=<0.5 0.5 1>
    gravityOverride {z=-980}
    physicsBodyDefinitions [{
      boundBone {boneName=dyn_strapHang_out_l}
      boxExtents=<20 30 30>
      constraintSetup {angularLimitsMax=<0 0 0>, angularLimitsMin=<0 -30 0>, angularTarget=<1 0 0>}
      localJointOffset=<5 0 0>
     }]
    simSpaceSettings {maxAngularAcceleration=3, maxAngularVelocity=10, simSpaceAngularAlpha=1}
   }
   n18 = Anim Dynamics - Bone: dyn_strapHang_in_r!(
    ComponentPose=@n17
    Alpha=1
    GravityOverride:Vector="0,0,-980"
    SimSpaceSettings:Anim Phys Sim Space Settings Structure=(SimSpaceAngularAlpha=1,MaxAngularVelocity=10,MaxAngularAcceleration=3,ExternalAngularVelocity=(X=0,Y=0,Z=0))
   ) -> (Pose) {
    angularDampingOverride=1
    angularSpringConstant=2000
    bAngularSpring=true
    bOverrideAngularDamping=true
    bUseGravityOverride=true
    gravityScale=0.35
    simulationSpace=RootRelative
    boundBone {boneName=dyn_strapHang_in_r}
    componentAppliedLinearAccClamp=<3000 3000 3000>
    componentLinearAccScale=<0.5 0.5 1>
    componentLinearVelScale=<0.5 0.5 1>
    gravityOverride {z=-980}
    physicsBodyDefinitions [{
      boundBone {boneName=dyn_strapHang_in_r}
      boxExtents=<20 30 30>
      constraintSetup {angularLimitsMax=<0 0 0>, angularLimitsMin=<0 -10 0>, angularTarget=<1 0 0>}
      localJointOffset=<5 0 0>
     }]
    simSpaceSettings {maxAngularAcceleration=3, maxAngularVelocity=10, simSpaceAngularAlpha=1}
   }
   n19 = Anim Dynamics - Bone: dyn_strapHang_out_r!(
    ComponentPose=@n18
    Alpha=1
    GravityOverride:Vector="0,0,-980"
    SimSpaceSettings:Anim Phys Sim Space Settings Structure=(SimSpaceAngularAlpha=1,MaxAngularVelocity=10,MaxAngularAcceleration=3,ExternalAngularVelocity=(X=0,Y=0,Z=0))
   ) -> (Pose) {
    angularDampingOverride=1
    angularSpringConstant=2000
    bAngularSpring=true
    bOverrideAngularDamping=true
    bUseGravityOverride=true
    gravityScale=0.35
    simulationSpace=RootRelative
    boundBone {boneName=dyn_strapHang_out_r}
    componentAppliedLinearAccClamp=<3000 3000 3000>
    componentLinearAccScale=<0.5 0.5 1>
    componentLinearVelScale=<0.5 0.5 1>
    gravityOverride {z=-980}
    physicsBodyDefinitions [{
      boundBone {boneName=dyn_strapHang_out_r}
      boxExtents=<20 30 30>
      constraintSetup {angularLimitsMax=<0 0 0>, angularLimitsMin=<0 -30 0>, angularTarget=<1 0 0>}
      localJointOffset=<5 0 0>
     }]
    simSpaceSettings {maxAngularAcceleration=3, maxAngularVelocity=10, simSpaceAngularAlpha=1}
   }
   n20 = Anim Dynamics - Bone: dyn_hipChains!(ComponentPose=@n19, Alpha=1, GravityOverride:Vector="0,0,-980") -> (Pose) {
    angularDampingOverride=1
    angularSpringConstant=2000
    bAngularSpring=true
    bOverrideAngularDamping=true
    bUseGravityOverride=true
    gravityScale=0.35
    simulationSpace=RootRelative
    boundBone {boneName=dyn_hipChains}
    componentAppliedLinearAccClamp=<3000 3000 3000>
    componentLinearAccScale=<1 1 1>
    componentLinearVelScale=<1 1 1>
    gravityOverride {z=-980}
    physicsBodyDefinitions [{
      boundBone {boneName=dyn_hipChains}
      boxExtents=<20 30 30>
      constraintSetup {angularLimitsMax=<0 15 0>, angularLimitsMin=<0 0 0>, angularTarget=<1 0 0>}
      localJointOffset=<10 0 0>
     }]
   }
   n21 = Anim Dynamics - Bone: dyn_calfLace_2_l!(
    ComponentPose=@n20
    Alpha=1
    GravityOverride:Vector="0,0,-980"
    SimSpaceSettings:Anim Phys Sim Space Settings Structure=(SimSpaceAngularAlpha=0,MaxAngularVelocity=10000,MaxAngularAcceleration=10000,ExternalAngularVelocity=(X=0,Y=0,Z=0))
   ) -> (Pose) {
    angularDampingOverride=1
    angularSpringConstant=2000
    bAngularSpring=true
    bOverrideAngularDamping=true
    bUseGravityOverride=true
    gravityScale=0.35
    simulationSpace=RootRelative
    boundBone {boneName=dyn_calfLace_2_l}
    componentAppliedLinearAccClamp=<3000 3000 3000>
    componentLinearAccScale=<1 1 1>
    componentLinearVelScale=<1 1 1>
    gravityOverride {z=-980}
    physicsBodyDefinitions [{
      boundBone {boneName=dyn_calfLace_2_l}
      boxExtents=<20 20 20>
      constraintSetup {angularLimitsMax=<0 25 10>, angularLimitsMin=<0 -25 -10>, angularTarget=<1 0 0>}
      localJointOffset=<3 0 0>
     }]
   }
   n22 = Anim Dynamics - Bone: dyn_calfLace_1_l!(
    ComponentPose=@n21
    Alpha=1
    GravityOverride:Vector="0,0,-980"
    SimSpaceSettings:Anim Phys Sim Space Settings Structure=(SimSpaceAngularAlpha=0,MaxAngularVelocity=10000,MaxAngularAcceleration=10000,ExternalAngularVelocity=(X=0,Y=0,Z=0))
   ) -> (Pose) {
    angularDampingOverride=1
    angularSpringConstant=2000
    bAngularSpring=true
    bOverrideAngularDamping=true
    bUseGravityOverride=true
    gravityScale=0.35
    simulationSpace=RootRelative
    boundBone {boneName=dyn_calfLace_1_l}
    componentAppliedLinearAccClamp=<3000 3000 3000>
    componentLinearAccScale=<1 1 1>
    componentLinearVelScale=<1 1 1>
    gravityOverride {z=-980}
    physicsBodyDefinitions [{
      boundBone {boneName=dyn_calfLace_1_l}
      boxExtents=<20 20 20>
      constraintSetup {angularLimitsMax=<0 25 10>, angularLimitsMin=<0 -25 -10>, angularTarget=<1 0 0>}
      localJointOffset=<3 0 0>
     }]
   }
   n23 = Anim Dynamics - Bone: dyn_hood!(ComponentPose=@n22, bAlphaBoolEnabled:bool=True, GravityOverride:Vector="0,0,-980") -> (Pose) {
    alphaInputType=Bool
    angularDampingOverride=1
    angularSpringConstant=5000
    bAngularSpring=true
    bOverrideAngularDamping=true
    bUseGravityOverride=true
    gravityScale=0.35
    simulationSpace=RootRelative
    alphaBoolBlend {blendOption=Linear}
    boundBone {boneName=dyn_hood}
    componentAppliedLinearAccClamp=<3000 3000 3000>
    componentLinearAccScale=<0.5 0.5 0.5>
    componentLinearVelScale=<0.5 0.5 0.5>
    gravityOverride {z=-980}
    physicsBodyDefinitions [{
      boundBone {boneName=dyn_hood}
      boxExtents=<20 20 20>
      constraintSetup {angularLimitsMax=<0 5 0>, angularLimitsMin=<0 -10 -60>, angularTarget=<1 0 0>}
      localJointOffset=<1.5 0 0>
     }]
   }
   n24 = "Constraint - target_strap_01_l (PP)"!(ComponentPose=@n23, Alpha=1, ConstraintWeights_0=0.9725, ConstraintWeights_1:float=0.0275) -> (Pose) {
    boneToModify {boneName=target_strap_01_l}
    constraintSetup [{targetBone {boneName=thigh_twist_00_l}}, {targetBone {boneName=deform_groin_l}}]
    constraintWeights [1,1]
   }
   n25 = "Constraint - target_strap_02_l (PP)"!(ComponentPose=@n24, Alpha=1, ConstraintWeights_0=0.8578, ConstraintWeights_1:float=0.1422) -> (Pose) {
    boneToModify {boneName=target_strap_02_l}
    constraintSetup [{targetBone {boneName=dyn_strapHang_out_l}}, {targetBone {boneName=thigh_twist_00_l}}]
    constraintWeights [1,1]
   }
   n26 = "Constraint - coll_strap_02_l (PP)"!(ComponentPose=@n25, Alpha=1, ConstraintWeights_0=0.951, ConstraintWeights_1:float=0.049) -> (Pose) {
    boneToModify {boneName=coll_strap_02_l}
    constraintSetup [{targetBone {boneName=thigh_twist_00_l}}, {targetBone {boneName=deform_groin_l}}]
    constraintWeights [1,1]
   }
   n27 = "Constraint - coll_strap_01_l (PPPPPP)"!(
    ComponentPose=@n26
    Alpha=1
    ConstraintWeights_0=0.2762
    ConstraintWeights_1:float=0.2681
    ConstraintWeights_2:float=0.2133
    ConstraintWeights_3:float=0.1618
    ConstraintWeights_4:float=0.0556
    ConstraintWeights_5:float=0.025
   ) -> (Pose) {
    boneToModify {boneName=coll_strap_01_l}
    constraintSetup [
     {targetBone {boneName=thigh_twist_01_l}}
     {targetBone {boneName=kin_strapStart_l}}
     {targetBone {boneName=deform_groin_l}}
     {targetBone {boneName=kin_strap_frontTris_l}}
     {targetBone {boneName=thigh_twist_01_r}}
     {targetBone {boneName=deform_groin_r}}
    ]
    constraintWeights [1,1,1,1,1,1]
   }
   n28 = Look At - Bone: coll_strap_01_l(ComponentPose=@n27, Alpha=1) {
    boneToModify {boneName=coll_strap_01_l}
    lookAtLocation {x=0}
    lookAtTarget {boneReference {boneName=target_strap_01_l}}
    lookAt_Axis {axis=<1 0>}
    lookUp_Axis {axis {x=0, z=1}}
   }
   n29 = Look At - Bone: coll_strap_02_l(ComponentPose=@n28, Alpha=1) {
    boneToModify {boneName=coll_strap_02_l}
    lookAtLocation {x=0}
    lookAtTarget {boneReference {boneName=target_strap_02_l}}
    lookAt_Axis {axis=<1 0>}
    lookUp_Axis {axis {x=0, z=1}}
   }
   n30 = "Constraint - target_hem_f_r (PPPPP)"!(
    ComponentPose=@n29
    Alpha=1
    ConstraintWeights_0=0.5467
    ConstraintWeights_1:float=0.2753
    ConstraintWeights_2:float=0.09
    ConstraintWeights_3:float=0.0671
    ConstraintWeights_4:float=0.0209
   ) -> (Pose) {
    boneToModify {boneName=target_hem_f_r}
    constraintSetup [
     {targetBone {boneName=dyn_hipChains_front}}
     {targetBone {boneName=thigh_twist_01_r}}
     {targetBone {boneName=deform_groin_r}}
     {targetBone {boneName=pelvis}}
     {targetBone {boneName=dyn_hipChains_back}}
    ]
    constraintWeights [1,1,1,1,1]
   }
   n31 = "Constraint - coll_hem_f_r (PP)"!(ComponentPose=@n30, Alpha=1, ConstraintWeights_0=0.5524, ConstraintWeights_1:float=0.4476) -> (Pose) {
    boneToModify {boneName=coll_hem_f_r}
    constraintSetup [{targetBone {boneName=dyn_hipChains_front}}, {targetBone {boneName=dyn_hipChains_back}}]
    constraintWeights [1,1]
   }
   n32 = Look At - Bone: coll_hem_f_r(ComponentPose=@n31, Alpha=1) {
    boneToModify {boneName=coll_hem_f_r}
    lookAtLocation {x=0}
    lookAtTarget {boneReference {boneName=target_hem_f_r}}
    lookAt_Axis {axis=<1 0>}
    lookUp_Axis {axis {x=0, z=1}}
   }
   Save cached pose 'animDynPostCalcCP'(Pose=Component To Local(ComponentPose=@n32)) {cacheName=animDynPostCalcCP}
  ]}
 EventGraph {nodes [
   Event BlueprintInitializeAnimation()
   n1 = TryGetPawnOwner()
   > Is Valid(InputObject=@n1)
   > n2 = Cast To FortPlayerPawnAthena(Object=@n1)
   > Set PlayerPawn(PlayerPawn:Fort Player Pawn Athena Object Reference=@n2) -> (Output_Get:Fort Player Pawn Athena Object Reference)
   Event OnParachuteOpened()
   > Switch on EGliderType(Selection=Get PlayerGliderType() -> (PlayerGliderType=HangGlider))
   > Set isUsingHangGlidingInParachute(isUsingHangGlidingInParachute:bool=true) -> (Output_Get:bool)
   Event OnParachuteClosed()
   > Set isUsingHangGlidingInParachute(isUsingHangGlidingInParachute=false)
   n3 = Event OnCharacterCustomizationLoaded()
   n4 = CheckForBackbling()
   n5 = OR Boolean!(A=@n4, B=@n4.bIsBackblingACape, C:bool=@n4.bIsBackblingLong) -> (ReturnValue)
   > Set isWearingBlackbling(isWearingBlackbling:bool=@n5) -> (Output_Get:bool)
   > Set isNotWearingBlackbling(isNotWearingBlackbling:bool=NOT Boolean(A=@n5)) -> (Output_Get:bool)
   n6 = CheckForBackbling()
   n7 = OR Boolean!(A=@n6, B=@n6.bIsBackblingLong) -> (ReturnValue)
   > Set isWearingBackpack(isWearingBackpack:bool=NOT Boolean(A="Equal (Boolean)"(A:bool=@n7, B:bool=@n6.bIsBackblingLong) -> (ReturnValue:bool=~) {is=title K2Node_PromotableOperator})) -> (Output_Get:bool)
   > Set SkelMeshBackpack(SkelMeshBackpack:Skeletal Mesh Component Budgeted Object Reference=Get Array Item with EFort Custom Part Type(Array=@n3.CustomMeshParts, PartType=Backpack)) -> (Output_Get:Skeletal Mesh Component Budgeted Object Reference)
   n8 = GetDisplayName(Object=Get SkelMeshBackpack() -> (SkelMeshBackpack:Skeletal Mesh Component Budgeted Object Reference))
   n9 = CheckForBackbling()
   > Set isWearingItsBP(isWearingItsBP:bool=AND Boolean(
     A=OR Boolean!(A=@n9, B=@n9.bIsBackblingACape, C:bool=@n9.bIsBackblingLong) -> (ReturnValue)
     B=MatchesWildcard(SourceString=@n8, Wildcard=*M_MED_IvoryDwell_Pack, SearchCase=CaseSensitive)
    )) -> (Output_Get:bool)
   > n10 = Sequence()
   n11 = Branch(execute=@n10.then_1, Condition=Get bIsInFrontEnd())
   > Set SimSpace animDyn output(SimSpace animDyn output:Anim Phys Sim Space Settings Structure=Get SimSpace animDyn FE() -> (SimSpace animDyn FE:Anim Phys Sim Space Settings Structure)) -> (Output_Get:Anim Phys Sim Space Settings Structure)
   > Set Sim Space RB output(Sim Space RB output:Sim Space Settings Structure=Get Sim Space RB FE() -> (Sim Space RB FE:Sim Space Settings Structure)) -> (Output_Get:Sim Space Settings Structure)
   Set SimSpace animDyn output(
    execute=@n11.else
    SimSpace animDyn output=Get SimSpace animDyn default() -> (SimSpace animDyn default:Anim Phys Sim Space Settings Structure)
   )
   > Set Sim Space RB output(Sim Space RB output=Get Sim Space RB default() -> (Sim Space RB default:Sim Space Settings Structure))
   Set BodyMesh(execute=@n10.then_0, BodyMesh=GetPartSkeletalMeshComponent())
   n12 = Get BodyMesh() -> (BodyMesh:Skeletal Mesh Component Object Reference)
   > n13 = Get Dynamic Material Instance(SkeletalMeshComponent=@n12)
   > Set Mat Instance Dynamic(Mat Instance Dynamic=@n13)
   > n14 = GetOrCreateDynamicOverlayMaterial(MeshComponent=@n12, MaterialName=None)
   > Set Overlay Mat Instance Dynamic(Overlay Mat Instance Dynamic=@n14)
   Event OnRightLegHidden() {is=title K2Node_Event}
   > Set isRightLegHidden(isRightLegHidden:bool=true) -> (Output_Get:bool)
   n15 = Event BlueprintUpdateAnimation()
   > n16 = Sequence!(execute) -> (then_0, then_1, then_2:Exec)
   > n17 = "Gravity Override State Function (Parameter Asset) (Nativized)"(
    CharacterPartAnimInstance=Self-Reference()
    GameTimeInSeconds=GetGameTimeInSeconds()
    Parameters=&/M_MED_IvoryDwell_Parameters
   )
   > Set GravityOverride(GravityOverride=@n17)
   n18 = GetOwnerSkeletalMeshComponent()
   > Set strapL animDyn alpha(strapL animDyn alpha:double=GetAlphaFromReferenceTransformDelta(SkeletalMeshComponent=@n18, BoneName=thigh_l, BaseName=pelvis, InputRange=(X=-50,Y=-100), OutputRange=(X=1,Y=0), TwistAxis="-1,0,0")) -> (Output_Get:double)
   > Set strapR animDyn alpha(strapR animDyn alpha:double=GetAlphaFromReferenceTransformDelta(SkeletalMeshComponent=@n18, BoneName=thigh_r, BaseName=pelvis, InputRange=(X=-50,Y=-100), OutputRange=(X=1,Y=0), TwistAxis="1,0,0")) -> (Output_Get:double)
   Is Valid(exec=@n16.then_2, InputObject=Get PlayerPawn() -> (PlayerPawn:Fort Player Pawn Athena Object Reference))
   > n19 = Branch(Condition=Get bIsInFrontEnd())
   Set WPO Globals Function(
    execute=@n19.else
    Material=Get Mat Instance Dynamic()
    Overlay Material=Get Overlay Mat Instance Dynamic()
    Skeletal Mesh=Get BodyMesh()
    "Local Noise Transform  Location"=Get Local Noise Transform Location() -> (Local Noise Transform Location:Vector)
    Local Noise Offset=Get Local Noise Offset()
   )
   > Set WPO Regional Function!(
    execute
    Material:Material Instance Dynamic Object Reference=Get Mat Instance Dynamic()
    Overlay Material:Material Instance Dynamic Object Reference=Get Overlay Mat Instance Dynamic()
    Skeletal Mesh:Skeletal Mesh Component Object Reference=Get BodyMesh()
    Velocity Transform:Name=foot_l
    Velocity Interp Speed:double=Get Velocity Interp Speed()
    Collision Transform:Name=calf_l
    Collision Transform Axis:EAxis Enum
    Invert Collision Transform Axis:bool=true
    Delta Time X:double=@n15.DeltaTimeX
    Previous Velocity Transform Location:"Vector (by ref)"=Get R1 Velocity Transform Location()
    Previous Velocity Transform Velocity:"Vector (by ref)"=Get R1 Velocity Transform Velocity()
    Region:int=1
   ) -> (then)
   > Set WPO Regional Function(
    Material=Get Mat Instance Dynamic()
    Overlay Material=Get Overlay Mat Instance Dynamic()
    Skeletal Mesh=Get BodyMesh()
    Velocity Transform=foot_r
    Velocity Interp Speed=Get Velocity Interp Speed()
    Collision Transform=calf_r
    Invert Collision Transform Axis=true
    Delta Time X=@n15.DeltaTimeX
    Previous Velocity Transform Location=Get R2 Velocity Transform Location()
    Previous Velocity Transform Velocity=Get R2 Velocity Transform Velocity()
    Region=2
   )
   > Set WPO Regional Function(
    Material=Get Mat Instance Dynamic()
    Overlay Material=Get Overlay Mat Instance Dynamic()
    Skeletal Mesh=Get BodyMesh()
    Velocity Transform=hand_l
    Velocity Interp Speed=Get Velocity Interp Speed()
    Collision Transform=clavicle_l
    Invert Collision Transform Axis=true
    Delta Time X=@n15.DeltaTimeX
    Previous Velocity Transform Location=Get R1 Velocity Transform Location()
    Previous Velocity Transform Velocity=Get R1 Velocity Transform Velocity()
    Region=1
   )
   > Set WPO Regional Function(
    Material=Get Mat Instance Dynamic()
    Overlay Material=Get Overlay Mat Instance Dynamic()
    Skeletal Mesh=Get BodyMesh()
    Velocity Transform=hand_r
    Velocity Interp Speed=Get Velocity Interp Speed()
    Collision Transform=clavicle_r
    Invert Collision Transform Axis=true
    Delta Time X=@n15.DeltaTimeX
    Previous Velocity Transform Location=Get R2 Velocity Transform Location()
    Previous Velocity Transform Velocity=Get R2 Velocity Transform Velocity()
    Region=2
   )
   > n20 = Branch(Condition=OR Boolean!(A=Get bIsSkydiveFloating(), B=Get bIsParachuteOpened()) -> (ReturnValue))
   > Set WPO Pho Wind Intensity(WPO Pho Wind Intensity=0.8)
   > n21 = Set WPO Pho Turbulence(WPO Pho Turbulence:double=4) -> (Output_Get:double)
   SetScalarParameterValue(
    execute=[@n21.then, @n23.then]
    self=[Get Overlay Mat Instance Dynamic(), Get Mat Instance Dynamic()]
    ParameterName=ClothSim_WindStrength
    Value=Get WPO Pho Wind Intensity()
   )
   n22 = Get WPO Pho Turbulence() -> (WPO Pho Turbulence:double)
   > SetScalarParameterValue(self=[Get Overlay Mat Instance Dynamic(), Get Mat Instance Dynamic()], ParameterName=ClothSim_R1_TurbulenceStrength, Value=@n22)
   > SetScalarParameterValue(self=[Get Mat Instance Dynamic(), Get Overlay Mat Instance Dynamic()], ParameterName=ClothSim_R2_TurbulenceStrength, Value=@n22)
   Set WPO Pho Wind Intensity(execute=@n20.else, WPO Pho Wind Intensity=0)
   > n23 = Set WPO Pho Turbulence(WPO Pho Turbulence=0)
   n24 = Branch(execute=@n16.then_1, Condition="integer >= integer"(A=Get CurrentLOD(), B=2))
   > n25 = Sequence!(execute) -> (then_0, then_1)
   > Do Once(Reset=@n27.then_0)
   n26 = GetPartSkeletalMeshComponent()
   > SetSkinWeightProfile!(execute, self:Skinned Mesh Component Object Reference=@n26, InProfileName:Name=Alternative, InLayer:ESkinWeightProfileLayer Enum=Primary) -> (then, ReturnValue:bool)
   Do Once(execute=@n27.then_1, Reset=@n25.then_1)
   > ClearSkinWeightProfile!(execute, self:Skinned Mesh Component Object Reference=@n26, InLayer:ESkinWeightProfileLayer Enum=Primary) -> (then)
   n27 = Sequence(execute=@n24.else)
   Event OnLeftLegVisible() {is=title K2Node_Event}
   > Set isLeftLegHidden(isLeftLegHidden:bool=false) -> (Output_Get:bool)
   Event OnRightLegVisible() {is=title K2Node_Event}
   > Set isRightLegHidden(isRightLegHidden=false)
   Event OnLeftLegHidden() {is=title K2Node_Event}
   > Set isLeftLegHidden(isLeftLegHidden=true)
  ]}
]
variables [
 BodyMesh {type=SkeletalMeshComponent}
 GravityOverride {type=Vector}
 isLeftLegHidden {type=boolean}
 isNotWearingBlackbling {type=boolean}
 isRightLegHidden {type=boolean}
 isUsingHangGlidingInParachute {type=boolean}
 isWearingBackpack {type=boolean}
 isWearingBlackbling {type=boolean}
 isWearingItsBP {type=boolean}
 Local Noise Offset {type=Vector}
 Local Noise Transform Location {type=Vector}
 Mat Instance Dynamic {type=MaterialInstanceDynamic}
 Overlay Mat Instance Dynamic {type=MaterialInstanceDynamic}
 PlayerPawn {type=FortPlayerPawnAthena}
 R1 Velocity Transform Location {type=Vector}
 R1 Velocity Transform Velocity {type=Vector}
 R2 Velocity Transform Location {type=Vector}
 R2 Velocity Transform Velocity {type=Vector}
 Sim Space RB default {
  type=SimSpaceSettings
  default {
   dampingAlpha=0
   maxAngularAcceleration=1
   maxAngularVelocity=3
   maxLinearAcceleration=800
   maxLinearVelocity=400
   velocityScaleZ=0
   worldAlpha=1
   externalLinearDragV=<1 1 1>
  }
 }
 Sim Space RB FE {
  type=SimSpaceSettings
  default {
   dampingAlpha=0
   maxAngularAcceleration=1
   maxAngularVelocity=3
   maxLinearAcceleration=1200
   maxLinearVelocity=400
   velocityScaleZ=0
   worldAlpha=1
   externalLinearDragV=<1 1 1>
  }
 }
 Sim Space RB output {
  type=SimSpaceSettings
  default {
   dampingAlpha=1
   maxAngularAcceleration=10000
   maxAngularVelocity=10000
   maxLinearAcceleration=10000
   maxLinearVelocity=10000
   velocityScaleZ=1
   worldAlpha=0
   externalLinearDragV=<0 0 0>
  }
 }
 SimSpace animDyn default {
  type=AnimPhysSimSpaceSettings
  default {maxAngularAcceleration=10000, maxAngularVelocity=10000, simSpaceAngularAlpha=0, externalAngularVelocity=<0 0 0>}
 }
 SimSpace animDyn FE {
  type=AnimPhysSimSpaceSettings
  default {maxAngularAcceleration=1, maxAngularVelocity=6, simSpaceAngularAlpha=1, externalAngularVelocity=<0 0 0>}
 }
 SimSpace animDyn output {
  type=AnimPhysSimSpaceSettings
  default {maxAngularAcceleration=10000, maxAngularVelocity=10000, simSpaceAngularAlpha=0, externalAngularVelocity=<0 0 0>}
 }
 SkelMeshBackpack {type=SkeletalMeshComponentBudgeted}
 strapL animDyn alpha {type=number}
 strapR animDyn alpha {type=number}
 Velocity Interp Speed {type=number, default=20}
 WPO Pho Turbulence {type=number}
 WPO Pho Wind Intensity {type=number}
]
