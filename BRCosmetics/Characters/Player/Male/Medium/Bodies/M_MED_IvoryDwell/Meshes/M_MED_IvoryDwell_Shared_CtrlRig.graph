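# Control Rig for the IvoryDwell medium male body that drives corrective deformation and accessory bones from the body pose.
# Parent and aim constraints set base placement for hip, ankle, jacket hem, hip chain and carabiner bones.
# Spherical pose readers, twist extraction, rotation-delta alphas and inverse-distance pose weights, often squared or multiplied together, drive additive transform offsets.
# These offsets adjust glutes, groin, knees, ankles, straps and hanging gear as limbs bend.
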ControlRigBlueprint /BRCosmetics/Characters/Player/Male/Medium/Bodies/M_MED_IvoryDwell/Meshes/M_MED_IvoryDwell_Shared_CtrlRig
graphs [
 CollapseNode_ContainedGraph {nodes [
   n1 = RigVMFunctionEntryNode()
   &RigUnit_ParentConstraint(
    AdvancedSettings=(RotationOrderForFilter=XZY)
    Child=(Type=Bone,Name="deform_groin_base_l")
    Parents=((Item=(Type=Bone,Name="thigh_twist_01_l"),Weight=0.7),(Item=(Type=Bone,Name="pelvis"),Weight=0.3))
    Weight=1
    ExecutePin=@n1.ExecutePin
   )
   > &RigUnit_ParentConstraint(
    AdvancedSettings=(RotationOrderForFilter=XZY)
    Child=(Type=Bone,Name="deform_groin_base_r")
    Parents=((Item=(Type=Bone,Name="thigh_twist_01_r"),Weight=0.7),(Item=(Type=Bone,Name="pelvis"),Weight=0.3))
    Weight=1
   )
   > &RigUnit_ParentConstraint(
    AdvancedSettings=(RotationOrderForFilter=XZY)
    Child=(Type=Bone,Name="deform_glute_base_l")
    Parents=((Item=(Type=Bone,Name="thigh_twist_01_l"),Weight=0.25),(Item=(Type=Bone,Name="pelvis"),Weight=0.75))
    Weight=1
   )
   > &RigUnit_ParentConstraint(
    AdvancedSettings=(RotationOrderForFilter=XZY)
    Child=(Type=Bone,Name="deform_glute_base_r")
    Parents=((Item=(Type=Bone,Name="thigh_twist_01_r"),Weight=0.25),(Item=(Type=Bone,Name="pelvis"),Weight=0.75))
    Weight=1
   )
   > n2 = &RigUnit_SphericalPoseReader(
    ActiveRegionScaleFactors=(PositiveWidth=3.5,NegativeWidth=0.5,PositiveHeight=3,NegativeHeight=3)
    ActiveRegionSize=0.1
    Debug=(bDrawDebug=false,DebugScale=25,DebugSegments=20,DebugThickness=0.25)
    DriverAxis=(X=-1,Y=0,Z=0)
    DriverItem=(Type=Bone,Name="thigh_l")
    FalloffRegionScaleFactors=(PositiveWidth=1,NegativeWidth=1,PositiveHeight=1,NegativeHeight=1)
    FalloffSize=0.45
    OptionalParentItem=(Type=Bone)
    RotationOffset=(X=90,Y=0,Z=0)
   )
   > &RigUnit_ModifyTransforms(
    ItemToModify="((Item=(Type=Bone,Name=\"deform_glute_l\"),Transform=(Rotation=(X=0,Y=0,Z=0,W=1),Translation=(X=-7,Y=0,Z=0),Scale3D=(X=1,Y=1,Z=1))),(Item=(Type=Bone,Name=\"deform_groin_l\"),Transform=(Rotation=(X=0,Y=0,Z=0,W=1),Translation=(X=-3,Y=12,Z=0),Scale3D=(X=1,Y=1,Z=1 ...(truncated)"
    Weight=@n2.OutputParam
   )
   > n3 = &RigUnit_SphericalPoseReader(
    ActiveRegionScaleFactors=(PositiveWidth=3.5,NegativeWidth=0.5,PositiveHeight=3,NegativeHeight=3)
    ActiveRegionSize=0.1
    Debug=(bDrawDebug=false,DebugScale=25,DebugSegments=20,DebugThickness=0.25)
    DriverAxis=(X=1,Y=0,Z=0)
    DriverItem=(Type=Bone,Name="thigh_r")
    FalloffRegionScaleFactors=(PositiveWidth=1,NegativeWidth=1,PositiveHeight=1,NegativeHeight=1)
    FalloffSize=0.45
    FlipHeightScaling=false
    FlipWidthScaling=true
    OptionalParentItem=(Type=Bone)
    RotationOffset=(X=-90,Y=0,Z=0)
   )
   > n4 = &RigUnit_ModifyTransforms(
    ItemToModify="((Item=(Type=Bone,Name=\"deform_glute_r\"),Transform=(Rotation=(X=0,Y=0,Z=0,W=1),Translation=(X=7,Y=0,Z=0),Scale3D=(X=1,Y=1,Z=1))),(Item=(Type=Bone,Name=\"deform_groin_r\"),Transform=(Rotation=(X=0,Y=0,Z=0,W=1),Translation=(X=3,Y=-12,Z=0),Scale3D=(X=1,Y=1,Z=1) ...(truncated)"
    Weight=@n3.OutputParam
   )
   RigVMFunctionReturnNode(ExecutePin=@n4.ExecutePin)
  ]}
 CollapseNode_ContainedGraph {nodes [
   RigVMFunctionEntryNode()
   > GetAlphaFromQuatDelta = RigVMFunctionReferenceNode(
    Rotation Order=ZYX
    child=(Type=Bone,Name="calf_l")
    outputMinMax=(X=0,Y=1)
    parent=(Type=Null,Name="calf_l_null")
    premult=1
    rotationMinMax=(X=120,Y=180)
    triggerAxis=Z
    twistAxis=(X=1,Y=0,Z=0)
   )
   > &RigUnit_ModifyTransforms(
    ItemToModify="((Item=(Type=Bone,Name=\"deform_knee_in_l\"),Transform=(Rotation=(X=0,Y=0,Z=0,W=1),Translation=(X=0,Y=-15,Z=0),Scale3D=(X=2,Y=1,Z=1))),(Item=(Type=Bone,Name=\"calf_twist_02_l\"),Transform=(Rotation=(X=0,Y=0,Z=0,W=1),Translation=(X=0,Y=4,Z=0),Scale3D=(X=1,Y=1,Z ...(truncated)"
    WeightMinimum=0
    Weight=@GetAlphaFromQuatDelta.alpha
   )
   > GetAlphaFromQuatDelta_1 = RigVMFunctionReferenceNode(
    Rotation Order=ZYX
    child=(Type=Bone,Name="calf_r")
    outputMinMax=(X=0,Y=1)
    parent=(Type=Null,Name="calf_r_null")
    premult=1
    rotationMinMax=(X=120,Y=180)
    triggerAxis=Z
    twistAxis=(X=1,Y=0,Z=0)
   )
   > &RigUnit_ModifyTransforms(
    ItemToModify="((Item=(Type=Bone,Name=\"deform_knee_in_r\"),Transform=(Rotation=(X=0,Y=0,Z=0,W=1),Translation=(X=0,Y=15,Z=0),Scale3D=(X=2,Y=1,Z=1))),(Item=(Type=Bone,Name=\"calf_twist_02_r\"),Transform=(Rotation=(X=0,Y=0,Z=0,W=1),Translation=(X=0,Y=-4,Z=0),Scale3D=(X=1,Y=1,Z ...(truncated)"
    WeightMinimum=0
    Weight=@GetAlphaFromQuatDelta_1.alpha
   )
   > RigVMFunctionReturnNode()
  ]}
 CollapseNode_ContainedGraph {nodes [
   n1 = RigVMFunctionEntryNode()
   RigVMFunction_Sequence = &RigVMFunction_Sequence(ExecuteContext=@n1.ExecutePin)
   constraints = RigVMCollapseNode(ExecutePin=@RigVMFunction_Sequence.A)
   n2 = &RigVMFunction_Sequence(ExecuteContext=@constraints.ExecutePin)
   Power = /Script/RigVM.RigVMFunction_MathDoublePow(B=2, A=RigVMVariableNode() {variable=thigh_l_rz_0_100}.Value) {is=struct RigVMUnitNode, template="Power::Execute(in A,in B,out Result)"}
   &RigUnit_ModifyTransforms(
    ItemToModify="((Item=(Type=Bone,Name=\"base_hemJacket_1_l\"),Transform=(Rotation=(X=0.0972,Y=0.0893,Z=-0.4587,W=0.8787),Translation=(X=0,Y=0,Z=0),Scale3D=(X=1,Y=1,Z=1))),(Item=(Type=Bone,Name=\"base_hemJacket_2_l\"),Transform=(Rotation=(X=0.0674,Y=-0.1711,Z=-0.2456,W=0.9518 ...(truncated)"
    WeightMinimum=0
    Weight=@Power.Result
    ExecutePin=@n2.A
   )
   > n3 = &RigUnit_SphericalPoseReader(
    ActiveRegionScaleFactors=(PositiveWidth=0.8,NegativeWidth=1,PositiveHeight=0.8,NegativeHeight=1)
    ActiveRegionSize=0.1
    Debug=(bDrawDebug=false,bDraw2D=False,bDrawLocalAxes=False,DebugScale=25,DebugSegments=20,DebugThickness=0.25)
    DriverAxis=(X=-1,Y=0,Z=0)
    DriverItem=(Type=Bone,Name="thigh_l")
    FalloffRegionScaleFactors=(PositiveWidth=1,NegativeWidth=1,PositiveHeight=1,NegativeHeight=1)
    FalloffSize=0.2
    FlipHeightScaling=False
    FlipWidthScaling=False
    OptionalParentItem=(Type=Bone,Name="None")
    RotationOffset=(X=140,Y=0,Z=-15)
   )
   > &RigUnit_ModifyTransforms(
    ItemToModify="((Item=(Type=Bone,Name=\"base_hemJacket_2_l\"),Transform=(Rotation=(X=0,Y=0,Z=-0.2588,W=0.9659),Translation=(X=0,Y=0,Z=1),Scale3D=(X=1,Y=1,Z=1))),(Item=(Type=Bone,Name=\"base_hemJacket_1_l\"),Transform=(Rotation=(X=-0.1565,Y=0.0864,Z=-0.2473,W=0.9523),Translat ...(truncated)"
    WeightMinimum=0
    Weight=@n3.OutputParam
   )
   &RigUnit_ModifyTransforms(
    ItemToModify=((Item=(Type=Bone,Name="base_hemJacket_1_r"),Transform=(Rotation=(X=-0.1998,Y=-0.0271,Z=-0.5362,W=0.8196),Translation=(X=0,Y=0,Z=-1),Scale3D=(X=1,Y=1,Z=1))))
    WeightMinimum=0
    Weight=&RigVMFunction_MathDoublePow(B=2, A=RigVMVariableNode() {variable=thigh_r_rz_0_100}.Value) {template="Power::Execute(in A,in B,out Result)"}.Result
    ExecutePin=@n2.B
   )
   > n4 = &RigUnit_SphericalPoseReader(
    ActiveRegionScaleFactors=(PositiveWidth=1,NegativeWidth=0.8,PositiveHeight=0.8,NegativeHeight=1)
    ActiveRegionSize=0.1
    Debug=(bDrawDebug=false,bDraw2D=False,bDrawLocalAxes=False,DebugScale=25,DebugSegments=20,DebugThickness=0.25)
    DriverAxis=(X=1,Y=0,Z=0)
    DriverItem=(Type=Bone,Name="thigh_r")
    FalloffRegionScaleFactors=(PositiveWidth=1,NegativeWidth=1,PositiveHeight=1,NegativeHeight=1)
    FalloffSize=0.25
    FlipHeightScaling=False
    FlipWidthScaling=False
    OptionalParentItem=(Type=Bone,Name="None")
    RotationOffset=(X=-40,Y=0,Z=-15)
   )
   > &RigUnit_ModifyTransforms(
    ItemToModify="((Item=(Type=Bone,Name=\"base_hemJacket_2_r\"),Transform=(Rotation=(X=0,Y=0,Z=-0.342,W=0.9397),Translation=(X=0,Y=0,Z=1),Scale3D=(X=1,Y=1,Z=1))),(Item=(Type=Bone,Name=\"base_hemJacket_1_r\"),Transform=(Rotation=(X=0.1549,Y=-0.0891,Z=-0.2639,W=0.9478),Translati ...(truncated)"
    WeightMinimum=0
    Weight=@n4.OutputParam
   )
   RigVMFunctionReturnNode(ExecutePin=@RigVMFunction_Sequence.B)
  ]}
 CollapseNode_ContainedGraph {nodes [
   n1 = RigVMFunctionEntryNode()
   RigVMFunction_Sequence = RigVMAggregateNode(ExecuteContext=@n1.ExecutePin)
   &RigUnit_ParentConstraint(
    AdvancedSettings=(RotationOrderForFilter=XZY)
    Child=(Type=Bone,Name="base_hemJacket_1_r")
    Filter=(TranslationFilter=(bX=True,bY=True,bZ=True),RotationFilter=(bX=True,bY=True,bZ=True),ScaleFilter=(bX=False,bY=False,bZ=False))
    Parents=((Item=(Type=Bone,Name="spine_04"),Weight=0.0289),(Item=(Type=Bone,Name="spine_01"),Weight=0.1889),(Item=(Type=Bone,Name="spine_03"),Weight=0.2049),(Item=(Type=Bone,Name="spine_02"),Weight=0.2733),(Item=(Type=Bone,Name="pelvis"),Weight=0.304))
    Weight=1
    ExecutePin=@RigVMFunction_Sequence.A
   )
   > &RigUnit_ParentConstraint(
    AdvancedSettings=(RotationOrderForFilter=XZY)
    Child=(Type=Bone,Name="base_hemJacket_1_l")
    Filter=(TranslationFilter=(bX=True,bY=True,bZ=True),RotationFilter=(bX=True,bY=True,bZ=True),ScaleFilter=(bX=False,bY=False,bZ=False))
    Parents=((Item=(Type=Bone,Name="spine_04"),Weight=0.0289),(Item=(Type=Bone,Name="spine_01"),Weight=0.1889),(Item=(Type=Bone,Name="spine_03"),Weight=0.2049),(Item=(Type=Bone,Name="spine_02"),Weight=0.2733),(Item=(Type=Bone,Name="pelvis"),Weight=0.304))
    Weight=1
   )
   > &RigUnit_ParentConstraint(
    AdvancedSettings=(RotationOrderForFilter=XZY)
    Child=(Type=Bone,Name="base_hemJacket_2_r")
    Filter=(TranslationFilter=(bX=True,bY=True,bZ=True),RotationFilter=(bX=True,bY=True,bZ=True),ScaleFilter=(bX=False,bY=False,bZ=False))
    Parents=((Item=(Type=Bone,Name="spine_04"),Weight=0.0227),(Item=(Type=Bone,Name="pelvis"),Weight=0.2015),(Item=(Type=Bone,Name="spine_01"),Weight=0.2217),(Item=(Type=Bone,Name="spine_03"),Weight=0.2243),(Item=(Type=Bone,Name="spine_02"),Weight=0.3298))
    Weight=1
   )
   > &RigUnit_ParentConstraint(
    AdvancedSettings=(RotationOrderForFilter=XZY)
    Child=(Type=Bone,Name="base_hemJacket_2_l")
    Filter=(TranslationFilter=(bX=True,bY=True,bZ=True),RotationFilter=(bX=True,bY=True,bZ=True),ScaleFilter=(bX=False,bY=False,bZ=False))
    Parents=((Item=(Type=Bone,Name="spine_04"),Weight=0.0227),(Item=(Type=Bone,Name="pelvis"),Weight=0.2015),(Item=(Type=Bone,Name="spine_01"),Weight=0.2217),(Item=(Type=Bone,Name="spine_03"),Weight=0.2243),(Item=(Type=Bone,Name="spine_02"),Weight=0.3298))
    Weight=1
   )
   > &RigUnit_ParentConstraint(
    AdvancedSettings=(RotationOrderForFilter=XZY)
    Child=(Type=Bone,Name="base_hemJacket_3_r")
    Filter=(TranslationFilter=(bX=True,bY=True,bZ=True),RotationFilter=(bX=True,bY=True,bZ=True),ScaleFilter=(bX=False,bY=False,bZ=False))
    Parents=((Item=(Type=Bone,Name="spine_04"),Weight=0.0329),(Item=(Type=Bone,Name="spine_01"),Weight=0.2014),(Item=(Type=Bone,Name="pelvis"),Weight=0.2104),(Item=(Type=Bone,Name="spine_03"),Weight=0.2261),(Item=(Type=Bone,Name="spine_02"),Weight=0.3292))
    Weight=1
   )
   > &RigUnit_ParentConstraint(
    AdvancedSettings=(RotationOrderForFilter=XZY)
    Child=(Type=Bone,Name="base_hemJacket_3_l")
    Filter=(TranslationFilter=(bX=True,bY=True,bZ=True),RotationFilter=(bX=True,bY=True,bZ=True),ScaleFilter=(bX=False,bY=False,bZ=False))
    Parents=((Item=(Type=Bone,Name="spine_04"),Weight=0.0329),(Item=(Type=Bone,Name="spine_01"),Weight=0.2014),(Item=(Type=Bone,Name="pelvis"),Weight=0.2104),(Item=(Type=Bone,Name="spine_03"),Weight=0.2261),(Item=(Type=Bone,Name="spine_02"),Weight=0.3292))
    Weight=1
   )
   > &RigUnit_ParentConstraint(
    AdvancedSettings=(RotationOrderForFilter=XZY)
    Child=(Type=Bone,Name="base_hemJacket_1_c")
    Filter=(TranslationFilter=(bX=True,bY=True,bZ=True),RotationFilter=(bX=True,bY=True,bZ=True),ScaleFilter=(bX=False,bY=False,bZ=False))
    Parents=((Item=(Type=Bone,Name="spine_04"),Weight=0.0569),(Item=(Type=Bone,Name="spine_01"),Weight=0.1787),(Item=(Type=Bone,Name="pelvis"),Weight=0.1869),(Item=(Type=Bone,Name="spine_03"),Weight=0.2848),(Item=(Type=Bone,Name="spine_02"),Weight=0.2927))
    Weight=1
   )
   > &RigUnit_ParentConstraint(
    AdvancedSettings=(RotationOrderForFilter=XZY)
    Child=(Type=Null,Name="target_hemJacket_1_r_null")
    Filter=(TranslationFilter=(bX=True,bY=True,bZ=True),RotationFilter=(bX=False,bY=False,bZ=False),ScaleFilter=(bX=False,bY=False,bZ=False))
    Parents=((Item=(Type=Bone,Name="spine_01"),Weight=0.0124),(Item=(Type=Bone,Name="spine_02"),Weight=0.0162),(Item=(Type=Bone,Name="pelvis"),Weight=0.9714))
    Weight=1
   )
   > &RigUnit_ParentConstraint(
    AdvancedSettings=(RotationOrderForFilter=XZY)
    Child=(Type=Null,Name="target_hemJacket_1_l_null")
    Filter=(TranslationFilter=(bX=True,bY=True,bZ=True),RotationFilter=(bX=False,bY=False,bZ=False),ScaleFilter=(bX=False,bY=False,bZ=False))
    Parents=((Item=(Type=Bone,Name="spine_01"),Weight=0.0106),(Item=(Type=Bone,Name="spine_02"),Weight=0.0145),(Item=(Type=Bone,Name="pelvis"),Weight=0.9749))
    Weight=1
   )
   > &RigUnit_ParentConstraint(
    AdvancedSettings=(RotationOrderForFilter=XZY)
    Child=(Type=Null,Name="target_hemJacket_2_r_null")
    Filter=(TranslationFilter=(bX=True,bY=True,bZ=True),RotationFilter=(bX=False,bY=False,bZ=False),ScaleFilter=(bX=False,bY=False,bZ=False))
    Parents=((Item=(Type=Bone,Name="pelvis"),Weight=1.0))
    Weight=1
   )
   > &RigUnit_ParentConstraint(
    AdvancedSettings=(RotationOrderForFilter=XZY)
    Child=(Type=Null,Name="target_hemJacket_2_l_null")
    Filter=(TranslationFilter=(bX=True,bY=True,bZ=True),RotationFilter=(bX=False,bY=False,bZ=False),ScaleFilter=(bX=False,bY=False,bZ=False))
    Parents=((Item=(Type=Bone,Name="pelvis"),Weight=1.0))
    Weight=1
   )
   > &RigUnit_ParentConstraint(
    AdvancedSettings=(RotationOrderForFilter=XZY)
    Child=(Type=Null,Name="target_hemJacket_3_r_null")
    Filter=(TranslationFilter=(bX=True,bY=True,bZ=True),RotationFilter=(bX=False,bY=False,bZ=False),ScaleFilter=(bX=False,bY=False,bZ=False))
    Parents=((Item=(Type=Bone,Name="pelvis"),Weight=1.0))
    Weight=1
   )
   > &RigUnit_ParentConstraint(
    AdvancedSettings=(RotationOrderForFilter=XZY)
    Child=(Type=Null,Name="target_hemJacket_3_l_null")
    Filter=(TranslationFilter=(bX=True,bY=True,bZ=True),RotationFilter=(bX=False,bY=False,bZ=False),ScaleFilter=(bX=False,bY=False,bZ=False))
    Parents=((Item=(Type=Bone,Name="pelvis"),Weight=1.0))
    Weight=1
   )
   > &RigUnit_ParentConstraint(
    AdvancedSettings=(RotationOrderForFilter=XZY)
    Child=(Type=Null,Name="target_jacketHem_back_null")
    Filter=(TranslationFilter=(bX=True,bY=True,bZ=True),RotationFilter=(bX=False,bY=False,bZ=False),ScaleFilter=(bX=False,bY=False,bZ=False))
    Parents=((Item=(Type=Bone,Name="pelvis"),Weight=1.0))
    Weight=1
   )
   internalName0 = &RigUnit_AimConstraintLocalSpaceOffset(
    Child=(Type=Bone,Name="base_hemJacket_1_r")
    Filter=(bX=True,bY=True,bZ=True)
    Parents=((Item=(Type=Null,Name="target_hemJacket_1_r_null"),Weight=1))
    Weight=1
    WorldUp=(Target=(X=0.0582,Y=-0.4783,Z=0.8762),Kind=Direction,Space=(Type=Bone,Name="kin_hemJacket"))
    bMaintainOffset=False
    ExecutePin=@RigVMFunction_Sequence.B
   )
   > internalName0_1 = &RigUnit_AimConstraintLocalSpaceOffset(
    Child=(Type=Bone,Name="base_hemJacket_1_l")
    Filter=(bX=True,bY=True,bZ=True)
    Parents=((Item=(Type=Null,Name="target_hemJacket_1_l_null"),Weight=1))
    Weight=1
    WorldUp=(Target=(X=-0.0582,Y=0.4783,Z=0.8762),Kind=Direction,Space=(Type=Bone,Name="kin_hemJacket"))
    bMaintainOffset=False
   )
   > internalName0_2 = &RigUnit_AimConstraintLocalSpaceOffset(
    Child=(Type=Bone,Name="base_hemJacket_2_r")
    Filter=(bX=True,bY=True,bZ=True)
    Parents=((Item=(Type=Null,Name="target_hemJacket_2_r_null"),Weight=1))
    Weight=1
    WorldUp=(Target=(X=0.1044,Y=-0.9353,Z=0.3381),Kind=Direction,Space=(Type=Bone,Name="kin_hemJacket"))
    bMaintainOffset=False
   )
   > internalName0_3 = &RigUnit_AimConstraintLocalSpaceOffset(
    Child=(Type=Bone,Name="base_hemJacket_2_l")
    Filter=(bX=True,bY=True,bZ=True)
    Parents=((Item=(Type=Null,Name="target_hemJacket_2_l_null"),Weight=1))
    Weight=1
    WorldUp=(Target=(X=-0.1044,Y=0.9353,Z=0.3381),Kind=Direction,Space=(Type=Bone,Name="kin_hemJacket"))
    bMaintainOffset=False
   )
   > internalName0_4 = &RigUnit_AimConstraintLocalSpaceOffset(
    Child=(Type=Bone,Name="base_hemJacket_3_r")
    Filter=(bX=True,bY=True,bZ=True)
    Parents=((Item=(Type=Null,Name="target_hemJacket_3_r_null"),Weight=1))
    Weight=1
    WorldUp=(Target=(X=0.0317,Y=-0.8918,Z=-0.4513),Kind=Direction,Space=(Type=Bone,Name="kin_hemJacket"))
    bMaintainOffset=False
   )
   > internalName0_5 = &RigUnit_AimConstraintLocalSpaceOffset(
    Child=(Type=Bone,Name="base_hemJacket_3_l")
    Filter=(bX=True,bY=True,bZ=True)
    Parents=((Item=(Type=Null,Name="target_hemJacket_3_l_null"),Weight=1))
    Weight=1
    WorldUp=(Target=(X=-0.0317,Y=0.8918,Z=-0.4513),Kind=Direction,Space=(Type=Bone,Name="kin_hemJacket"))
    bMaintainOffset=False
   )
   > internalName0_6 = &RigUnit_AimConstraintLocalSpaceOffset(
    Child=(Type=Bone,Name="base_hemJacket_1_c")
    Filter=(bX=True,bY=True,bZ=True)
    Parents=((Item=(Type=Null,Name="target_jacketHem_back_null"),Weight=1))
    Weight=1
    WorldUp=(Target=(X=0,Y=0,Z=-1.0),Kind=Direction,Space=(Type=Bone,Name="kin_hemJacket"))
    bMaintainOffset=False
   )
   RigVMFunctionReturnNode(ExecutePin=@RigVMFunction_Sequence.C)
  ]}
 CollapseNode_ContainedGraph {nodes [
   n1 = RigVMFunctionEntryNode()
   RigVMFunction_Sequence = &RigVMFunction_Sequence(ExecuteContext=@n1.ExecutePin)
   DeltaQuatExtraction = RigVMFunctionReferenceNode(
    child=(Type=Bone,Name="thigh_l")
    invert=true
    onlyTwist=true
    parent=(Type=Null,Name="thigh_l_null")
    twistAxis=(X=1,Y=0,Z=0)
    ExecuteContext=@RigVMFunction_Sequence.A
   )
   Interpolate = /Script/RigVM.RigVMFunction_MathQuaternionSlerp(B=(X=0,Y=0,Z=0,W=1), T=0.2, A=@DeltaQuatExtraction.outputQuat) {is=struct RigVMUnitNode, template="Interpolate::Execute(in A,in B,in T,out Result)"}
   > &RigUnit_ModifyTransforms(
    ItemToModify=((Item=(Type=Bone,Name="thigh_twist_00_l"),Transform=(Rotation=(X=0,Y=0,Z=0,W=1),Translation=(X=0,Y=0,Z=0),Scale3D=(X=1,Y=1,Z=1))))
    Weight=1
    WeightMinimum=0
    ItemToModify.0.Transform.Rotation=@Interpolate.Result
   )
   > DeltaQuatExtraction_1 = RigVMFunctionReferenceNode(child=(Type=Bone,Name="thigh_r"), invert=true, onlyTwist=true, parent=(Type=Null,Name="thigh_r_null"), twistAxis=(X=1,Y=0,Z=0))
   > &RigUnit_ModifyTransforms(
    ItemToModify=((Item=(Type=Bone,Name="thigh_twist_00_r"),Transform=(Rotation=(X=0,Y=0,Z=0,W=1),Translation=(X=0,Y=0,Z=0),Scale3D=(X=1,Y=1,Z=1))))
    Weight=1
    WeightMinimum=0
    ItemToModify.0.Transform.Rotation=&RigVMFunction_MathQuaternionSlerp(B=(X=0,Y=0,Z=0,W=1), T=0.2, A=@DeltaQuatExtraction_1.outputQuat) {template="Interpolate::Execute(in A,in B,in T,out Result)"}.Result
   )
   > GetAlphaFromQuatDelta = RigVMFunctionReferenceNode(
    Rotation Order=ZYX
    absoluteValue=false
    child=(Type=Bone,Name="calf_l")
    debugRotation=false
    outputMinMax=(X=0,Y=1)
    parent=(Type=Null,Name="calf_l_null")
    premult=1
    rotationMinMax=(X=20,Y=120)
    triggerAxis=Z
    twistAxis=(X=1,Y=0,Z=0)
   )
   > &RigUnit_ModifyTransforms(
    ItemToModify="((Item=(Type=Bone,Name=\"thigh_twist_00_l\"),Transform=(Rotation=(X=0,Y=0,Z=0,W=1),Translation=(X=8,Y=0,Z=0),Scale3D=(X=1,Y=1,Z=1))),(Item=(Type=Bone,Name=\"base_strapHang_in_l\"),Transform=(Rotation=(X=0,Y=0,Z=0,W=1),Translation=(X=0,Y=-1,Z=1),Scale3D=(X=1,Y= ...(truncated)"
    WeightMinimum=0
    Weight=&RigVMFunction_MathDoublePow(B=2, A=@GetAlphaFromQuatDelta.alpha) {template="Power::Execute(in A,in B,out Result)"}.Result
   )
   > GetAlphaFromQuatDelta_1 = RigVMFunctionReferenceNode(
    Rotation Order=ZYX
    absoluteValue=false
    child=(Type=Bone,Name="calf_r")
    debugRotation=false
    outputMinMax=(X=0,Y=1)
    parent=(Type=Null,Name="calf_r_null")
    premult=1
    rotationMinMax=(X=20,Y=120)
    triggerAxis=Z
    twistAxis=(X=1,Y=0,Z=0)
   )
   Power = &RigVMFunction_MathDoublePow(B=2, A=@GetAlphaFromQuatDelta_1.alpha) {template="Power::Execute(in A,in B,out Result)"}
   > &RigUnit_ModifyTransforms(
    ItemToModify="((Item=(Type=Bone,Name=\"thigh_twist_00_r\"),Transform=(Rotation=(X=0,Y=0,Z=0,W=1),Translation=(X=-8,Y=0,Z=0),Scale3D=(X=1,Y=1,Z=1))),(Item=(Type=Bone,Name=\"base_strapHang_in_r\"),Transform=(Rotation=(X=0,Y=0,Z=0,W=1),Translation=(X=0,Y=1,Z=1),Scale3D=(X=1,Y= ...(truncated)"
    WeightMinimum=0
    Weight=@Power.Result
   )
   > &RigUnit_ModifyTransforms(
    ItemToModify="((Item=(Type=Bone,Name=\"kin_strapStart_l\"),Transform=(Rotation=(X=0,Y=0,Z=-0.7934,W=0.6088),Translation=(X=1.2,Y=0,Z=-0.4),Scale3D=(X=1,Y=1,Z=1))),(Item=(Type=Bone,Name=\"kin_strap_frontTris_l\"),Transform=(Rotation=(X=0.1695,Y=-0.0376,Z=-0.2132,W=0.9615),Tr ...(truncated)"
    WeightMinimum=0
    Weight=&RigVMFunction_MathDoublePow(B=2.3, A=RigVMVariableNode() {variable=thigh_l_rz_0_100}.Value) {template="Power::Execute(in A,in B,out Result)"}.Result
   )
   > &RigUnit_ModifyTransforms(
    ItemToModify="((Item=(Type=Bone,Name=\"kin_strapStart_r\"),Transform=(Rotation=(X=0,Y=0,Z=-0.6088,W=0.7934),Translation=(X=0.4,Y=0.25,Z=0),Scale3D=(X=1,Y=1,Z=1))),(Item=(Type=Bone,Name=\"kin_strap_frontTris_r\"),Transform=(Rotation=(X=0.1695,Y=-0.0376,Z=-0.2132,W=0.9615),Tr ...(truncated)"
    WeightMinimum=0
    Weight=&RigVMFunction_MathDoublePow(B=2, A=RigVMVariableNode() {variable=thigh_r_rz_0_100}.Value) {template="Power::Execute(in A,in B,out Result)"}.Result
   )
   RigVMFunctionReturnNode(ExecutePin=@RigVMFunction_Sequence.B)
  ]}
 CollapseNode_ContainedGraph {nodes [
   n1 = RigVMFunctionEntryNode()
   RigVMFunction_Sequence = &RigVMFunction_Sequence(ExecuteContext=@n1.ExecutePin)
   &RigUnit_ParentConstraint(
    AdvancedSettings=(RotationOrderForFilter=XZY)
    Child=(Type=Bone,Name="base_hipChains")
    Parents=((Item=(Type=Bone,Name="deform_glute_r"),Weight=0.0189),(Item=(Type=Bone,Name="deform_groin_r"),Weight=0.2769),(Item=(Type=Bone,Name="thigh_twist_01_r"),Weight=0.3411),(Item=(Type=Bone,Name="pelvis"),Weight=0.3631))
    Weight=1
    ExecutePin=@RigVMFunction_Sequence.A
   )
   > &RigUnit_ParentConstraint(
    AdvancedSettings=(RotationOrderForFilter=XZY)
    Child=(Type=Bone,Name="hipChains_twist_bone")
    Parents=((Item=(Type=Bone,Name="deform_glute_r"),Weight=0.0189),(Item=(Type=Bone,Name="deform_groin_r"),Weight=0.2769),(Item=(Type=Bone,Name="thigh_twist_01_r"),Weight=0.3411),(Item=(Type=Bone,Name="pelvis"),Weight=0.3631))
    Weight=1
   )
   > &RigUnit_ParentConstraint(
    AdvancedSettings=(RotationOrderForFilter=XZY)
    Child=(Type=Bone,Name="hipChains_target_bone")
    Parents=((Item=(Type=Bone,Name="thigh_r"),Weight=0.1),(Item=(Type=Bone,Name="thigh_twist_01_r"),Weight=0.9))
    Weight=1
   )
   > n2 = &RigUnit_SphericalPoseReader(
    ActiveRegionScaleFactors=(PositiveWidth=1,NegativeWidth=1,PositiveHeight=2.5,NegativeHeight=2.5)
    ActiveRegionSize=0.05
    Debug=(bDrawDebug=false,bDraw2D=False,bDrawLocalAxes=False,DebugScale=25,DebugSegments=20,DebugThickness=0.25)
    DriverAxis=(X=1,Y=0,Z=0)
    DriverItem=(Type=Bone,Name="thigh_r")
    FalloffRegionScaleFactors=(PositiveWidth=1,NegativeWidth=1,PositiveHeight=4,NegativeHeight=4)
    FalloffSize=0.15
    FlipHeightScaling=False
    FlipWidthScaling=False
    OptionalParentItem=(Type=Bone,Name="None")
    RotationOffset=(X=-72.0344,Y=-1.1062,Z=-4.7134)
   )
   Remap = &RigVMFunction_MathDoubleRemap(SourceMaximum=1, SourceMinimum=0, TargetMaximum=20, TargetMinimum=0, Value=@n2.OutputParam)
   RigVMFunction_MathQuaternionFromEuler = &RigVMFunction_MathQuaternionFromEuler(Euler=(X=0,Y=0,Z=0), Euler.Y=@Remap.Result)
   > &RigUnit_AimConstraintLocalSpaceOffset(
    AdvancedSettings=(DebugSettings=(bEnabled=false,Scale=10,WorldOffset=(Rotation=(X=0,Y=0,Z=0,W=1),Translation=(X=0,Y=0,Z=0),Scale3D=(X=1,Y=1,Z=1))),RotationOrderForFilter=XZY)
    Child=(Type=Bone,Name="base_hipChains")
    Filter=(bX=True,bY=True,bZ=True)
    Parents=((Item=(Type=Bone,Name="hipChains_target_bone"),Weight=1))
    UpAxis=(X=0,Y=0,Z=-1)
    Weight=1
    WorldUp=(Target=(X=0,Y=0,Z=-1),Kind=Direction,Space=(Type=Bone,Name="hipChains_twist_bone"))
    bMaintainOffset=false
    WorldUp.Target=/Script/RigVM.RigVMFunction_MathQuaternionRotateVector(Vector=(X=0,Y=0,Z=-1), Transform=@RigVMFunction_MathQuaternionFromEuler.Result) {is=struct RigVMUnitNode, template="Rotate Vector::Execute(in Transform,in Vector,out Result)"}.Result
   )
   > &RigUnit_ModifyTransforms(
    ItemToModify=((Item=(Type=Bone,Name="dyn_hipChains_front"),Transform=(Rotation=(X=0.1635,Y=0.028,Z=0.1193,W=-0.9789),Translation=(X=5.0973,Y=-6.0657,Z=-2.5244),Scale3D=(X=1,Y=1,Z=1))))
    WeightMinimum=0
    Weight=RigVMVariableNode() {variable=thigh_r_rz_0_100}.Value
   )
   RigVMFunctionReturnNode(ExecutePin=@RigVMFunction_Sequence.B)
  ]}
 CollapseNode_ContainedGraph {nodes [
   n1 = RigVMFunctionEntryNode()
   RigVMFunction_Sequence = RigVMAggregateNode(ExecuteContext=@n1.ExecutePin)
   &RigUnit_ParentConstraint(
    AdvancedSettings=(RotationOrderForFilter=XZY)
    Child=(Type=Bone,Name="kin_carabiners_01_r")
    Parents="((Item=(Type=Bone,Name=\"deform_glute_r\"),Weight=0.0189),(Item=(Type=Bone,Name=\"kin_hoodie_hem_01_r\"),Weight=0.0699),(Item=(Type=Bone,Name=\"deform_groin_r\"),Weight=0.2769),(Item=(Type=Bone,Name=\"pelvis\"),Weight=0.2932),(Item=(Type=Bone,Name=\"thigh_twist_01_ ...(truncated)"
    Weight=1
    ExecutePin=@RigVMFunction_Sequence.A
   )
   > GetAlphaFromQuatDelta_1 = RigVMFunctionReferenceNode(
    Rotation Order=ZYX
    child=(Type=Bone,Name="thigh_r")
    outputMinMax=(X=0,Y=1)
    parent=(Type=Null,Name="thigh_r_null")
    premult=1
    rotationMinMax=(X=-30,Y=-100)
    triggerAxis=Z
    twistAxis=(X=1,Y=0,Z=0)
   )
   Power = &RigVMFunction_MathDoublePow(B=2, A=@GetAlphaFromQuatDelta_1.alpha) {template="Power::Execute(in A,in B,out Result)"}
   > &RigUnit_ModifyTransforms(
    ItemToModify="((Item=(Type=Bone,Name=\"kin_carabiners_01_r\"),Transform=(Rotation=(X=-0.2473,Y=0.0628,Z=-0.1491,W=-0.9553),Translation=(X=0.4809,Y=0.3869,Z=-0.9071),Scale3D=(X=0.8267,Y=0.8267,Z=0.8267))),(Item=(Type=Bone,Name=\"kin_hoodie_hem_01_r\"),Transform=(Rotation=(X= ...(truncated)"
    WeightMinimum=0
    Weight=@Power.Result
   )
   > n2 = &RigUnit_SphericalPoseReader(
    ActiveRegionScaleFactors=(PositiveWidth=1,NegativeWidth=0.8,PositiveHeight=0.8,NegativeHeight=1)
    ActiveRegionSize=0.1
    Debug=(bDrawDebug=false,bDraw2D=False,bDrawLocalAxes=False,DebugScale=25,DebugSegments=20,DebugThickness=0.25)
    DriverAxis=(X=1,Y=0,Z=0)
    DriverItem=(Type=Bone,Name="thigh_r")
    FalloffRegionScaleFactors=(PositiveWidth=1,NegativeWidth=1,PositiveHeight=1,NegativeHeight=1)
    FalloffSize=0.2
    FlipHeightScaling=False
    FlipWidthScaling=False
    OptionalParentItem=(Type=Bone,Name="None")
    RotationOffset=(X=-50,Y=0,Z=-15)
   )
   > &RigUnit_ModifyTransforms(
    ItemToModify=((Item=(Type=Bone,Name="kin_carabiners_01_r"),Transform=(Rotation=(X=-0.1706,Y=-0.0225,Z=0.1286,W=-0.9767),Translation=(X=-0.3047,Y=0.7337,Z=0.2644),Scale3D=(X=1,Y=1,Z=1))))
    WeightMinimum=0
    Weight=@n2.OutputParam
   )
   &RigUnit_ParentConstraint(
    AdvancedSettings=(RotationOrderForFilter=XZY)
    Child=(Type=Bone,Name="kin_carabiners_02_l")
    Parents=((Item=(Type=Bone,Name="deform_glute_l"),Weight=0.0189),(Item=(Type=Bone,Name="deform_groin_l"),Weight=0.3061),(Item=(Type=Bone,Name="thigh_twist_01_l"),Weight=0.3305),(Item=(Type=Bone,Name="pelvis"),Weight=0.3445))
    Weight=1
    ExecutePin=@RigVMFunction_Sequence.B
   )
   > &RigUnit_ParentConstraint(
    AdvancedSettings=(RotationOrderForFilter=XZY)
    Child=(Type=Bone,Name="kin_carabiners_01_l")
    Parents=((Item=(Type=Bone,Name="thigh_twist_01_r"),Weight=0.0153),(Item=(Type=Bone,Name="thigh_twist_01_l"),Weight=0.2949),(Item=(Type=Bone,Name="pelvis"),Weight=0.3299),(Item=(Type=Bone,Name="deform_groin_l"),Weight=0.3599))
    Weight=1
   )
   > GetAlphaFromQuatDelta_2 = RigVMFunctionReferenceNode(
    Rotation Order=ZYX
    child=(Type=Bone,Name="thigh_l")
    outputMinMax=(X=0,Y=1)
    parent=(Type=Null,Name="thigh_l_null")
    premult=1
    rotationMinMax=(X=-10,Y=-100)
    triggerAxis=Z
    twistAxis=(X=1,Y=0,Z=0)
   )
   > &RigUnit_ModifyTransforms(
    ItemToModify="((Item=(Type=Bone,Name=\"kin_carabiners_02_l\"),Transform=(Rotation=(X=-0.0832,Y=0.112,Z=-0.0215,W=0.99),Translation=(X=-0.6,Y=-0.5,Z=0),Scale3D=(X=1,Y=1,Z=1))),(Item=(Type=Bone,Name=\"kin_carabiners_01_l\"),Transform=(Rotation=(X=-0.1992,Y=0.3061,Z=0.0466,W=0 ...(truncated)"
    WeightMinimum=0
    Weight=@GetAlphaFromQuatDelta_2.alpha
   )
   > n3 = &RigUnit_SphericalPoseReader(
    ActiveRegionScaleFactors=(PositiveWidth=1.2,NegativeWidth=1,PositiveHeight=1.2,NegativeHeight=1)
    ActiveRegionSize=0.1
    Debug=(bDrawDebug=false,bDraw2D=False,bDrawLocalAxes=False,DebugScale=25,DebugSegments=20,DebugThickness=0.25)
    DriverAxis=(X=-1,Y=0,Z=0)
    DriverItem=(Type=Bone,Name="thigh_l")
    FalloffRegionScaleFactors=(PositiveWidth=1,NegativeWidth=1,PositiveHeight=1,NegativeHeight=1)
    FalloffSize=0.25
    FlipHeightScaling=False
    FlipWidthScaling=False
    OptionalParentItem=(Type=Bone,Name="None")
    RotationOffset=(X=130,Y=0,Z=0)
   )
   > &RigUnit_ModifyTransforms(
    ItemToModify="((Item=(Type=Bone,Name=\"kin_carabiners_02_l\"),Transform=(Rotation=(X=-0.1063,Y=0.1463,Z=-0.0184,W=0.9833),Translation=(X=-0.3466,Y=0,Z=-1.1105),Scale3D=(X=1,Y=1,Z=1))),(Item=(Type=Bone,Name=\"kin_carabiners_01_l\"),Transform=(Rotation=(X=0.1853,Y=0,Z=0,W=-0. ...(truncated)"
    WeightMinimum=0
    Weight=@n3.OutputParam
   )
   Multiply = &RigVMFunction_MathDoubleMul(B=@GetAlphaFromQuatDelta_2.alpha, A=@n3.OutputParam)
   > &RigUnit_ModifyTransforms(
    ItemToModify="((Item=(Type=Bone,Name=\"base_hangCarabinerF_1\"),Transform=(Rotation=(X=0,Y=0,Z=0,W=1),Translation=(X=0,Y=0,Z=0),Scale3D=(X=1,Y=1,Z=1))),(Item=(Type=Bone,Name=\"base_hangCarabinerS_1\"),Transform=(Rotation=(X=0,Y=0.2588,Z=0,W=0.9659),Translation=(X=0,Y=0,Z=0) ...(truncated)"
    WeightMinimum=0
    Weight=@Multiply.Result
   )
   RigVMFunctionReturnNode(ExecutePin=@RigVMFunction_Sequence.C)
  ]}
 CollapseNode_ContainedGraph {nodes [
   RigVMFunctionEntryNode()
   > RigVMFunction_Sequence = RigVMAggregateNode()
   InverseDistance = RigVMFunctionReferenceNode(
    child=(Type=Bone,Name="thigh_l")
    drawLines=false
    ease=false
    parent=(Type=Null,Name="thigh_l_null")
    primaryAxis=(X=-10,Y=0,Z=0)
    remapCurve=(EditorCurveData=(Keys=((InterpMode=RCIM_Cubic,TangentMode=RCTM_User,ArriveTangent=2.1874,LeaveTangent=2.1874),(InterpMode=RCIM_Cubic,TangentMode=RCTM_User,Time=0.2889,Value=0.6431,ArriveTangent=1.2215,LeaveTangent=1.2215),(Time=1,Value=1))))
    remapMult=4
    targets="((X=0,Y=0,Z=0,W=1),(X=0,Y=-0.7071,Z=0,W=0.7071),(X=0,Y=0,Z=0.7071,W=0.7071),(X=0,Y=0,Z=-0.7071,W=0.7071),(X=-0.1983,Y=-0.1664,Z=-0.7399,W=0.6209),(X=-0.092,Y=-0.1266,Z=-0.5805,W=0.7991),(X=-0.2149,Y=-0.3069,Z=-0.5318,W=0.7595),(X=-0.1,Y=-0.2747,Z=-0.3271,W ...(truncated)"
    ExecuteContext=@RigVMFunction_Sequence.A
   )
   > &RigUnit_ModifyTransforms(
    ItemToModify="((Item=(Type=Bone,Name=\"base_hemJacket_1_l\"),Transform=(Rotation=(X=0,Y=0,Z=-0.1305,W=0.9914),Translation=(X=0,Y=0,Z=0),Scale3D=(X=1,Y=1,Z=1))),(Item=(Type=Bone,Name=\"base_hemJacket_2_l\"),Transform=(Rotation=(X=0,Y=0,Z=-0.1305,W=0.9914),Translation=(X=0,Y= ...(truncated)"
    WeightMinimum=0
    Weight=RigVMDispatchNode(Index=4, Array=@InverseDistance.weights) {template="DISPATCH_RigVMDispatch_ArrayGetAtIndex(in Array,in Index,out Element)"}.Element
   )
   > &RigUnit_ModifyTransforms(
    ItemToModify="((Item=(Type=Bone,Name=\"kin_carabiners_01_l\"),Transform=(Rotation=(X=-0.0763,Y=0,Z=0,W=-0.9971),Translation=(X=0,Y=0,Z=0),Scale3D=(X=1,Y=1,Z=1))),(Item=(Type=Bone,Name=\"base_hemJacket_1_l\"),Transform=(Rotation=(X=0,Y=0,Z=-0.0872,W=0.9962),Translation=(X=0, ...(truncated)"
    WeightMinimum=0
    Weight=RigVMDispatchNode(Index=5, Array=@InverseDistance.weights) {template="DISPATCH_RigVMDispatch_ArrayGetAtIndex(in Array,in Index,out Element)"}.Element
   )
   > &RigUnit_ModifyTransforms(
    ItemToModify="((Item=(Type=Bone,Name=\"base_hangCarabinerS_1\"),Transform=(Rotation=(X=-0.1513,Y=0.1912,Z=0.0024,W=-0.9698),Translation=(X=0,Y=0,Z=0),Scale3D=(X=1,Y=1,Z=1))),(Item=(Type=Bone,Name=\"base_hemJacket_1_l\"),Transform=(Rotation=(X=0,Y=0,Z=-0.0436,W=0.999),Transl ...(truncated)"
    WeightMinimum=0
    Weight=RigVMDispatchNode(Index=6, Array=@InverseDistance.weights) {template="DISPATCH_RigVMDispatch_ArrayGetAtIndex(in Array,in Index,out Element)"}.Element
   )
   > &RigUnit_ModifyTransforms(
    ItemToModify="((Item=(Type=Bone,Name=\"base_hangCarabinerS_1\"),Transform=(Rotation=(X=0,Y=-0.1556,Z=0,W=0.9878),Translation=(X=0,Y=0,Z=0),Scale3D=(X=1,Y=1,Z=1))),(Item=(Type=Bone,Name=\"base_hangCarabinerF_1\"),Transform=(Rotation=(X=-0.0115,Y=-0.0354,Z=-0.1243,W=0.9916),T ...(truncated)"
    WeightMinimum=0
    Weight=RigVMDispatchNode(Index=7, Array=@InverseDistance.weights) {template="DISPATCH_RigVMDispatch_ArrayGetAtIndex(in Array,in Index,out Element)"}.Element
   )
   n1 = RigVMDispatchNode(Index=8, Array=@InverseDistance.weights) {template="DISPATCH_RigVMDispatch_ArrayGetAtIndex(in Array,in Index,out Element)"}
   > &RigUnit_ModifyTransforms(
    ItemToModify="((Item=(Type=Bone,Name=\"base_hemJacket_1_l\"),Transform=(Rotation=(X=0.0883,Y=0.1139,Z=-0.3254,W=0.9345),Translation=(X=0.4139,Y=-5.3257,Z=-1.6653),Scale3D=(X=1,Y=1,Z=1))),(Item=(Type=Bone,Name=\"base_hemJacket_2_l\"),Transform=(Rotation=(X=-0.0343,Y=-0.3019, ...(truncated)"
    WeightMinimum=0
    Weight=@n1.Element
   )
   FortRigUnit_DeltaRef = RigVMUnitNode(
    Base=(Type=Bone,Name="spine_02")
    Bone=(Type=Bone,Name="pelvis")
    DeltaComponentType=EulerZ
    InputMax=-20
    InputMin=0
    OutputMax=1
    OutputMin=0
    TwistAxis=(X=1,Y=0,Z=0)
   )
   Multiply = &RigVMFunction_MathDoubleMul(B=@FortRigUnit_DeltaRef.AlphaOutput, A=@n1.Element)
   > &RigUnit_ModifyTransforms(
    ItemToModify="((Item=(Type=Bone,Name=\"dyn_hangCarabinerS_1\"),Transform=(Rotation=(X=0.2093,Y=0.1228,Z=-0.3713,W=0.8962),Translation=(X=0,Y=0,Z=0),Scale3D=(X=1,Y=1,Z=1))),(Item=(Type=Bone,Name=\"dyn_hangCarabinerF_1\"),Transform=(Rotation=(X=0.1657,Y=-0.1816,Z=0.2602,W=0.9 ...(truncated)"
    WeightMinimum=0
    Weight=@Multiply.Result
   )
   n2 = &RigUnit_SphericalPoseReader(
    ActiveRegionScaleFactors=(PositiveWidth=0.6,NegativeWidth=2,PositiveHeight=2,NegativeHeight=2)
    ActiveRegionSize=0.1
    Debug=(bDrawDebug=false,bDraw2D=False,bDrawLocalAxes=False,DebugScale=25,DebugSegments=20,DebugThickness=0.25)
    DriverAxis=(X=1,Y=0,Z=0)
    DriverItem=(Type=Bone,Name="thigh_r")
    FalloffRegionScaleFactors=(PositiveWidth=1,NegativeWidth=1,PositiveHeight=1,NegativeHeight=1)
    FalloffSize=0.25
    FlipHeightScaling=False
    FlipWidthScaling=False
    OptionalParentItem=(Type=Bone,Name="None")
    RotationOffset=(X=-90,Y=0,Z=-50)
    ExecutePin=@RigVMFunction_Sequence.B
   )
   Power = &RigVMFunction_MathDoublePow(B=2, A=@n2.OutputParam) {template="Power::Execute(in A,in B,out Result)"}
   > &RigUnit_ModifyTransforms(
    ItemToModify="((Item=(Type=Bone,Name=\"base_hemJacket_1_r\"),Transform=(Rotation=(X=-0.0657,Y=0.0062,Z=-0.2185,W=0.9736),Translation=(X=-1.7681,Y=-4.2877,Z=-0.6775),Scale3D=(X=1,Y=1,Z=1))),(Item=(Type=Bone,Name=\"base_hemJacket_2_r\"),Transform=(Rotation=(X=-0.2406,Y=0.1768 ...(truncated)"
    WeightMinimum=0
    Weight=@Power.Result
   )
   FortRigUnit_DeltaRef_1 = RigVMUnitNode(
    Base=(Type=Null,Name="thigh_r_null")
    Bone=(Type=Bone,Name="thigh_r")
    DeltaComponentType=EulerZ
    InputMax=-120
    InputMin=0
    OutputMax=1
    OutputMin=0
    TwistAxis=(X=1,Y=0,Z=0)
   )
   FortRigUnit_DeltaRef_2 = RigVMUnitNode(
    Base=(Type=Bone,Name="spine_02")
    Bone=(Type=Bone,Name="pelvis")
    DeltaComponentType=EulerZ
    InputMax=-36
    InputMin=-15
    OutputMax=1
    OutputMin=0
    TwistAxis=(X=1,Y=0,Z=0)
   )
   Multiply_1 = RigVMAggregateNode(C=0.5, A=@FortRigUnit_DeltaRef_1.AlphaOutput, B=@FortRigUnit_DeltaRef_2.AlphaOutput)
   > &RigUnit_ModifyTransforms(
    ItemToModify="((Item=(Type=Bone,Name=\"base_hemJacket_1_r\"),Transform=(Rotation=(X=-0.1631,Y=0.1176,Z=-0.0673,W=0.9773),Translation=(X=-5,Y=-15,Z=1.2),Scale3D=(X=1,Y=1,Z=1))),(Item=(Type=Bone,Name=\"base_hemJacket_2_r\"),Transform=(Rotation=(X=0,Y=0,Z=-0.0262,W=0.9997),Tra ...(truncated)"
    WeightMinimum=0
    Weight=@Multiply_1.Result
   )
   > n3 = &RigUnit_SphericalPoseReader(
    ActiveRegionScaleFactors=(PositiveWidth=1,NegativeWidth=1,PositiveHeight=1,NegativeHeight=1)
    ActiveRegionSize=0.1
    Debug=(bDrawDebug=false,bDraw2D=False,bDrawLocalAxes=False,DebugScale=25,DebugSegments=20,DebugThickness=0.25)
    DriverAxis=(X=1,Y=0,Z=0)
    DriverItem=(Type=Bone,Name="thigh_r")
    FalloffRegionScaleFactors=(PositiveWidth=1,NegativeWidth=1,PositiveHeight=1,NegativeHeight=1)
    FalloffSize=0.2
    FlipHeightScaling=False
    FlipWidthScaling=False
    OptionalParentItem=(Type=Bone,Name="None")
    RotationOffset=(X=-72.4874,Y=-10.3393,Z=-22.4069)
   )
   > &RigUnit_ModifyTransforms(
    ItemToModify="((Item=(Type=Bone,Name=\"base_hemJacket_1_r\"),Transform=(Rotation=(X=0,Y=0,Z=0,W=1),Translation=(X=0,Y=0,Z=0),Scale3D=(X=1,Y=1,Z=1))),(Item=(Type=Bone,Name=\"base_hemJacket_2_r\"),Transform=(Rotation=(X=-0.0788,Y=0.1203,Z=-0.2751,W=0.9506),Translation=(X=0,Y= ...(truncated)"
    WeightMinimum=0
    Weight=@n3.OutputParam
   )
   RigVMFunctionReturnNode(ExecuteContext=@RigVMFunction_Sequence.C)
  ]}
 CollapseNode_ContainedGraph {nodes [
   n1 = RigVMFunctionEntryNode()
   &RigUnit_ParentConstraint(
    AdvancedSettings=(InterpolationType=Average,RotationOrderForFilter=XZY)
    Child=(Type=Bone,Name="deform_ankle_in_base_l")
    Parents=((Item=(Type=Bone,Name="foot_l"),Weight=0.7),(Item=(Type=Bone,Name="calf_twist_01_l"),Weight=0.3))
    Weight=1
    ExecutePin=@n1.ExecutePin
   )
   > &RigUnit_ParentConstraint(
    AdvancedSettings=(InterpolationType=Average,RotationOrderForFilter=XZY)
    Child=(Type=Bone,Name="deform_ankle_out_base_l")
    Parents=((Item=(Type=Bone,Name="foot_l"),Weight=0.8),(Item=(Type=Bone,Name="calf_twist_01_l"),Weight=0.2))
    Weight=1
   )
   > &RigUnit_ParentConstraint(
    AdvancedSettings=(InterpolationType=Average,RotationOrderForFilter=XZY)
    Child=(Type=Bone,Name="deform_ankle_in_base_r")
    Parents=((Item=(Type=Bone,Name="foot_r"),Weight=0.7),(Item=(Type=Bone,Name="calf_twist_01_r"),Weight=0.3))
    Weight=1
   )
   > &RigUnit_ParentConstraint(
    AdvancedSettings=(InterpolationType=Average,RotationOrderForFilter=XZY)
    Child=(Type=Bone,Name="deform_ankle_out_base_r")
    Parents=((Item=(Type=Bone,Name="foot_r"),Weight=0.8),(Item=(Type=Bone,Name="calf_twist_01_r"),Weight=0.2))
    Weight=1
   )
   > GetAlphaFromQuatDelta = RigVMFunctionReferenceNode(
    Rotation Order=ZYX
    child=(Type=Bone,Name="foot_l")
    outputMinMax=(X=0,Y=1)
    parent=(Type=Null,Name="foot_l_null")
    premult=1
    rotationMinMax=(X=0,Y=-40)
    triggerAxis=Z
    twistAxis=(X=1,Y=0,Z=0)
   )
   > &RigUnit_ModifyTransforms(
    ItemToModify=((Item=(Type=Bone,Name="deform_ankle_in_l"),Transform=(Rotation=(X=0,Y=0,Z=0,W=1),Translation=(X=2,Y=1,Z=0),Scale3D=(X=1,Y=0.8,Z=1))))
    WeightMinimum=0
    Weight=@GetAlphaFromQuatDelta.alpha
   )
   > GetAlphaFromQuatDelta_1 = RigVMFunctionReferenceNode(
    Rotation Order=ZYX
    child=(Type=Bone,Name="foot_l")
    outputMinMax=(X=0,Y=1)
    parent=(Type=Null,Name="foot_l_null")
    premult=1
    rotationMinMax=(X=0,Y=80)
    triggerAxis=Z
    twistAxis=(X=1,Y=0,Z=0)
   )
   Power = &RigVMFunction_MathDoublePow(B=2, A=@GetAlphaFromQuatDelta_1.alpha) {template="Power::Execute(in A,in B,out Result)"}
   > &RigUnit_ModifyTransforms(
    ItemToModify="((Item=(Type=Bone,Name=\"deform_ankle_out_l\"),Transform=(Rotation=(X=0,Y=0,Z=-0.2079,W=0.9781),Translation=(X=-5.5,Y=-0.3,Z=0),Scale3D=(X=1,Y=0.8,Z=1))),(Item=(Type=Bone,Name=\"deform_ankle_in_l\"),Transform=(Rotation=(X=0,Y=0,Z=0,W=1),Translation=(X=-2,Y=0,Z ...(truncated)"
    WeightMinimum=0
    Weight=@Power.Result
   )
   > GetAlphaFromQuatDelta_2 = RigVMFunctionReferenceNode(
    Rotation Order=ZYX
    child=(Type=Bone,Name="foot_r")
    outputMinMax=(X=0,Y=1)
    parent=(Type=Null,Name="foot_r_null")
    premult=1
    rotationMinMax=(X=0,Y=-40)
    triggerAxis=Z
    twistAxis=(X=1,Y=0,Z=0)
   )
   > &RigUnit_ModifyTransforms(
    ItemToModify=((Item=(Type=Bone,Name="deform_ankle_in_r"),Transform=(Rotation=(X=0,Y=0,Z=0.0872,W=0.9962),Translation=(X=-2,Y=-1,Z=0),Scale3D=(X=1,Y=0.5,Z=1))))
    WeightMinimum=0
    Weight=@GetAlphaFromQuatDelta_2.alpha
   )
   > GetAlphaFromQuatDelta_3 = RigVMFunctionReferenceNode(
    Rotation Order=ZYX
    child=(Type=Bone,Name="foot_r")
    outputMinMax=(X=0,Y=1)
    parent=(Type=Null,Name="foot_r_null")
    premult=1
    rotationMinMax=(X=0,Y=80)
    triggerAxis=Z
    twistAxis=(X=1,Y=0,Z=0)
   )
   > n2 = &RigUnit_ModifyTransforms(
    ItemToModify="((Item=(Type=Bone,Name=\"deform_ankle_out_r\"),Transform=(Rotation=(X=0,Y=0,Z=-0.4226,W=0.9063),Translation=(X=10.3,Y=-0.5,Z=0),Scale3D=(X=1,Y=0.6,Z=1))),(Item=(Type=Bone,Name=\"deform_ankle_in_r\"),Transform=(Rotation=(X=0,Y=0,Z=-0.1305,W=0.9914),Translation= ...(truncated)"
    WeightMinimum=0
    Weight=&RigVMFunction_MathDoublePow(B=1.6, A=@GetAlphaFromQuatDelta_3.alpha) {template="Power::Execute(in A,in B,out Result)"}.Result
   )
   RigVMFunctionReturnNode(ExecutePin=@n2.ExecutePin)
  ]}
 CollapseNode_ContainedGraph {nodes [
   n1 = RigVMFunctionEntryNode()
   RigVMFunction_Sequence = RigVMAggregateNode(ExecuteContext=@n1.ExecutePin)
   &RigUnit_ParentConstraint(
    AdvancedSettings=(RotationOrderForFilter=XZY)
    Child=(Type=Bone,Name="base_calfLace_1_l")
    Parents=((Item=(Type=Bone,Name="calf_twist_02_r"),Weight=0.2993),(Item=(Type=Bone,Name="calf_twist_01_r"),Weight=0.7007))
    Weight=1
    ExecutePin=@RigVMFunction_Sequence.A
   )
   > &RigUnit_ParentConstraint(
    AdvancedSettings=(RotationOrderForFilter=XZY)
    Child=(Type=Bone,Name="deform_lat_base_l")
    Filter=(TranslationFilter=(bX=True,bY=True,bZ=False),RotationFilter=(bX=False,bY=False,bZ=False),ScaleFilter=(bX=False,bY=False,bZ=False))
    Parents="((Item=(Type=Bone,Name=\"spine_01\"),Weight=0.0119),(Item=(Type=Bone,Name=\"deform_pec_l\"),Weight=0.0146),(Item=(Type=Bone,Name=\"deform_upperarm_bk_l\"),Weight=0.018),(Item=(Type=Bone,Name=\"spine_04\"),Weight=0.0306),(Item=(Type=Bone,Name=\"spine_02\"),Weight=0.0 ...(truncated)"
    Weight=1
   )
   > &RigUnit_ParentConstraint(
    AdvancedSettings=(RotationOrderForFilter=XZY)
    Child=(Type=Bone,Name="deform_lat_base_r")
    Filter=(TranslationFilter=(bX=True,bY=True,bZ=False),RotationFilter=(bX=False,bY=False,bZ=False),ScaleFilter=(bX=False,bY=False,bZ=False))
    Parents="((Item=(Type=Bone,Name=\"spine_01\"),Weight=0.0136),(Item=(Type=Bone,Name=\"deform_upperarm_bk_r\"),Weight=0.0243),(Item=(Type=Bone,Name=\"spine_04\"),Weight=0.0326),(Item=(Type=Bone,Name=\"spine_02\"),Weight=0.0346),(Item=(Type=Bone,Name=\"spine_03\"),Weight=0.0539 ...(truncated)"
    Weight=1
   )
   > GetAlphaFromQuatDelta = RigVMFunctionReferenceNode(
    Rotation Order=ZYX
    child=(Type=Bone,Name="clavicle_l")
    outputMinMax=(X=0,Y=1)
    parent=(Type=Null,Name="clavicle_l_null")
    premult=1
    rotationMinMax=(X=-10,Y=-50)
    triggerAxis=Y
    twistAxis=(X=1,Y=0,Z=0)
   )
   > &RigUnit_ModifyTransforms(
    ItemToModify=((Item=(Type=Bone,Name="deform_lat_l"),Transform=(Rotation=(X=0,Y=0,Z=0,W=1),Translation=(X=-5,Y=0,Z=0),Scale3D=(X=1,Y=1,Z=1))))
    WeightMinimum=0
    Weight=@GetAlphaFromQuatDelta.alpha
   )
   > GetAlphaFromQuatDelta_1 = RigVMFunctionReferenceNode(
    Rotation Order=ZYX
    child=(Type=Bone,Name="clavicle_r")
    outputMinMax=(X=0,Y=1)
    parent=(Type=Null,Name="clavicle_r_null")
    premult=1
    rotationMinMax=(X=-10,Y=-50)
    triggerAxis=Y
    twistAxis=(X=1,Y=0,Z=0)
   )
   > &RigUnit_ModifyTransforms(
    ItemToModify=((Item=(Type=Bone,Name="deform_lat_r"),Transform=(Rotation=(X=0,Y=0,Z=0,W=1),Translation=(X=-5,Y=0,Z=0),Scale3D=(X=1,Y=1,Z=1))))
    WeightMinimum=0
    Weight=@GetAlphaFromQuatDelta_1.alpha
   )
   > n2 = &RigUnit_SphericalPoseReader(
    ActiveRegionScaleFactors=(PositiveWidth=3,NegativeWidth=2,PositiveHeight=3,NegativeHeight=3)
    ActiveRegionSize=0.05
    Debug=(bDrawDebug=false,bDraw2D=False,bDrawLocalAxes=False,DebugScale=25,DebugSegments=20,DebugThickness=0.25)
    DriverAxis=(X=-1,Y=0,Z=0)
    DriverItem=(Type=Bone,Name="thigh_l")
    FalloffRegionScaleFactors=(PositiveWidth=1,NegativeWidth=0.4,PositiveHeight=0.35,NegativeHeight=0.35)
    FalloffSize=0.4
    FlipHeightScaling=False
    FlipWidthScaling=False
    OptionalParentItem=(Type=Bone,Name="None")
    RotationOffset=(X=99.1705,Y=-8.6124,Z=44.353)
   )
   > &RigUnit_ModifyTransforms(
    ItemToModify=((Item=(Type=Bone,Name="thigh_tshirt_l"),Transform=(Rotation=(X=0,Y=0,Z=0.3233,W=0.9463),Translation=(X=8.7301,Y=4.1836,Z=-0.3665),Scale3D=(X=1,Y=1,Z=1))))
    WeightMinimum=0
    Weight=@n2.OutputParam
   )
   > &RigUnit_ModifyTransforms(
    ItemToModify=((Item=(Type=Bone,Name="thigh_tshirt_l"),Transform=(Rotation=(X=-0.0729,Y=0.0894,Z=0.4161,W=0.902),Translation=(X=-6.1729,Y=23.8898,Z=-7.7079),Scale3D=(X=1,Y=1,Z=1))))
    WeightMinimum=0
    Weight=RigVMVariableNode() {variable=thigh_l_rz_0_100}.Value
   )
   > &RigUnit_ModifyTransforms(
    ItemToModify=((Item=(Type=Bone,Name="thigh_tshirt_r"),Transform=(Rotation=(X=0,Y=0,Z=-0.2755,W=-0.9613),Translation=(X=-1.9461,Y=-13.3445,Z=0.5712),Scale3D=(X=1,Y=1,Z=1))))
    WeightMinimum=0
    Weight=RigVMVariableNode() {variable=thigh_r_rz_0_100}.Value
   )
   GetAlphaFromQuatDelta_2 = RigVMFunctionReferenceNode(
    Rotation Order=ZYX
    child=(Type=Bone,Name="hand_l")
    onlyTwist=True
    outputMinMax=(X=0,Y=1)
    parent=(Type=Null,Name="hand_l_null")
    premult=1
    rotationMinMax=(X=120,Y=150)
    twistAxis=(X=1,Y=0,Z=0)
    ExecuteContext=@RigVMFunction_Sequence.B
   )
   > &RigUnit_ModifyTransforms(
    ItemToModify="((Item=(Type=Bone,Name=\"lowerarm_twist_01_l\"),Transform=(Rotation=(X=0,Y=0,Z=0,W=1),Translation=(X=0,Y=0,Z=0),Scale3D=(X=1,Y=1.2,Z=1.2))),(Item=(Type=Bone,Name=\"lowerarm_twist_02_l\"),Transform=(Rotation=(X=0.1305,Y=0,Z=0,W=0.9914),Translation=(X=0,Y=0,Z=0) ...(truncated)"
    WeightMinimum=0
    Weight=@GetAlphaFromQuatDelta_2.alpha
   )
   > GetAlphaFromQuatDelta_5 = RigVMFunctionReferenceNode(
    Rotation Order=ZYX
    child=(Type=Bone,Name="hand_r")
    onlyTwist=True
    outputMinMax=(X=0,Y=1)
    parent=(Type=Null,Name="hand_r_null")
    premult=1
    rotationMinMax=(X=120,Y=150)
    twistAxis=(X=1,Y=0,Z=0)
   )
   > &RigUnit_ModifyTransforms(
    ItemToModify="((Item=(Type=Bone,Name=\"lowerarm_twist_01_r\"),Transform=(Rotation=(X=0,Y=0,Z=0,W=1),Translation=(X=0,Y=0,Z=0),Scale3D=(X=1,Y=1.2,Z=1.2))),(Item=(Type=Bone,Name=\"lowerarm_twist_02_r\"),Transform=(Rotation=(X=0.1305,Y=0,Z=0,W=0.9914),Translation=(X=0,Y=0,Z=0) ...(truncated)"
    WeightMinimum=0
    Weight=@GetAlphaFromQuatDelta_5.alpha
   )
   > GetAlphaFromQuatDelta_3 = RigVMFunctionReferenceNode(
    Rotation Order=ZYX
    absoluteValue=true
    child=(Type=Bone,Name="upperarm_l")
    onlyTwist=true
    outputMinMax=(X=-1,Y=1)
    parent=(Type=Null,Name="upperarm_l_null")
    premult=1
    rotationMinMax=(X=-100,Y=100)
    twistAxis=(X=1,Y=0,Z=0)
   )
   > &RigUnit_ModifyTransforms(
    ItemToModify=((Item=(Type=Bone,Name="upperarm_twist_02_l"),Transform=(Rotation=(X=-0.1478,Y=0,Z=0,W=0.989),Translation=(X=0,Y=0,Z=0),Scale3D=(X=1,Y=1.2,Z=1.2))))
    WeightMinimum=0
    Weight=@GetAlphaFromQuatDelta_3.alpha
   )
   > GetAlphaFromQuatDelta_4 = RigVMFunctionReferenceNode(
    Rotation Order=ZYX
    absoluteValue=true
    child=(Type=Bone,Name="upperarm_r")
    onlyTwist=true
    outputMinMax=(X=-1,Y=1)
    parent=(Type=Null,Name="upperarm_r_null")
    premult=1
    rotationMinMax=(X=-100,Y=100)
    twistAxis=(X=1,Y=0,Z=0)
   )
   > &RigUnit_ModifyTransforms(
    ItemToModify=((Item=(Type=Bone,Name="upperarm_twist_02_r"),Transform=(Rotation=(X=-0.1478,Y=0,Z=0,W=0.989),Translation=(X=0,Y=0,Z=0),Scale3D=(X=1,Y=1.2,Z=1.2))))
    WeightMinimum=0
    Weight=@GetAlphaFromQuatDelta_4.alpha
   )
   > GetAlphaFromQuatDelta_6 = RigVMFunctionReferenceNode(
    Rotation Order=ZYX
    absoluteValue=false
    child=(Type=Bone,Name="base_hemJacket_1_l")
    debugRotation=false
    onlyTwist=false
    outputMinMax=(X=0,Y=1)
    parent=(Type=Null,Name="base_hemJacket_1_l_null")
    premult=1
    rotationMinMax=(X=-20,Y=-45)
    triggerAxis=Z
    twistAxis=(X=1,Y=0,Z=0)
   )
   > &RigUnit_ModifyTransforms(
    ItemToModify=((Item=(Type=Bone,Name="kin_delta_hemJacket_1_l"),Transform=(Rotation=(X=0,Y=0,Z=0.342,W=0.9397),Translation=(X=-15,Y=-2,Z=0),Scale3D=(X=1,Y=1,Z=1))))
    WeightMinimum=0
    Weight=@GetAlphaFromQuatDelta_6.alpha
   )
   > GetAlphaFromQuatDelta_7 = RigVMFunctionReferenceNode(
    Rotation Order=ZYX
    absoluteValue=false
    child=(Type=Bone,Name="base_hemJacket_1_r")
    debugRotation=false
    onlyTwist=false
    outputMinMax=(X=0,Y=1)
    parent=(Type=Null,Name="base_hemJacket_1_r_null")
    premult=1
    rotationMinMax=(X=-20,Y=-45)
    triggerAxis=Z
    twistAxis=(X=1,Y=0,Z=0)
   )
   > &RigUnit_ModifyTransforms(
    ItemToModify=((Item=(Type=Bone,Name="kin_delta_hemJacket_1_r"),Transform=(Rotation=(X=0,Y=0,Z=0,W=1),Translation=(X=-15,Y=-2,Z=0),Scale3D=(X=1,Y=1,Z=1))))
    WeightMinimum=0
    Weight=@GetAlphaFromQuatDelta_7.alpha
   )
   RigVMFunctionReturnNode(ExecutePin=@RigVMFunction_Sequence.C)
  ]}
 Multiply_2_ContainedGraph {nodes [
   n1 = RigVMFunctionEntryNode()
   RigVMFunctionReturnNode(Result=&RigVMFunction_MathDoubleMul(B=@n1.C, A=&RigVMFunction_MathDoubleMul(A=@n1.A, B=@n1.B).Result).Result)
  ]}
 Multiply_3_ContainedGraph {nodes [
   n1 = RigVMFunctionEntryNode()
   RigVMFunctionReturnNode(Result=&RigVMFunction_MathTransformMul(B=@n1.C, A=&RigVMFunction_MathTransformMul(A=@n1.A, B=@n1.B) {template="Multiply::Execute(in A,in B,out Result)"}.Result) {template="Multiply::Execute(in A,in B,out Result)"}.Result)
  ]}
 New Function_ContainedGraph {nodes [
   RigVMFunctionEntryNode()
   RigVMFunction_MathBoolToInteger = &RigVMFunction_MathBoolToInteger(Value=RigVMVariableNode() {variable=onlyTwist}.Value)
   RigVMFunction_MathQuaternionSwingTwist = &RigVMFunction_MathQuaternionSwingTwist(
    Input=&RigUnit_GetRelativeTransformForItem(
     bChildInitial=False
     bParentInitial=False
     Child=RigVMVariableNode() {variable=child}.Value
     Parent=RigVMVariableNode() {variable=parent}.Value
    ).RelativeTransform.Rotation
    TwistAxis=RigVMVariableNode() {variable=twistAxis}.Value
   )
   n1 = RigVMDispatchNode(
    Values=((),())
    Index=@RigVMFunction_MathBoolToInteger.Result
    Values.0=@RigVMFunction_MathQuaternionSwingTwist.Swing
    Values.1=@RigVMFunction_MathQuaternionSwingTwist.Twist
   ) {template="DISPATCH_RigVMDispatch_SelectInt32(in Index,in Values,out Result)"}
   > RigVMVariableNode(Value=@n1.Result) {variable=entry_outputQuat}
   RigVMFunction_MathQuaternionToEuler = &RigVMFunction_MathQuaternionToEuler(Value=@n1.Result, RotationOrder=RigVMVariableNode() {variable=Rotation Order}.Value)
   > DISPATCH_RigVMDispatch_Print = RigVMDispatchNode(
    ScreenColor=(R=1,G=1,B=1,A=1)
    ScreenDuration=1.5
    Value=@RigVMFunction_MathQuaternionToEuler.Result
    Enabled=RigVMVariableNode() {variable=debugRotation}.Value
   ) {template="DISPATCH_RigVMDispatch_Print(in Prefix,in Value,in Enabled,in ScreenDuration,in ScreenColor)"}
   Cast = RigVMDispatchNode(Value=RigVMVariableNode() {variable=triggerAxis}.Value) {template="DISPATCH_RigVMDispatch_CastEnumToInt(in Value,out Result)"}
   DISPATCH_RigVMDispatch_SelectInt32 = RigVMDispatchNode(
    Values=(0,0,0)
    Index=@Cast.Result
    Values.0=@RigVMFunction_MathQuaternionToEuler.Result.X
    Values.1=@RigVMFunction_MathQuaternionToEuler.Result.Y
    Values.2=@RigVMFunction_MathQuaternionToEuler.Result.Z
   ) {template="DISPATCH_RigVMDispatch_SelectInt32(in Index,in Values,out Result)"}
   n2 = RigVMVariableNode() {variable=rotationMinMax}
   n3 = RigVMVariableNode() {variable=outputMinMax}
   Remap = &RigVMFunction_MathDoubleRemap(
    bClamp=False
    Value=@DISPATCH_RigVMDispatch_SelectInt32.Result
    SourceMinimum=@n2.Value.X
    SourceMaximum=@n2.Value.Y
    TargetMinimum=@n3.Value.X
    TargetMaximum=@n3.Value.Y
   )
   Absolute = &RigVMFunction_MathDoubleAbs(Value=@Remap.Result)
   n4 = RigVMVariableNode() {variable=absoluteValue}
   Multiply = &RigVMFunction_MathDoubleMul(
    A=&RigVMFunction_MathDoubleClamp(Value=RigVMDispatchNode(True=@Absolute.Result, False=@Remap.Result, Condition=@n4.Value) {template="DISPATCH_RigVMDispatch_If(in Condition,in True,in False,out Result)"}.Result).Result
    B=RigVMVariableNode() {variable=premult}.Value
   )
   > RigVMFunctionReturnNode(alpha=@Multiply.Result, rawQuat=RigVMVariableNode() {variable=entry_outputQuat}.Value)
  ]}
 New Function_ContainedGraph {nodes [
   RigVMFunctionEntryNode()
   RigVMFunction_MathBoolToInteger = &RigVMFunction_MathBoolToInteger(Value=RigVMVariableNode() {variable=onlyTwist}.Value)
   RigVMFunction_MathQuaternionSwingTwist = &RigVMFunction_MathQuaternionSwingTwist(
    Input=&RigUnit_GetRelativeTransformForItem(
     bChildInitial=False
     bParentInitial=False
     Child=RigVMVariableNode() {variable=child}.Value
     Parent=RigVMVariableNode() {variable=parent}.Value
    ).RelativeTransform.Rotation
    TwistAxis=RigVMVariableNode() {variable=twistAxis}.Value
   )
   n1 = RigVMDispatchNode(
    Values=((),())
    Index=@RigVMFunction_MathBoolToInteger.Result
    Values.0=@RigVMFunction_MathQuaternionSwingTwist.Swing
    Values.1=@RigVMFunction_MathQuaternionSwingTwist.Twist
   ) {template="DISPATCH_RigVMDispatch_SelectInt32(in Index,in Values,out Result)"}
   Inverse = /Script/RigVM.RigVMFunction_MathQuaternionInverse(Value=@n1.Result) {is=struct RigVMUnitNode, template="Inverse::Execute(in Value,out Result)"}
   > RigVMFunctionReturnNode(outputQuat=RigVMDispatchNode(True=@Inverse.Result, Condition=RigVMVariableNode() {variable=invert}.Value, False=@n1.Result) {template="DISPATCH_RigVMDispatch_If(in Condition,in True,in False,out Result)"}.Result)
  ]}
 New Function_ContainedGraph {nodes [
   RigVMFunctionEntryNode()
   Num = RigVMDispatchNode(Array=RigVMVariableNode() {variable=targets}.Value) {template="DISPATCH_RigVMDispatch_ArrayGetNum(in Array,out Num)"}
   > RigVMDispatchNode(Element=0, Num=@Num.Num, Array=RigVMVariableNode() {variable=outputValues}.Value) {template="DISPATCH_RigVMDispatch_ArrayInit(io Array,in Num,in Element)"}
   > RigVMDispatchNode(Element=0, Array=RigVMVariableNode() {variable=divisors}.Value, Num=@Num.Num) {template="DISPATCH_RigVMDispatch_ArrayInit(io Array,in Num,in Element)"}
   > RigVMDispatchNode(Element=0, Array=RigVMVariableNode() {variable=outLenRemaped}.Value, Num=@Num.Num) {template="DISPATCH_RigVMDispatch_ArrayInit(io Array,in Num,in Element)"}
   n1 = RigVMVariableNode() {variable=child}
   > RigVMVariableNode(Value=&RigVMFunction_MathTransformMul(
     A=(Rotation=(X=0,Y=0,Z=0,W=1),Translation=(X=0,Y=0,Z=0),Scale3D=(X=1,Y=1,Z=1))
     B=&RigUnit_GetTransform(bInitial=False, Item=@n1.Value).Transform
     A.Translation=RigVMVariableNode() {variable=primaryAxis}.Value
    ) {template="Multiply::Execute(in A,in B,out Result)"}.Result.Translation) {variable=target vector}
   n2 = RigVMVariableNode() {variable=parent}
   > RigVMVariableNode(Value=&RigVMFunction_MathTransformMul(
     A=&RigUnit_GetRelativeTransformForItem(bChildInitial=true, bParentInitial=true, Child=@n1.Value, Parent=@n2.Value).RelativeTransform
     B=&RigUnit_GetTransform(bInitial=False, Item=@n2.Value).Transform
    ) {template="Multiply::Execute(in A,in B,out Result)"}.Result) {variable=parentChildLocated transform}
   > /Script/RigVM.RigVMFunction_DebugLineNoSpace(
    Color=(R=0,G=0,B=1,A=1)
    DebugDrawSettings=(DepthPriority=SDPG_Foreground,Lifetime=-1)
    Thickness=0.5
    WorldOffset=(Rotation=(X=0,Y=0,Z=0,W=1),Translation=(X=0,Y=0,Z=0),Scale3D=(X=1,Y=1,Z=1))
    B=RigVMVariableNode() {variable=target vector}.Value
    A=RigVMVariableNode() {variable=parentChildLocated transform}.Value.Translation
    bEnabled=RigVMVariableNode() {variable=drawLines}.Value
   ) {is=struct RigVMUnitNode}
   > For_Each = RigVMDispatchNode(Array=RigVMVariableNode() {variable=targets}.Value) {template="DISPATCH_RigVMDispatch_ArrayIterator(in Array,out Element,out Index,out Count,out Ratio)"}
   > RigVMVariableNode(Value=@For_Each.Index) {variable=index1 iter}
   Multiply_1 = RigVMAggregateNode(
    A=(Rotation=(X=0,Y=0,Z=0,W=1),Translation=(X=0,Y=0,Z=0),Scale3D=(X=1,Y=1,Z=1))
    B=(Rotation=(X=0,Y=0,Z=0,W=1),Translation=(X=0,Y=0,Z=0),Scale3D=(X=1,Y=1,Z=1))
    C=RigVMVariableNode() {variable=parentChildLocated transform}.Value
    A.Translation=RigVMVariableNode() {variable=primaryAxis}.Value
    B.Rotation=@For_Each.Element
   )
   > RigVMFunction_DebugTransformMutableNoSpace = /Script/RigVM.RigVMFunction_DebugTransformMutableNoSpace(
    Color=(R=1,G=0,B=1,A=0)
    DebugDrawSettings=(DepthPriority=SDPG_Foreground,Lifetime=-1)
    Mode=Box
    Scale=1
    Thickness=0.5
    Transform=(Rotation=(X=0,Y=0,Z=0,W=1),Translation=(X=0,Y=0,Z=0),Scale3D=(X=1,Y=1,Z=1))
    WorldOffset=(Rotation=(X=0,Y=0,Z=0,W=1),Translation=(X=0,Y=0,Z=0),Scale3D=(X=1,Y=1,Z=1))
    bEnabled=RigVMVariableNode() {variable=drawLines}.Value
    Transform.Translation=@Multiply_1.Result.Translation
   ) {is=struct RigVMUnitNode}
   n3 = RigVMVariableNode() {variable=parentChildLocated transform}
   Unit = /Script/RigVM.RigVMFunction_MathVectorUnit(Value=/Script/RigVM.RigVMFunction_MathVectorSub(A=RigVMVariableNode() {variable=target vector}.Value, B=@n3.Value.Translation) {is=struct RigVMUnitNode, template="Subtract::Execute(in A,in B,out Result)"}.Result) {is=struct RigVMUnitNode, template="Unit::Execute(in Value,out Result)"}
   Subtract = &RigVMFunction_MathVectorSub(
    A=@Unit.Result
    B=&RigVMFunction_MathVectorUnit(Value=&RigVMFunction_MathVectorSub(A=@Multiply_1.Result.Translation, B=@n3.Value.Translation) {template="Subtract::Execute(in A,in B,out Result)"}.Result) {template="Unit::Execute(in Value,out Result)"}.Result
   ) {template="Subtract::Execute(in A,in B,out Result)"}
   RigVMFunction_MathVectorLength = /Script/RigVM.RigVMFunction_MathVectorLength(Value=@Subtract.Result) {is=struct RigVMUnitNode, template="Length::Execute(in Value,out Result)"}
   Remap = &RigVMFunction_MathDoubleRemap(
    SourceMaximum=0
    TargetMaximum=1
    TargetMinimum=0
    bClamp=False
    SourceMinimum=RigVMVariableNode() {variable=remapMult}.Value
    Value=@RigVMFunction_MathVectorLength.Result
   )
   > Set_At_2 = RigVMDispatchNode(Element=@Remap.Result, Array=RigVMVariableNode() {variable=outLenRemaped}.Value, Index=RigVMVariableNode() {variable=index1 iter}.Value) {template="DISPATCH_RigVMDispatch_ArraySetAtIndex(io Array,in Index,in Element)"}
   Power = &RigVMFunction_MathDoublePow(B=2, A=@RigVMFunction_MathVectorLength.Result) {template="Power::Execute(in A,in B,out Result)"}
   Clamp = &RigVMFunction_MathDoubleClamp(Maximum=10000000000, Value=@Power.Result)
   Divide = &RigVMFunction_MathDoubleDiv(A=1, B=@Clamp.Result)
   > Set_At = RigVMDispatchNode(Element=@Divide.Result, Index=RigVMVariableNode() {variable=index1 iter}.Value, Array=RigVMVariableNode() {variable=divisors}.Value) {template="DISPATCH_RigVMDispatch_ArraySetAtIndex(io Array,in Index,in Element)"}
   n4 = RigVMVariableNode() {variable=divisors}
   For_Each_1 = RigVMDispatchNode(ExecuteContext=@For_Each.Completed, Array=@n4.Value) {template="DISPATCH_RigVMDispatch_ArrayIterator(in Array,out Element,out Index,out Count,out Ratio)"}
   ArraySum = /Script/RigVM.RigVMFunction_MathDoubleArraySum(Array=@n4.Value) {is=struct RigVMUnitNode, template="ArraySum::Execute(in Array,out Sum)"}
   n5 = &RigVMFunction_MathDoubleDiv(B=@ArraySum.Sum, A=@For_Each_1.Element)
   Multiply = &RigVMFunction_MathDoubleMul(
    B=RigVMDispatchNode(Index=@For_Each_1.Index, Array=RigVMVariableNode() {variable=outLenRemaped}.Value) {template="DISPATCH_RigVMDispatch_ArrayGetAtIndex(in Array,in Index,out Element)"}.Element
    A=@n5.Result
   )
   RigVMFunction_AnimEvalRichCurve = &RigVMFunction_AnimEvalRichCurve(
    SourceMinimum=0
    TargetMinimum=0
    Value=RigVMDispatchNode(False=@n5.Result, True=@Multiply.Result, Condition=RigVMVariableNode() {variable=ease}.Value) {template="DISPATCH_RigVMDispatch_If(in Condition,in True,in False,out Result)"}.Result
    Curve=RigVMVariableNode() {variable=remapCurve}.Value
   )
   > Set_At_1 = RigVMDispatchNode(Index=@For_Each_1.Index, Element=@RigVMFunction_AnimEvalRichCurve.Result, Array=RigVMVariableNode() {variable=outputValues}.Value) {template="DISPATCH_RigVMDispatch_ArraySetAtIndex(io Array,in Index,in Element)"}
   RigVMFunctionReturnNode(weights=RigVMVariableNode() {variable=outputValues}.Value, ExecuteContext=@For_Each_1.Completed)
  ]}
 New Function_ContainedGraph {nodes [
   n1 = RigVMFunctionEntryNode()
   patchGroinGlute = RigVMCollapseNode(ExecutePin=@n1.ExecuteContext)
   RigVMFunction_Sequence = RigVMAggregateNode(ExecuteContext=@patchGroinGlute.ExecutePin)
   GetAlphaFromQuatDelta = RigVMFunctionReferenceNode(
    Rotation Order=ZYX
    child=(Type=Bone,Name="thigh_l")
    debugRotation=false
    outputMinMax=(X=0,Y=1)
    parent=(Type=Null,Name="thigh_l_null")
    premult=1
    rotationMinMax=(X=0,Y=-100)
    triggerAxis=Z
    twistAxis=(X=1,Y=0,Z=0)
    ExecuteContext=@RigVMFunction_Sequence.A
   )
   > RigVMVariableNode(Value=@GetAlphaFromQuatDelta.alpha) {variable=thigh_l_rz_0_100_local}
   > GetAlphaFromQuatDelta_1 = RigVMFunctionReferenceNode(
    Rotation Order=ZYX
    child=(Type=Bone,Name="thigh_r")
    outputMinMax=(X=0,Y=1)
    parent=(Type=Null,Name="thigh_r_null")
    premult=1
    rotationMinMax=(X=0,Y=-100)
    triggerAxis=Z
    twistAxis=(X=1,Y=0,Z=0)
   )
   > RigVMVariableNode(Value=@GetAlphaFromQuatDelta_1.alpha) {variable=thigh_r_rz_0_100_local}
   customKneeFix = RigVMCollapseNode(ExecuteContext=@RigVMFunction_Sequence.B)
   ankleSetup = RigVMCollapseNode(ExecutePin=@customKneeFix.ExecuteContext)
   n2 = RigVMVariableNode() {variable=thigh_l_rz_0_100_local}
   n3 = RigVMVariableNode() {variable=thigh_r_rz_0_100_local}
   strapsSetup = RigVMCollapseNode(ExecutePin=@RigVMFunction_Sequence.C, thigh_l_rz_0_100=@n2.Value, thigh_r_rz_0_100=@n3.Value)
   chainsSetup = RigVMCollapseNode(ExecutePin=@strapsSetup.ExecutePin, thigh_r_rz_0_100=@n3.Value)
   jacketHemSetup = RigVMCollapseNode(ExecutePin=@chainsSetup.ExecutePin, thigh_l_rz_0_100=@n2.Value, thigh_r_rz_0_100=@n3.Value)
   carabinersSetup = RigVMCollapseNode(ExecutePin=@jacketHemSetup.ExecutePin)
   carabiners_p_hemSetup = RigVMCollapseNode(ExecuteContext=@carabinersSetup.ExecutePin)
   misc = RigVMCollapseNode(ExecutePin=@carabiners_p_hemSetup.ExecuteContext, thigh_l_rz_0_100=@n2.Value, thigh_r_rz_0_100=@n3.Value)
   RigVMFunctionReturnNode(ExecuteContext=@misc.ExecutePin)
  ]}
 RigVMFunctionLibrary {nodes [
   DeltaQuatExtraction = RigVMCollapseNode(ExecuteContext=(), child=(Type=Bone,Name="None"), parent=(Type=Bone,Name="None"), twistAxis=(X=1,Y=0,Z=0))
   GetAlphaFromQuatDelta = RigVMCollapseNode(
    ExecuteContext=()
    Rotation Order=ZYX
    child=(Type=Bone,Name="None")
    outputMinMax=(X=0,Y=1)
    parent=(Type=Bone,Name="None")
    premult=1
    rotationMinMax=(X=0,Y=0)
    twistAxis=(X=1,Y=0,Z=0)
   )
   InverseDistance = RigVMCollapseNode(
    ExecuteContext=()
    child=(Type=Bone,Name="None")
    drawLines=true
    parent=(Type=Bone,Name="None")
    primaryAxis=(X=0,Y=0,Z=0)
    remapCurve=(EditorCurveData=(Keys=((),(Time=0.5,Value=0.5,ArriveTangent=1.203,LeaveTangent=1.203),(Time=1,Value=1))))
    remapMult=4
    targets=()
   )
   IvoryDwellChar_lowerBody = RigVMCollapseNode(ExecuteContext=())
  ]}
 RigVMFunction_Sequence_1_ContainedGraph {nodes [
   n1 = RigVMFunctionEntryNode()
   RigVMFunction_Sequence = &RigVMFunction_Sequence(ExecuteContext=@n1.ExecuteContext)
   n2 = &RigVMFunction_Sequence(ExecuteContext=@RigVMFunction_Sequence.B)
   RigVMFunctionReturnNode(A=@RigVMFunction_Sequence.A, B=@n2.A, C=@n2.B)
  ]}
 RigVMFunction_Sequence_1_ContainedGraph {nodes [
   n1 = RigVMFunctionEntryNode()
   RigVMFunction_Sequence = &RigVMFunction_Sequence(ExecuteContext=@n1.ExecuteContext)
   n2 = &RigVMFunction_Sequence(ExecuteContext=@RigVMFunction_Sequence.B)
   RigVMFunctionReturnNode(A=@RigVMFunction_Sequence.A, B=@n2.A, C=@n2.B)
  ]}
 RigVMFunction_Sequence_1_ContainedGraph {nodes [
   n1 = RigVMFunctionEntryNode()
   RigVMFunction_Sequence = &RigVMFunction_Sequence(ExecuteContext=@n1.ExecuteContext)
   n2 = &RigVMFunction_Sequence(ExecuteContext=@RigVMFunction_Sequence.B)
   RigVMFunctionReturnNode(A=@RigVMFunction_Sequence.A, B=@n2.A, C=@n2.B)
  ]}
 RigVMFunction_Sequence_1_ContainedGraph {nodes [
   n1 = RigVMFunctionEntryNode()
   RigVMFunction_Sequence = &RigVMFunction_Sequence(ExecuteContext=@n1.ExecuteContext)
   n2 = &RigVMFunction_Sequence(ExecuteContext=@RigVMFunction_Sequence.B)
   RigVMFunctionReturnNode(A=@RigVMFunction_Sequence.A, B=@n2.A, C=@n2.B)
  ]}
 RigVMFunction_Sequence_1_ContainedGraph {nodes [
   n1 = RigVMFunctionEntryNode()
   RigVMFunction_Sequence = &RigVMFunction_Sequence(ExecuteContext=@n1.ExecuteContext)
   n2 = &RigVMFunction_Sequence(ExecuteContext=@RigVMFunction_Sequence.B)
   RigVMFunctionReturnNode(A=@RigVMFunction_Sequence.A, B=@n2.A, C=@n2.B)
  ]}
 RigVMModel {nodes [
   IvoryDwellChar_lowerBody = RigVMFunctionReferenceNode(ExecuteContext=())
   RigUnit_BeginExecution = &RigUnit_BeginExecution()
  ]}
]
hierarchy [
 root {loc=<0 0 0>, rot=<0 0 0 1>}
 pelvis {parent=root, loc=<0 -0.7071 93.1191>, rot=<0.0127 0.707 -0.0127 -0.707>}
 spine_01 {parent=pelvis, loc=<6.7494 -0.6388 0>, rot=<0 0 0.034 -0.9994>}
 spine_02 {parent=spine_01, loc=<7.2518 -0.035 0>, rot=<0 0 0.0238 -0.9997>}
 spine_03 {parent=spine_02, loc=<6.0104 0 0>, rot=<0 0 0.0149 -0.9999>}
 spine_04 {parent=spine_03, loc=<8.1024 0 0>, rot=<0 0 -0.0132 -0.9999>}
 spine_05 {parent=spine_04, loc=<10.0146 0 0>, rot=<0 0 -0.0015 -1>}
 neck_01 {parent=spine_05, loc=<11.3138 0.1529 0>, rot=<0 0 -0.2003 -0.9797>}
 neck_02 {parent=neck_01, loc=<4.5627 0 0>, rot=<0 0 0.0504 -0.9987>}
 head {parent=neck_02, loc=<5.3914 0 0>, rot=<0 0 0.115 -0.9934>}
 attach_cape {parent=spine_05, loc=<9.017 -11.714 0>, rot=<0.0616 -0.7044 -0.0616 -0.7044>}
 cape {parent=attach_cape, type=Null, loc=<0 -2.1 -5.2>, rot=<0 0 0 1>}
 attach_backpack {parent=spine_05, loc=<-0.123 -13.473 0>, rot=<-0.0093 -0.707 0.0093 -0.707>}
 Backpack_BR {parent=attach_backpack, type=Null, loc=<0 -0.4786 -1.4116>, rot=<0 0 0 1>}
 clavicle_l {parent=spine_05, loc=<4.3575 -0.1466 -2.3223>, rot=<0.0465 -0.7289 -0.009 -0.6829>}
 upperarm_l {parent=clavicle_l, loc=<15.5111 0 0>, rot=<0.0226 -0.3234 -0.0088 -0.9459>}
 lowerarm_l {parent=upperarm_l, loc=<26.6514 0 0>, rot=<0.0391 -0.0099 -0.2453 -0.9686>}
 hand_l {parent=lowerarm_l, loc=<22.3491 0 0>, rot=<-0.7081 -0.0507 0.0975 -0.6975>}
 thumb_01_l {parent=hand_l, loc=<1.3118 -0.3191 -2.459>, rot=<0.6546 -0.4042 0.0664 -0.6354>}
 thumb_02_l {parent=thumb_01_l, loc=<5.9336 -0.0448 0.0401>, rot=<-0.0037 0.0213 0.298 -0.9543>}
 thumb_03_l {parent=thumb_02_l, loc=<2.2223 0.3682 -0.0907>, rot=<0.0118 -0.013 0.0066 -0.9998>}
 index_metacarpal_l {parent=hand_l, loc=<1.6535 0.0183 -2.96>, rot=<0.0158 -0.0124 -0.0173 -0.9996>}
 index_01_l {parent=index_metacarpal_l, loc=<8.3008 0.0357 -0.0524>, rot=<-0.0044 0.0485 0.0924 -0.9945>}
 index_02_l {parent=index_01_l, loc=<3.3587 -0.3004 0.0452>, rot=<-0.002 0.0027 0.3225 -0.9466>}
 index_03_l {parent=index_02_l, loc=<2.5578 -0.123 -0.0252>, rot=<-0.002 0.0067 0.2558 -0.9667>}
 middle_metacarpal_l {parent=hand_l, loc=<2.1856 -0.1725 -0.5984>, rot=<0.006 0.0168 -0.0329 -0.9993>}
 middle_01_l {parent=middle_metacarpal_l, loc=<7.3275 -0.2447 0.0129>, rot=<-0.0208 0.0042 0.0858 -0.9961>}
 middle_02_l {parent=middle_01_l, loc=<4.0123 0.1401 0.1175>, rot=<-0.0218 0.0258 0.3512 -0.9357>}
 middle_03_l {parent=middle_02_l, loc=<2.6077 0.0534 0.1373>, rot=<0.0082 0.0188 0.1708 -0.9851>}
 ring_metacarpal_l {parent=hand_l, loc=<2.2074 -0.3625 1.3638>, rot=<-0.0284 0.0722 -0.0458 -0.9959>}
 ring_01_l {parent=ring_metacarpal_l, loc=<7.0295 -0.6197 -0.1446>, rot=<-0.0489 -0.0252 0.1892 -0.9804>}
 ring_02_l {parent=ring_01_l, loc=<3.6174 0.4647 -0.1027>, rot=<0 0 0.2806 -0.9598>}
 ring_03_l {parent=ring_02_l, loc=<2.4152 0.114 0.0199>, rot=<0 0 0.1189 -0.9929>}
 pinky_metacarpal_l {parent=hand_l, loc=<1.7616 -0.9753 3.3814>, rot=<-0.0397 0.0814 -0.069 -0.9935>}
 pinky_01_l {parent=pinky_metacarpal_l, loc=<7.1709 -0.4831 -0.1094>, rot=<-0.0654 -0.0035 0.2242 -0.9723>}
 pinky_02_l {parent=pinky_01_l, loc=<2.8274 0.0893 -0.1043>, rot=<0.0037 0.0011 0.2854 -0.9584>}
 pinky_03_l {parent=pinky_02_l, loc=<2.2046 0.0307 -0.0078>, rot=<-0.0002 0.0098 0.1445 -0.9895>}
 deform_wrist_dn_base_l {parent=hand_l, loc=<0 -2.2342 0>, rot=<0 0 0 -1>}
 deform_wrist_dn_l {parent=deform_wrist_dn_base_l, loc=<0 0 0>, rot=<0 0 0 1>}
 deform_wrist_up_base_l {parent=hand_l, loc=<0 2.2682 0>, rot=<0 0 0 -1>}
 deform_wrist_up_l {parent=deform_wrist_up_base_l, loc=<0 0 0>, rot=<0 0 0 1>}
 lowerarm_twist_02_l {parent=lowerarm_l, loc=<9.6199 0 0>, rot=<0 0 0 -1>}
 lowerarm_twist_01_l {parent=lowerarm_l, loc=<17.8549 0 0>, rot=<0 0 0 -1>}
 deform_elbow_in_base_l {parent=lowerarm_l, loc=<0.2112 5.259 0.8274>, rot=<0 0 0 -1>}
 deform_elbow_in_l {parent=deform_elbow_in_base_l, loc=<0 0 0>, rot=<0 0 0 1>}
 deform_elbow_out_base_l {parent=lowerarm_l, loc=<0 -4 0>, rot=<0 0 0 -1>}
 deform_elbow_out_l {parent=deform_elbow_out_base_l, loc=<0 0 0>, rot=<0 0 0 1>}
 upperarm_twist_01_l {parent=upperarm_l, loc=<0.0041 0 0>, rot=<0 0 0 -1>}
 deform_upperarm_bk_base_l {parent=upperarm_twist_01_l, loc=<1.0813 -7.2193 -1.9453>, rot=<0 0 0 1>}
 deform_upperarm_bk_l {parent=deform_upperarm_bk_base_l, loc=<0 0 0>, rot=<0 0 0 1>}
 deform_upperarm_fr_base_l {parent=upperarm_twist_01_l, loc=<-0.3691 7.0506 -2.765>, rot=<0 0 0 1>}
 deform_upperarm_fr_l {parent=deform_upperarm_fr_base_l, loc=<0 0 0>, rot=<0 0 0 1>}
 deform_upperarm_up_base_l {parent=upperarm_twist_01_l, loc=<-3.2216 -1.1391 3.3311>, rot=<0 0 0 1>}
 deform_upperarm_up_l {parent=deform_upperarm_up_base_l, loc=<0 0 0>, rot=<0 0 0 1>}
 upperarm_twist_02_l {parent=upperarm_l, loc=<13.7941 0 0>, rot=<0 0 0 -1>}
 clavicle_r {parent=spine_05, loc=<4.3575 -0.1466 2.3223>, rot=<0.7289 0.0465 -0.6829 0.009>}
 upperarm_r {parent=clavicle_r, loc=<-15.5111 0 0>, rot=<0.0226 -0.3234 -0.0088 -0.9459>}
 lowerarm_r {parent=upperarm_r, loc=<-26.6514 0 0>, rot=<0.0391 -0.0099 -0.2453 -0.9686>}
 hand_r {parent=lowerarm_r, loc=<-22.3491 0 0>, rot=<-0.7081 -0.0507 0.0975 -0.6975>}
 thumb_01_r {parent=hand_r, loc=<-1.3118 0.3191 2.459>, rot=<0.6546 -0.4042 0.0664 -0.6354>}
 thumb_02_r {parent=thumb_01_r, loc=<-5.9336 0.0448 -0.0401>, rot=<-0.0037 0.0213 0.298 -0.9543>}
 thumb_03_r {parent=thumb_02_r, loc=<-2.2223 -0.3682 0.0907>, rot=<0.0118 -0.013 0.0066 -0.9998>}
 index_metacarpal_r {parent=hand_r, loc=<-1.6535 -0.0183 2.96>, rot=<0.0158 -0.0124 -0.0173 -0.9996>}
 index_01_r {parent=index_metacarpal_r, loc=<-8.3008 -0.0357 0.0524>, rot=<-0.0044 0.0485 0.0924 -0.9945>}
 index_02_r {parent=index_01_r, loc=<-3.3587 0.3004 -0.0452>, rot=<-0.002 0.0027 0.3225 -0.9466>}
 index_03_r {parent=index_02_r, loc=<-2.5578 0.123 0.0252>, rot=<-0.002 0.0067 0.2558 -0.9667>}
 middle_metacarpal_r {parent=hand_r, loc=<-2.1856 0.1725 0.5984>, rot=<0.006 0.0168 -0.0329 -0.9993>}
 middle_01_r {parent=middle_metacarpal_r, loc=<-7.3275 0.2447 -0.0129>, rot=<-0.0208 0.0042 0.0858 -0.9961>}
 middle_02_r {parent=middle_01_r, loc=<-4.0123 -0.1401 -0.1175>, rot=<-0.0218 0.0258 0.3512 -0.9357>}
 middle_03_r {parent=middle_02_r, loc=<-2.6077 -0.0534 -0.1373>, rot=<0.0082 0.0188 0.1708 -0.9851>}
 ring_metacarpal_r {parent=hand_r, loc=<-2.2074 0.3625 -1.3638>, rot=<-0.0284 0.0722 -0.0458 -0.9959>}
 ring_01_r {parent=ring_metacarpal_r, loc=<-7.0295 0.6197 0.1446>, rot=<-0.0489 -0.0252 0.1892 -0.9804>}
 ring_02_r {parent=ring_01_r, loc=<-3.6174 -0.4647 0.1027>, rot=<0 0 0.2806 -0.9598>}
 ring_03_r {parent=ring_02_r, loc=<-2.4152 -0.114 -0.0199>, rot=<0 0 0.1189 -0.9929>}
 pinky_metacarpal_r {parent=hand_r, loc=<-1.7616 0.9753 -3.3814>, rot=<-0.0397 0.0814 -0.069 -0.9935>}
 pinky_01_r {parent=pinky_metacarpal_r, loc=<-7.1709 0.4831 0.1094>, rot=<-0.0654 -0.0035 0.2242 -0.9723>}
 pinky_02_r {parent=pinky_01_r, loc=<-2.8274 -0.0893 0.1043>, rot=<0.0037 0.0011 0.2854 -0.9584>}
 pinky_03_r {parent=pinky_02_r, loc=<-2.2046 -0.0307 0.0078>, rot=<-0.0002 0.0098 0.1445 -0.9895>}
 deform_wrist_dn_base_r {parent=hand_r, loc=<-0.0004 2.2345 -0.0001>, rot=<0 0 0 -1>}
 deform_wrist_dn_r {parent=deform_wrist_dn_base_r, loc=<0 0 0>, rot=<0 0 0 1>}
 deform_wrist_up_base_r {parent=hand_r, loc=<0.0003 -2.2684 0.0001>, rot=<0 0 0 -1>}
 deform_wrist_up_r {parent=deform_wrist_up_base_r, loc=<0 0 0>, rot=<0 0 0 1>}
 lowerarm_twist_02_r {parent=lowerarm_r, loc=<-9.6199 0 0>, rot=<0 0 0 -1>}
 lowerarm_twist_01_r {parent=lowerarm_r, loc=<-17.8549 0 0>, rot=<0 0 0 -1>}
 deform_elbow_in_base_r {parent=lowerarm_r, loc=<-0.2111 -5.2591 -0.8275>, rot=<0 0 0 -1>}
 deform_elbow_in_r {parent=deform_elbow_in_base_r, loc=<-0.0001 0 0.0002>, rot=<0 0 0 1>}
 deform_elbow_out_base_r {parent=lowerarm_r, loc=<0 4 -0.0001>, rot=<0 0 0 -1>}
 deform_elbow_out_r {parent=deform_elbow_out_base_r, loc=<0 0 0>, rot=<0 0 0 1>}
 upperarm_twist_01_r {parent=upperarm_r, loc=<-0.0041 0 0>, rot=<0 0 0 -1>}
 deform_upperarm_bk_base_r {parent=upperarm_twist_01_r, loc=<-1.0814 7.2192 1.9455>, rot=<0 0 0 1>}
 deform_upperarm_bk_r {parent=deform_upperarm_bk_base_r, loc=<0 0 0>, rot=<0 0 0 1>}
 deform_upperarm_fr_base_r {parent=upperarm_twist_01_r, loc=<0.3689 -7.0506 2.7652>, rot=<0 0 0 1>}
 deform_upperarm_fr_r {parent=deform_upperarm_fr_base_r, loc=<0 0 0>, rot=<0 0 0 1>}
 deform_upperarm_up_base_r {parent=upperarm_twist_01_r, loc=<3.2214 1.1391 -3.3308>, rot=<0 0 0 1>}
 deform_upperarm_up_r {parent=deform_upperarm_up_base_r, loc=<0 0 0>, rot=<0 0 0 -1>}
 upperarm_twist_02_r {parent=upperarm_r, loc=<-13.7941 0 0>, rot=<0 0 0 -1>}
 base_jacketCollar_1_l {parent=spine_05, loc=<7.9885 6.5741 -7.859>, rot=<0.8424 0.4031 -0.3441 -0.0967>}
 dyn_jacketCollar_1_l {parent=base_jacketCollar_1_l, loc=<0 0 0>, rot=<0 0 0 1>}
 base_jacketCollar_1_r {parent=spine_05, loc=<7.5644 6.6 8.2754>, rot=<0.1555 0.3723 -0.3485 -0.8461>}
 dyn_jacketCollar_1_r {parent=base_jacketCollar_1_r, loc=<0 0 0>, rot=<0 0 0 1>}
 base_hoodieCollar_1_c {parent=spine_05, loc=<12.2591 -3.9745 0>, rot=<0.6676 -0.2331 0.2331 -0.6676>}
 dyn_hoodieCollar_1_c {parent=base_hoodieCollar_1_c, loc=<0 0 0>, rot=<0 0 0 1>}
 base_hoodieCollar_3_l {parent=spine_05, loc=<8.7833 5.0518 -4.1504>, rot=<-0.2874 -0.3328 -0.3495 -0.8274>}
 dyn_hoodieCollar_3_l {parent=base_hoodieCollar_3_l, loc=<0 0 0>, rot=<0 0 0 1>}
 base_hoodieCollar_2_l {parent=spine_05, loc=<12.0968 0.7383 -6.6949>, rot=<0.0201 -0.3514 -0.1012 -0.9305>}
 dyn_hoodieCollar_2_l {parent=base_hoodieCollar_2_l, loc=<0 0 0>, rot=<0 0 0 1>}
 base_hoodieCollar_1_l {parent=spine_05, loc=<12.6465 -2.9942 -4.4824>, rot=<0.413 -0.2851 0.1092 -0.858>}
 dyn_hoodieCollar_1_l {parent=base_hoodieCollar_1_l, loc=<0 0 0>, rot=<0 0 0 1>}
 base_hoodieCollar_3_r {parent=spine_05, loc=<8.7834 5.0519 4.1504>, rot=<0.8274 0.3495 0.3328 0.2874>}
 dyn_hoodieCollar_3_r {parent=base_hoodieCollar_3_r, loc=<0 0 0>, rot=<0 0 0 1>}
 base_hoodieCollar_2_r {parent=spine_05, loc=<12.0967 0.7383 6.6949>, rot=<0.9305 0.1012 0.3514 -0.0201>}
 dyn_hoodieCollar_2_r {parent=base_hoodieCollar_2_r, loc=<0 0 0>, rot=<0 0 0 1>}
 base_hoodieCollar_1_r {parent=spine_05, loc=<12.6466 -2.9942 4.4824>, rot=<0.858 -0.1092 0.2851 -0.413>}
 dyn_hoodieCollar_1_r {parent=base_hoodieCollar_1_r, loc=<0 0 0>, rot=<0 0 0 1>}
 base_hood {parent=spine_05, loc=<16.6559 -6.0763 0>, rot=<0.3664 -0.9305 0 0>}
 dyn_hood {parent=base_hood, loc=<0 0 0>, rot=<0 0 0 1>}
 deform_pec_aim_base_l {parent=spine_04, loc=<7.8732 12.6388 -2.4123>, rot=<0.0102 -0.4602 0.3979 -0.7936>}
 deform_pec_aim_l {parent=deform_pec_aim_base_l, loc=<0 0 0>, rot=<0 0 0 1>}
 deform_pec_base_l {parent=deform_pec_aim_l, loc=<10.766 0 0>, rot=<0 0 0 1>}
 deform_pec_l {parent=deform_pec_base_l, loc=<0.055 0 0>, rot=<0 0 0 1>}
 deform_pec_aim_base_r {parent=spine_04, loc=<7.8727 12.6388 2.4123>, rot=<-0.4602 -0.0102 0.7936 0.3979>}
 deform_pec_aim_r {parent=deform_pec_aim_base_r, loc=<0 0 0>, rot=<0 0 0 1>}
 deform_pec_base_r {parent=deform_pec_aim_r, loc=<-10.7662 -0.0006 -0.0007>, rot=<0 0 0 1>}
 deform_pec_r {parent=deform_pec_base_r, loc=<0 0 0>, rot=<0 0 0 1>}
 deform_lat_base_l {parent=spine_04, loc=<-0.11 0.3046 -15.4494>, rot=<0.0294 -0.7065 -0.0294 -0.7065>}
 deform_lat_l {parent=deform_lat_base_l, loc=<0 0 0>, rot=<0 0 0 1>}
 deform_lat_base_r {parent=spine_04, loc=<-0.11 0.3046 15.4494>, rot=<0.7065 0.0294 0.7065 -0.0294>}
 deform_lat_r {parent=deform_lat_base_r, loc=<0 0 0>, rot=<0 0 0 1>}
 kin_hemJacket {parent=spine_02, loc=<0 0 0>, rot=<0 0 0 1>}
 base_hemJacket_1_r {parent=kin_hemJacket, loc=<1.1219 13.0098 9.0976>, rot=<0.0283 -0.2471 0.9685 -0.0068>}
 dyn_hemJacket_1_r {parent=base_hemJacket_1_r, loc=<0 0 0>, rot=<0 0 0 1>}
 base_hemJacket_1_l {parent=kin_hemJacket, loc=<1.1224 13.0099 -9.0976>, rot=<-0.0283 0.2471 0.9685 -0.0068>}
 dyn_hemJacket_1_l {parent=base_hemJacket_1_l, loc=<0 0 0>, rot=<0 0 0 1>}
 base_hemJacket_2_r {parent=kin_hemJacket, loc=<1.8724 5.6926 16.8384>, rot=<0.0707 -0.5709 0.8179 0.0098>}
 dyn_hemJacket_2_r {parent=base_hemJacket_2_r, loc=<0 0 0>, rot=<0 0 0 1>}
 base_hemJacket_2_l {parent=kin_hemJacket, loc=<1.8725 5.6926 -16.8384>, rot=<-0.0707 0.5709 0.8179 0.0098>}
 dyn_hemJacket_2_l {parent=base_hemJacket_2_l, loc=<0 0 0>, rot=<0 0 0 1>}
 base_hemJacket_3_r {parent=kin_hemJacket, loc=<2.7438 -6.4373 15.7136>, rot=<0.0567 -0.85 0.5235 0.0163>}
 dyn_hemJacket_3_r {parent=base_hemJacket_3_r, loc=<0 0 0>, rot=<0 0 0 1>}
 base_hemJacket_3_l {parent=kin_hemJacket, loc=<2.7435 -6.4373 -15.7136>, rot=<0.0567 -0.85 -0.5235 -0.0163>}
 dyn_hemJacket_3_l {parent=base_hemJacket_3_l, loc=<0 0 0>, rot=<0 0 0 1>}
 base_hemJacket_1_c {parent=kin_hemJacket, loc=<2.741 -12.3031 -0.4968>, rot=<0.0399 -0.9992 0 0>}
 dyn_hemJacket_1_c {parent=base_hemJacket_1_c, loc=<0 0 0>, rot=<0 0 0 1>}
 kin_delta_hemJacket_1_l {parent=kin_hemJacket, loc=<1.6216 13.0234 -9.0719>, rot=<-0.0283 0.2471 0.9685 -0.0068>}
 kin_delta_hemJacket_1_r {parent=kin_hemJacket, loc=<1.6211 13.0234 9.0719>, rot=<0.0283 -0.2471 0.9685 -0.0068>}
 thigh_l {parent=pelvis, loc=<-2.5693 -0.0109 -7.7501>, rot=<0.0012 0.0572 0.0155 -0.9982>}
 calf_l {parent=thigh_l, loc=<-39.0484 0 0>, rot=<0.0006 -0.0161 -0.0391 -0.9991>}
 foot_l {parent=calf_l, loc=<-42.561 0.1077 0.4292>, rot=<0.0418 -0.0455 0.0343 -0.9975>}
 ball_l {parent=foot_l, loc=<-6.1599 13.8264 0.6806>, rot=<0.0001 -0.0001 -0.7063 -0.7079>}
 deform_ankle_in_base_l {parent=foot_l, loc=<2.9481 5.9274 0.65>, rot=<-0.0253 0.0419 -0.526 -0.8491>}
 deform_ankle_in_l {parent=deform_ankle_in_base_l, loc=<0 0 0>, rot=<0 0 0 1>}
 deform_ankle_out_base_l {parent=foot_l, loc=<-2.4371 -4.9 -0.5374>, rot=<-0.0253 0.0419 -0.526 -0.8491>}
 deform_ankle_out_l {parent=deform_ankle_out_base_l, loc=<0 0 0>, rot=<0 0 0 1>}
 calf_twist_02_l {parent=calf_l, loc=<-18.8064 -0.0967 0.0397>, rot=<0.0237 0.0095 -0.0035 -0.9997>}
 calf_twist_01_l {parent=calf_l, loc=<-35.813 -0.0967 0.0397>, rot=<0.0237 0.0095 -0.0035 -0.9997>}
 deform_knee_in_base_l {parent=calf_l, loc=<1.298 -5.2957 -0.0496>, rot=<0 0 0 -1>}
 deform_knee_in_l {parent=deform_knee_in_base_l, loc=<0 0 0>, rot=<0 0 0 1>}
 deform_knee_out_base_l {parent=calf_l, loc=<0 4 0>, rot=<0 0 0 -1>}
 deform_knee_out_l {parent=deform_knee_out_base_l, loc=<0 0 0>, rot=<0 0 0 1>}
 calf_twist_00_l {parent=calf_l, loc=<0 0 0>, rot=<0 0 0 1>}
 thigh_twist_01_l {parent=thigh_l, loc=<-21.6303 0 0>, rot=<0.0013 0 0 -1>}
 thigh_tshirt_l {parent=thigh_twist_01_l, loc=<20.6303 0 0>, rot=<-0.0013 0 0 -1>}
 deform_groin_base_l {parent=thigh_l, loc=<-2.1908 8.1142 -1.7442>, rot=<0 0 0 -1>}
 deform_groin_l {parent=deform_groin_base_l, loc=<0 0 0>, rot=<0 0 0 1>}
 thigh_twist_00_l {parent=thigh_l, loc=<-24.348 0 0>, rot=<0.0013 0 0 -1>}
 base_strapHang_out_l {parent=thigh_twist_00_l, loc=<-0.0785 0.9313 -0.9309>, rot=<-0.0028 -0.9153 -0.0002 -0.4027>}
 dyn_strapHang_out_l {parent=base_strapHang_out_l, loc=<0 0 0>, rot=<0 0 0 1>}
 base_strapHang_in_l {parent=thigh_twist_00_l, loc=<-0.2906 0.9234 2.3819>, rot=<0.3453 -0.0021 0.9385 0.0018>}
 dyn_strapHang_in_l {parent=base_strapHang_in_l, loc=<0 0 0>, rot=<0 0 0 1>}
 kin_strap_frontTris_l {parent=thigh_twist_00_l, loc=<1.2 10 1.2>, rot=<0 0 0 1>}
 target_strap_01_l {parent=thigh_l, loc=<-25.6021 10.4002 0.6359>, rot=<0 0 0 1>}
 target_strap_02_l {parent=thigh_l, loc=<-32.9711 3.8803 -9.6501>, rot=<0 0 0 1>}
 coll_strap_01_l {parent=thigh_l, loc=<-8.3471 10.8121 3.5943>, rot=<0.0117 -0.9963 0.0013 -0.0848>}
 coll_strap_02_l {parent=thigh_l, loc=<-23.9279 10.1298 -0.7387>, rot=<0.2091 -0.901 0.0877 -0.3698>}
 thigh_r {parent=pelvis, loc=<-2.5693 -0.0109 7.7501>, rot=<0.0572 -0.0012 0.9982 0.0155>}
 calf_r {parent=thigh_r, loc=<39.0484 0 0>, rot=<0.0006 -0.0161 -0.0391 -0.9991>}
 foot_r {parent=calf_r, loc=<42.561 -0.1077 -0.4292>, rot=<0.0418 -0.0455 0.0343 -0.9975>}
 ball_r {parent=foot_r, loc=<6.1599 -13.8264 -0.6806>, rot=<0.0001 -0.0001 -0.7063 -0.7079>}
 deform_ankle_in_base_r {parent=foot_r, loc=<-2.9481 -5.9274 -0.6501>, rot=<-0.0253 0.0419 -0.526 -0.8491>}
 deform_ankle_in_r {parent=deform_ankle_in_base_r, loc=<0 0 0>, rot=<0 0 0 1>}
 deform_ankle_out_base_r {parent=foot_r, loc=<2.4371 4.9 0.5373>, rot=<-0.0253 0.0419 -0.526 -0.8491>}
 deform_ankle_out_r {parent=deform_ankle_out_base_r, loc=<0 0 0>, rot=<0 0 0 1>}
 calf_twist_02_r {parent=calf_r, loc=<18.8064 0.0967 -0.0397>, rot=<0.0237 0.0095 -0.0035 -0.9997>}
 base_calfLace_1_l {parent=calf_twist_02_r, loc=<12.3206 4.037 4.6723>, rot=<0.8692 -0.099 0.3205 -0.3632>}
 dyn_calfLace_1_l {parent=base_calfLace_1_l, loc=<0 0 0>, rot=<0 0 0 1>}
 dyn_calfLace_2_l {parent=dyn_calfLace_1_l, loc=<3.623 0 0>, rot=<0.0857 0.1663 0.0775 -0.9793>}
 calf_twist_01_r {parent=calf_r, loc=<35.813 0.0967 -0.0397>, rot=<0.0237 0.0095 -0.0035 -0.9997>}
 deform_knee_in_base_r {parent=calf_r, loc=<-1.298 5.2957 0.0496>, rot=<0 0 0 -1>}
 deform_knee_in_r {parent=deform_knee_in_base_r, loc=<0 0 0>, rot=<0 0 0 1>}
 deform_knee_out_base_r {parent=calf_r, loc=<0 -4 0>, rot=<0 0 0 -1>}
 deform_knee_out_r {parent=deform_knee_out_base_r, loc=<0 0 0>, rot=<0 0 0 1>}
 calf_twist_00_r {parent=calf_r, loc=<0 0 0>, rot=<0 0 0 1>}
 thigh_twist_01_r {parent=thigh_r, loc=<21.6303 0 0>, rot=<0.0013 0 0 -1>}
 thigh_tshirt_r {parent=thigh_twist_01_r, loc=<-20.6303 0 0>, rot=<-0.0013 0 0 -1>}
 deform_groin_base_r {parent=thigh_r, loc=<2.1908 -8.1142 1.7442>, rot=<0 0 0 -1>}
 deform_groin_r {parent=deform_groin_base_r, loc=<0 0 0>, rot=<0 0 0 1>}
 thigh_twist_00_r {parent=thigh_r, loc=<24.348 0 0>, rot=<0.0013 0 0 -1>}
 base_strapHang_out_r {parent=thigh_twist_00_r, loc=<0.0785 -0.9313 0.9309>, rot=<-0.0002 0.4027 0.0028 -0.9153>}
 dyn_strapHang_out_r {parent=base_strapHang_out_r, loc=<0 0 0>, rot=<0 0 0 1>}
 base_strapHang_in_r {parent=thigh_twist_00_r, loc=<0.2906 -0.9234 -2.3818>, rot=<0.9526 -0.0019 -0.3042 -0.002>}
 dyn_strapHang_in_r {parent=base_strapHang_in_r, loc=<0 0 0>, rot=<0 0 0 1>}
 kin_strap_frontTris_r {parent=thigh_twist_00_r, loc=<-1.2 -10 -1.2>, rot=<0 0 0 1>}
 deform_glute_base_l {parent=pelvis, loc=<-8.8064 -8.8222 -8.4287>, rot=<0.0275 0.0501 -0.4472 -0.8926>}
 deform_glute_l {parent=deform_glute_base_l, loc=<0 0 0>, rot=<0 0 0 1>}
 deform_glute_base_r {parent=pelvis, loc=<-8.8062 -8.8222 8.4287>, rot=<0.0501 -0.0275 0.8926 -0.4472>}
 deform_glute_r {parent=deform_glute_base_r, loc=<0 0 0>, rot=<0 0 0 1>}
 base_hipChains {parent=pelvis, loc=<-3.9069 3.4534 12.7954>, rot=<0.0075 -0.975 0.1399 0.1726>}
 dyn_hipChains {parent=base_hipChains, loc=<0 0 0>, rot=<0 0 0 1>}
 dyn_hipChains_front {parent=dyn_hipChains, loc=<8.2695 7.1653 1.0993>, rot=<0 0 0 1>}
 dyn_hipChains_back {parent=dyn_hipChains, loc=<9.5755 -7.7709 -0.2513>, rot=<0 0 0 1>}
 kin_carabiners_01_r {parent=pelvis, loc=<-0.1112 9.6521 11.9341>, rot=<-0.2431 -0.6744 -0.1867 -0.6718>}
 kin_hoodie_hem_01_r {parent=pelvis, loc=<5.885 7.3222 12.7507>, rot=<-0.0127 -0.707 0.0127 -0.707>}
 kin_carabiners_01_l {parent=pelvis, loc=<-0.3502 8.6716 -13.0299>, rot=<0.6951 0.13 0.6899 0.1548>}
 base_hangCarabinerF_1 {parent=kin_carabiners_01_l, loc=<-0.0253 -0.2536 -4.1373>, rot=<-0.4976 -0.4826 0.5393 -0.4782>}
 dyn_hangCarabinerF_1 {parent=base_hangCarabinerF_1, loc=<0 0 0>, rot=<0 0 0 1>}
 dyn_hangCarabinerF_2 {parent=dyn_hangCarabinerF_1, loc=<5.4674 0 0>, rot=<-0.0399 0.0392 -0.0016 -0.9984>}
 kin_carabiners_02_l {parent=pelvis, loc=<-0.1246 5.3503 -16.0407>, rot=<0.6206 0.3388 0.6081 0.3608>}
 base_hangCarabinerS_1 {parent=kin_carabiners_02_l, loc=<0.174 -0.6164 -3.8873>, rot=<-0.4044 -0.5523 0.5061 -0.5246>}
 dyn_hangCarabinerS_1 {parent=base_hangCarabinerS_1, loc=<0 0 0>, rot=<0 0 0 1>}
 dyn_hangCarabinerS_2 {parent=dyn_hangCarabinerS_1, loc=<6.6494 -0.0053 -0.0157>, rot=<0 0.0633 0 -0.998>}
 coll_hem_f_r {parent=pelvis, loc=<-19.5365 5.0036 17.6577>, rot=<-0.4365 -0.3413 -0.1164 -0.8243>}
 target_hem_f_r {parent=pelvis, loc=<-11.3146 10.4464 12.535>, rot=<-0.0127 -0.707 0.0127 -0.707>}
 kin_strapStart_l {parent=pelvis, loc=<-2.7956 10.9807 -3.1745>, rot=<0 0 0 1>}
 kin_strapStart_r {parent=pelvis, loc=<-2.7956 10.9807 3.3659>, rot=<0 0 0 1>}
]
