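# Post-style AnimBlueprint for the IvoryDwell male medium player body's face accessory mesh, animating its hair strand bones.
# It copies the attached parent's pose, then weighted constraints make base strands 7–16 follow the head and nearby facial bones, mostly forehead.
# A chain of AnimDynamics nodes then adds damped, angle-limited spring sway to dynamic strands 1–16, with zero gravity and partial alpha.
# The event graph has no logic.
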AnimBlueprint /BRCosmetics/Characters/Player/Male/Medium/Bodies/M_MED_IvoryDwell/Meshes/Parts/M_MED_IvoryDwell_FaceAcc_AnimBP
target_skeleton=&/M_MED_IvoryDwell_FaceAcc_Skeleton
graphs [
 AnimGraph {nodes [
   n1 = "Constraint - base_hairStrand_7 (PPP)"!(
    ComponentPose=Local To Component(LocalPose=Copy Pose From Mesh() {bUseAttachedParent=true})
    Alpha=1
    ConstraintWeights_0=0.8354
    ConstraintWeights_1:float=0.1439
    ConstraintWeights_2:float=0.0207
   ) -> (Pose) {
    boneToModify {boneName=base_hairStrand_7}
    constraintSetup [{targetBone {boneName=head}}, {targetBone {boneName=FACIAL_R_ForeheadOut}}, {targetBone {boneName=FACIAL_R_ForeheadMid}}]
    constraintWeights [1,1,1]
   }
   n2 = "Constraint - base_hairStrand_8 (PPPPPPPP)"!(
    ComponentPose=@n1
    Alpha=1
    ConstraintWeights_0=0.3137
    ConstraintWeights_1:float=0.2624
    ConstraintWeights_2:float=0.2002
    ConstraintWeights_3:float=0.1032
    ConstraintWeights_4:float=0.0433
    ConstraintWeights_5:float=0.0415
    ConstraintWeights_6:float=0.0193
    ConstraintWeights_7:float=0.0164
   ) -> (Pose) {
    boneToModify {boneName=base_hairStrand_8}
    constraintSetup [
     {targetBone {boneName=FACIAL_L_ForeheadMid}}
     {targetBone {boneName=head}}
     {targetBone {boneName=FACIAL_L_ForeheadIn}}
     {targetBone {boneName=FACIAL_L_ForeheadOut}}
     {targetBone {boneName=FACIAL_C_Forehead}}
     {targetBone {boneName=FACIAL_L_EyesackUpper}}
     {targetBone {boneName=FACIAL_C_NoseBridge}}
     {targetBone {boneName=FACIAL_R_ForeheadIn}}
    ]
    constraintWeights [1,1,1,1,1,1,1,1]
   }
   n3 = "Constraint - base_hairStrand_9 (PPPPPPP)"!(
    ComponentPose=@n2
    Alpha=1
    ConstraintWeights_0=0.4552
    ConstraintWeights_1:float=0.1807
    ConstraintWeights_2:float=0.1605
    ConstraintWeights_3:float=0.0684
    ConstraintWeights_4:float=0.058
    ConstraintWeights_5:float=0.0466
    ConstraintWeights_6:float=0.0306
   ) -> (Pose) {
    boneToModify {boneName=base_hairStrand_9}
    constraintSetup [
     {targetBone {boneName=head}}
     {targetBone {boneName=FACIAL_R_ForeheadMid}}
     {targetBone {boneName=FACIAL_R_ForeheadIn}}
     {targetBone {boneName=FACIAL_C_Forehead}}
     {targetBone {boneName=FACIAL_R_ForeheadOut}}
     {targetBone {boneName=FACIAL_L_ForeheadIn}}
     {targetBone {boneName=FACIAL_C_NoseBridge}}
    ]
    constraintWeights [1,1,1,1,1,1,1]
   }
   n4 = "Constraint - base_hairStrand_10 (PPPP)"!(
    ComponentPose=@n3
    Alpha=1
    ConstraintWeights_0=0.7198
    ConstraintWeights_1:float=0.162
    ConstraintWeights_2:float=0.0757
    ConstraintWeights_3:float=0.0426
   ) -> (Pose) {
    boneToModify {boneName=base_hairStrand_10}
    constraintSetup [
     {targetBone {boneName=head}}
     {targetBone {boneName=FACIAL_L_ForeheadMid}}
     {targetBone {boneName=FACIAL_L_ForeheadOut}}
     {targetBone {boneName=FACIAL_L_ForeheadIn}}
    ]
    constraintWeights [1,1,1,1]
   }
   n5 = "Constraint - base_hairStrand_11 (PPPPPP)"!(
    ComponentPose=@n4
    Alpha=1
    ConstraintWeights_0=0.654
    ConstraintWeights_1:float=0.1594
    ConstraintWeights_2:float=0.0805
    ConstraintWeights_3:float=0.0701
    ConstraintWeights_4:float=0.0228
    ConstraintWeights_5:float=0.0132
   ) -> (Pose) {
    boneToModify {boneName=base_hairStrand_11}
    constraintSetup [
     {targetBone {boneName=head}}
     {targetBone {boneName=FACIAL_R_ForeheadMid}}
     {targetBone {boneName=FACIAL_R_ForeheadIn}}
     {targetBone {boneName=FACIAL_R_ForeheadOut}}
     {targetBone {boneName=FACIAL_C_Forehead}}
     {targetBone {boneName=FACIAL_L_ForeheadIn}}
    ]
    constraintWeights [1,1,1,1,1,1]
   }
   n6 = "Constraint - base_hairStrand_12 (PPPPPPPP)"!(
    ComponentPose=@n5
    Alpha=1
    ConstraintWeights_0=0.4073
    ConstraintWeights_1:float=0.1819
    ConstraintWeights_2:float=0.1593
    ConstraintWeights_3:float=0.0914
    ConstraintWeights_4:float=0.0641
    ConstraintWeights_5:float=0.0428
    ConstraintWeights_6:float=0.0401
    ConstraintWeights_7:float=0.0131
   ) -> (Pose) {
    boneToModify {boneName=base_hairStrand_12}
    constraintSetup [
     {targetBone {boneName=head}}
     {targetBone {boneName=FACIAL_L_ForeheadIn}}
     {targetBone {boneName=FACIAL_L_ForeheadMid}}
     {targetBone {boneName=FACIAL_C_Forehead}}
     {targetBone {boneName=FACIAL_R_ForeheadIn}}
     {targetBone {boneName=FACIAL_C_NoseBridge}}
     {targetBone {boneName=FACIAL_L_ForeheadOut}}
     {targetBone {boneName=FACIAL_R_ForeheadMid}}
    ]
    constraintWeights [1,1,1,1,1,1,1,1]
   }
   n7 = "Constraint - base_hairStrand_13 (PPPP)"!(
    ComponentPose=@n6
    Alpha=1
    ConstraintWeights_0=0.6562
    ConstraintWeights_1:float=0.173
    ConstraintWeights_2:float=0.1546
    ConstraintWeights_3:float=0.0162
   ) -> (Pose) {
    boneToModify {boneName=base_hairStrand_13}
    constraintSetup [
     {targetBone {boneName=head}}
     {targetBone {boneName=FACIAL_L_ForeheadOut}}
     {targetBone {boneName=FACIAL_L_ForeheadMid}}
     {targetBone {boneName=FACIAL_L_ForeheadIn}}
    ]
    constraintWeights [1,1,1,1]
   }
   n8 = "Constraint - base_hairStrand_14 (PPPP)"!(
    ComponentPose=@n7
    Alpha=1
    ConstraintWeights_0=0.8884
    ConstraintWeights_1:float=0.0675
    ConstraintWeights_2:float=0.0264
    ConstraintWeights_3:float=0.0177
   ) -> (Pose) {
    boneToModify {boneName=base_hairStrand_14}
    constraintSetup [
     {targetBone {boneName=head}}
     {targetBone {boneName=FACIAL_L_ForeheadMid}}
     {targetBone {boneName=FACIAL_L_ForeheadOut}}
     {targetBone {boneName=FACIAL_L_ForeheadIn}}
    ]
    constraintWeights [1,1,1,1]
   }
   n9 = "Constraint - base_hairStrand_15 (PPPPPPPP)"!(
    ComponentPose=@n8
    Alpha=1
    ConstraintWeights_0=0.24
    ConstraintWeights_1:float=0.2114
    ConstraintWeights_2:float=0.1769
    ConstraintWeights_3:float=0.1175
    ConstraintWeights_4:float=0.1155
    ConstraintWeights_5:float=0.0925
    ConstraintWeights_6:float=0.0353
    ConstraintWeights_7:float=0.0109
   ) -> (Pose) {
    boneToModify {boneName=base_hairStrand_15}
    constraintSetup [
     {targetBone {boneName=FACIAL_L_ForeheadIn}}
     {targetBone {boneName=head}}
     {targetBone {boneName=FACIAL_C_Forehead}}
     {targetBone {boneName=FACIAL_R_ForeheadIn}}
     {targetBone {boneName=FACIAL_L_ForeheadMid}}
     {targetBone {boneName=FACIAL_C_NoseBridge}}
     {targetBone {boneName=FACIAL_R_ForeheadMid}}
     {targetBone {boneName=FACIAL_L_ForeheadOut}}
    ]
    constraintWeights [1,1,1,1,1,1,1,1]
   }
   n10 = "Constraint - base_hairStrand_16 (PPPP)"!(
    ComponentPose=@n9
    Alpha=1
    ConstraintWeights_0=0.4754
    ConstraintWeights_1:float=0.3946
    ConstraintWeights_2:float=0.115
    ConstraintWeights_3:float=0.0151
   ) -> (Pose) {
    boneToModify {boneName=base_hairStrand_16}
    constraintSetup [
     {targetBone {boneName=head}}
     {targetBone {boneName=FACIAL_L_ForeheadOut}}
     {targetBone {boneName=FACIAL_L_ForeheadMid}}
     {targetBone {boneName=FACIAL_L_EyesackUpper}}
    ]
    constraintWeights [1,1,1,1]
   }
   n11 = Anim Dynamics - Bone: dyn_hairStrand_1!(ComponentPose=@n10, Alpha=0.75) -> (Pose) {
    alpha=0.75
    angularDampingOverride=1
    angularSpringConstant=12000
    bAngularSpring=true
    bOverrideAngularDamping=true
    gravityScale=0
    simulationSpace=RootRelative
    boundBone {boneName=dyn_hairStrand_1}
    componentAppliedLinearAccClamp=<4000 4000 4000>
    componentLinearAccScale=<0.4 0.4 0.4>
    componentLinearVelScale=<0.4 0.4 0.4>
    gravityOverride {z=-980}
    physicsBodyDefinitions [{
      boundBone {boneName=dyn_hairStrand_1}
      boxExtents=<20 30 30>
      constraintSetup {angularLimitsMax=<0 12 12>, angularLimitsMin=<0 -12 -12>, angularTarget=<1 0 0>}
      localJointOffset=<2 0 0>
     }]
   }
   n12 = Anim Dynamics - Bone: dyn_hairStrand_2!(ComponentPose=@n11, Alpha=0.75) -> (Pose) {
    alpha=0.75
    angularDampingOverride=1
    angularSpringConstant=12000
    bAngularSpring=true
    bOverrideAngularDamping=true
    gravityScale=0
    simulationSpace=RootRelative
    boundBone {boneName=dyn_hairStrand_2}
    componentAppliedLinearAccClamp=<4000 4000 4000>
    componentLinearAccScale=<0.4 0.4 0.4>
    componentLinearVelScale=<0.4 0.4 0.4>
    gravityOverride {z=-980}
    physicsBodyDefinitions [{
      boundBone {boneName=dyn_hairStrand_2}
      boxExtents=<20 30 30>
      constraintSetup {angularLimitsMax=<0 12 12>, angularLimitsMin=<0 -12 -12>, angularTarget=<1 0 0>}
      localJointOffset=<2 0 0>
     }]
   }
   n13 = Anim Dynamics - Bone: dyn_hairStrand_3!(ComponentPose=@n12, Alpha=0.75) -> (Pose) {
    alpha=0.75
    angularDampingOverride=1
    angularSpringConstant=12000
    bAngularSpring=true
    bOverrideAngularDamping=true
    gravityScale=0
    simulationSpace=RootRelative
    boundBone {boneName=dyn_hairStrand_3}
    componentAppliedLinearAccClamp=<4000 4000 4000>
    componentLinearAccScale=<0.4 0.4 0.4>
    componentLinearVelScale=<0.4 0.4 0.4>
    gravityOverride {z=-980}
    physicsBodyDefinitions [{
      boundBone {boneName=dyn_hairStrand_3}
      boxExtents=<20 30 30>
      constraintSetup {angularLimitsMax=<0 12 12>, angularLimitsMin=<0 -12 -12>, angularTarget=<1 0 0>}
      localJointOffset=<2 0 0>
     }]
   }
   n14 = Anim Dynamics - Bone: dyn_hairStrand_4!(ComponentPose=@n13, Alpha=0.75) -> (Pose) {
    alpha=0.75
    angularDampingOverride=1
    angularSpringConstant=12000
    bAngularSpring=true
    bOverrideAngularDamping=true
    gravityScale=0
    simulationSpace=RootRelative
    boundBone {boneName=dyn_hairStrand_4}
    componentAppliedLinearAccClamp=<4000 4000 4000>
    componentLinearAccScale=<0.4 0.4 0.4>
    componentLinearVelScale=<0.4 0.4 0.4>
    gravityOverride {z=-980}
    physicsBodyDefinitions [{
      boundBone {boneName=dyn_hairStrand_4}
      boxExtents=<20 30 30>
      constraintSetup {angularLimitsMax=<0 12 12>, angularLimitsMin=<0 -12 -12>, angularTarget=<1 0 0>}
      localJointOffset=<2 0 0>
     }]
   }
   n15 = Anim Dynamics - Bone: dyn_hairStrand_5!(ComponentPose=@n14, Alpha=0.75) -> (Pose) {
    alpha=0.75
    angularDampingOverride=1
    angularSpringConstant=12000
    bAngularSpring=true
    bOverrideAngularDamping=true
    gravityScale=0
    simulationSpace=RootRelative
    boundBone {boneName=dyn_hairStrand_5}
    componentAppliedLinearAccClamp=<4000 4000 4000>
    componentLinearAccScale=<0.4 0.4 0.4>
    componentLinearVelScale=<0.4 0.4 0.4>
    gravityOverride {z=-980}
    physicsBodyDefinitions [{
      boundBone {boneName=dyn_hairStrand_5}
      boxExtents=<20 30 30>
      constraintSetup {angularLimitsMax=<0 12 12>, angularLimitsMin=<0 -12 -12>, angularTarget=<1 0 0>}
      localJointOffset=<2 0 0>
     }]
   }
   n16 = Anim Dynamics - Bone: dyn_hairStrand_6!(ComponentPose=@n15, Alpha=0.75) -> (Pose) {
    alpha=0.75
    angularDampingOverride=1
    angularSpringConstant=12000
    bAngularSpring=true
    bOverrideAngularDamping=true
    gravityScale=0
    simulationSpace=RootRelative
    boundBone {boneName=dyn_hairStrand_6}
    componentAppliedLinearAccClamp=<4000 4000 4000>
    componentLinearAccScale=<0.4 0.4 0.4>
    componentLinearVelScale=<0.4 0.4 0.4>
    gravityOverride {z=-980}
    physicsBodyDefinitions [{
      boundBone {boneName=dyn_hairStrand_6}
      boxExtents=<20 30 30>
      constraintSetup {angularLimitsMax=<0 12 12>, angularLimitsMin=<0 -12 -12>, angularTarget=<1 0 0>}
      localJointOffset=<2 0 0>
     }]
   }
   n17 = Anim Dynamics - Bone: dyn_hairStrand_7!(ComponentPose=@n16, Alpha=0.75) -> (Pose) {
    alpha=0.75
    angularDampingOverride=1
    angularSpringConstant=12000
    bAngularSpring=true
    bOverrideAngularDamping=true
    gravityScale=0
    simulationSpace=RootRelative
    boundBone {boneName=dyn_hairStrand_7}
    componentAppliedLinearAccClamp=<4000 4000 4000>
    componentLinearAccScale=<0.4 0.4 0.4>
    componentLinearVelScale=<0.4 0.4 0.4>
    gravityOverride {z=-980}
    physicsBodyDefinitions [{
      boundBone {boneName=dyn_hairStrand_7}
      boxExtents=<20 30 30>
      constraintSetup {angularLimitsMax=<0 12 12>, angularLimitsMin=<0 -12 -12>, angularTarget=<1 0 0>}
      localJointOffset=<2 0 0>
     }]
   }
   n18 = Anim Dynamics - Bone: dyn_hairStrand_8!(ComponentPose=@n17, Alpha=0.75) -> (Pose) {
    alpha=0.75
    angularDampingOverride=1
    angularSpringConstant=12000
    bAngularSpring=true
    bOverrideAngularDamping=true
    gravityScale=0
    simulationSpace=RootRelative
    boundBone {boneName=dyn_hairStrand_8}
    componentAppliedLinearAccClamp=<4000 4000 4000>
    componentLinearAccScale=<0.4 0.4 0.4>
    componentLinearVelScale=<0.4 0.4 0.4>
    gravityOverride {z=-980}
    physicsBodyDefinitions [{
      boundBone {boneName=dyn_hairStrand_8}
      boxExtents=<20 30 30>
      constraintSetup {angularLimitsMax=<0 12 12>, angularLimitsMin=<0 -12 -12>, angularTarget=<1 0 0>}
      localJointOffset=<2 0 0>
     }]
   }
   n19 = Anim Dynamics - Bone: dyn_hairStrand_9!(ComponentPose=@n18, Alpha=0.75) -> (Pose) {
    alpha=0.75
    angularDampingOverride=1
    angularSpringConstant=12000
    bAngularSpring=true
    bOverrideAngularDamping=true
    gravityScale=0
    simulationSpace=RootRelative
    boundBone {boneName=dyn_hairStrand_9}
    componentAppliedLinearAccClamp=<4000 4000 4000>
    componentLinearAccScale=<0.4 0.4 0.4>
    componentLinearVelScale=<0.4 0.4 0.4>
    gravityOverride {z=-980}
    physicsBodyDefinitions [{
      boundBone {boneName=dyn_hairStrand_9}
      boxExtents=<20 30 30>
      constraintSetup {angularLimitsMax=<0 12 12>, angularLimitsMin=<0 -12 -12>, angularTarget=<1 0 0>}
      localJointOffset=<2 0 0>
     }]
   }
   n20 = Anim Dynamics - Bone: dyn_hairStrand_10!(ComponentPose=@n19, Alpha=0.75) -> (Pose) {
    alpha=0.75
    angularDampingOverride=1
    angularSpringConstant=12000
    bAngularSpring=true
    bOverrideAngularDamping=true
    gravityScale=0
    simulationSpace=RootRelative
    boundBone {boneName=dyn_hairStrand_10}
    componentAppliedLinearAccClamp=<4000 4000 4000>
    componentLinearAccScale=<0.4 0.4 0.4>
    componentLinearVelScale=<0.4 0.4 0.4>
    gravityOverride {z=-980}
    physicsBodyDefinitions [{
      boundBone {boneName=dyn_hairStrand_10}
      boxExtents=<20 30 30>
      constraintSetup {angularLimitsMax=<0 12 12>, angularLimitsMin=<0 -12 -12>, angularTarget=<1 0 0>}
      localJointOffset=<2 0 0>
     }]
   }
   n21 = Anim Dynamics - Bone: dyn_hairStrand_11!(ComponentPose=@n20, Alpha=0.75) -> (Pose) {
    alpha=0.75
    angularDampingOverride=1
    angularSpringConstant=12000
    bAngularSpring=true
    bOverrideAngularDamping=true
    gravityScale=0
    simulationSpace=RootRelative
    boundBone {boneName=dyn_hairStrand_11}
    componentAppliedLinearAccClamp=<4000 4000 4000>
    componentLinearAccScale=<0.4 0.4 0.4>
    componentLinearVelScale=<0.4 0.4 0.4>
    gravityOverride {z=-980}
    physicsBodyDefinitions [{
      boundBone {boneName=dyn_hairStrand_11}
      boxExtents=<20 30 30>
      constraintSetup {angularLimitsMax=<0 12 12>, angularLimitsMin=<0 -12 -12>, angularTarget=<1 0 0>}
      localJointOffset=<2 0 0>
     }]
   }
   n22 = Anim Dynamics - Bone: dyn_hairStrand_12!(ComponentPose=@n21, Alpha=0.75) -> (Pose) {
    alpha=0.75
    angularDampingOverride=1
    angularSpringConstant=12000
    bAngularSpring=true
    bOverrideAngularDamping=true
    gravityScale=0
    simulationSpace=RootRelative
    boundBone {boneName=dyn_hairStrand_12}
    componentAppliedLinearAccClamp=<4000 4000 4000>
    componentLinearAccScale=<0.4 0.4 0.4>
    componentLinearVelScale=<0.4 0.4 0.4>
    gravityOverride {z=-980}
    physicsBodyDefinitions [{
      boundBone {boneName=dyn_hairStrand_12}
      boxExtents=<20 30 30>
      constraintSetup {angularLimitsMax=<0 12 12>, angularLimitsMin=<0 -12 -12>, angularTarget=<1 0 0>}
      localJointOffset=<2 0 0>
     }]
   }
   n23 = Anim Dynamics - Bone: dyn_hairStrand_13!(ComponentPose=@n22, Alpha=0.75) -> (Pose) {
    alpha=0.75
    angularDampingOverride=1
    angularSpringConstant=12000
    bAngularSpring=true
    bOverrideAngularDamping=true
    gravityScale=0
    simulationSpace=RootRelative
    boundBone {boneName=dyn_hairStrand_13}
    componentAppliedLinearAccClamp=<4000 4000 4000>
    componentLinearAccScale=<0.4 0.4 0.4>
    componentLinearVelScale=<0.4 0.4 0.4>
    gravityOverride {z=-980}
    physicsBodyDefinitions [{
      boundBone {boneName=dyn_hairStrand_13}
      boxExtents=<20 30 30>
      constraintSetup {angularLimitsMax=<0 12 12>, angularLimitsMin=<0 -12 -12>, angularTarget=<1 0 0>}
      localJointOffset=<2 0 0>
     }]
   }
   n24 = Anim Dynamics - Bone: dyn_hairStrand_14!(ComponentPose=@n23, Alpha=0.75) -> (Pose) {
    alpha=0.75
    angularDampingOverride=1
    angularSpringConstant=12000
    bAngularSpring=true
    bOverrideAngularDamping=true
    gravityScale=0
    simulationSpace=RootRelative
    boundBone {boneName=dyn_hairStrand_14}
    componentAppliedLinearAccClamp=<4000 4000 4000>
    componentLinearAccScale=<0.4 0.4 0.4>
    componentLinearVelScale=<0.4 0.4 0.4>
    gravityOverride {z=-980}
    physicsBodyDefinitions [{
      boundBone {boneName=dyn_hairStrand_14}
      boxExtents=<20 30 30>
      constraintSetup {angularLimitsMax=<0 12 12>, angularLimitsMin=<0 -12 -12>, angularTarget=<1 0 0>}
      localJointOffset=<2 0 0>
     }]
   }
   n25 = Anim Dynamics - Bone: dyn_hairStrand_15!(ComponentPose=@n24, Alpha=0.75) -> (Pose) {
    alpha=0.75
    angularDampingOverride=1
    angularSpringConstant=12000
    bAngularSpring=true
    bOverrideAngularDamping=true
    gravityScale=0
    simulationSpace=RootRelative
    boundBone {boneName=dyn_hairStrand_15}
    componentAppliedLinearAccClamp=<4000 4000 4000>
    componentLinearAccScale=<0.4 0.4 0.4>
    componentLinearVelScale=<0.4 0.4 0.4>
    gravityOverride {z=-980}
    physicsBodyDefinitions [{
      boundBone {boneName=dyn_hairStrand_15}
      boxExtents=<20 30 30>
      constraintSetup {angularLimitsMax=<0 12 12>, angularLimitsMin=<0 -12 -12>, angularTarget=<1 0 0>}
      localJointOffset=<2 0 0>
     }]
   }
   n26 = Anim Dynamics - Bone: dyn_hairStrand_16!(ComponentPose=@n25, Alpha=0.75) -> (Pose) {
    alpha=0.75
    angularDampingOverride=1
    angularSpringConstant=12000
    bAngularSpring=true
    bOverrideAngularDamping=true
    gravityScale=0
    simulationSpace=RootRelative
    boundBone {boneName=dyn_hairStrand_16}
    componentAppliedLinearAccClamp=<4000 4000 4000>
    componentLinearAccScale=<0.4 0.4 0.4>
    componentLinearVelScale=<0.4 0.4 0.4>
    gravityOverride {z=-980}
    physicsBodyDefinitions [{
      boundBone {boneName=dyn_hairStrand_16}
      boxExtents=<20 30 30>
      constraintSetup {angularLimitsMax=<0 12 12>, angularLimitsMin=<0 -12 -12>, angularTarget=<1 0 0>}
      localJointOffset=<2 0 0>
     }]
   }
   Output Pose(Result=Component To Local(ComponentPose=@n26)) {layerGroup=None}
  ]}
 EventGraph {nodes [Event BlueprintUpdateAnimation()]}
]
variables []
